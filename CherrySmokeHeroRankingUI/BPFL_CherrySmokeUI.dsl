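; Blueprint function library for the CherrySmoke hero ranking UI, converting gameplay tags and a boolean into display values.
; ContextTagToText and RankTagToTexture each read one fixed data table row and look up the tag in its map.
; Failed lookups return "INVALID CONTEXT" text or an empty result; the rank map holds material instances despite the function's name.
; ValidPlayerStateToVisibility maps true to HitTestInvisible and false to Hidden.
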(fn ContextTagToText
  (params (InRankedTag :type Gameplay%20Tag%20Structure))
  (bind _returnvalue (Utilities|GetDataTableRow "/CherrySmokeHeroRankingUI/PlayerRank/DT_CherrySmoke_TagContextTextMap.DT_CherrySmoke_TagContextTextMap" "Map")
    (:then
      (if (.ReturnValue (Utilities|Map|Find (Utilities|Struct|BreakSCherrySmokePointContextTagStringMap _returnvalue) InRankedTag))
        (return (Utilities|Map|Find (Utilities|Struct|BreakSCherrySmokePointContextTagStringMap _returnvalue) InRankedTag))
        (else
          _)))
    (:RowNotFound
      _))
  (return "NSLOCTEXT(\"[4BF44A92C0E354862F98CE4CB0970D06]\", \"DBB17C17460A37FC05B797A621FC09B8\", \"INVALID CONTEXT\")"))

(fn RankTagToTexture
  (params (InRankedTag :type Gameplay%20Tag%20Structure))
  (bind _returnvalue (Utilities|GetDataTableRow "/CherrySmokeHeroRankingUI/Icons/DT_CherrySmoke_TagIconMapData.DT_CherrySmoke_TagIconMapData" "TagTextureRankMap")
    (:then
      (if (.ReturnValue (Utilities|Map|Find (.MaterialInstance_Map_13_52CF24884507C1ACA96D71BE252D9AF2 (Utilities|Struct|BreakSCherrySmokeTagTextureRankMap _returnvalue)) InRankedTag))
        (return (Utilities|Map|Find (.MaterialInstance_Map_13_52CF24884507C1ACA96D71BE252D9AF2 (Utilities|Struct|BreakSCherrySmokeTagTextureRankMap _returnvalue)) InRankedTag))
        (else
          _)))
    (:RowNotFound
      _))
  (return 0))

(fn ValidPlayerStateToVisibility
  (params (InBool :type Boolean))
  (return (select (> InBool 0) "HitTestInvisible" "Hidden")))
