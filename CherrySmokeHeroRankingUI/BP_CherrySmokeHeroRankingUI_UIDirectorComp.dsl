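; UI director component that adds and removes the Cherry Smoke Hero Ranking HUD and post-match placement scenes.
; When ready, it shows the HUD unless a hotfix flag hides it for Starter Island; that branch starts a playlist-ready async with no visible follow-up and enables the post-match scene upon local placement change.
; Delta-gated setters avoid redundant scene changes and choose normal or splitscreen HUD.
; EndPlay deactivates both.
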(event EventOnReadyToShowUI
  ;[1 0.5772 0 0.5] HOTFIX
  ; Hide Hero Ranking widget on Starter Island
  (if (ScalableFloat|GetValueasBool (Variables|Default|GetSF%20Hide%20Hero%20Rank%20on%20Starter%20Island))
    (Utilities|FlowControl|Sequence
      (:then_0
        ;[0.08854 0.564 1 0.5] FORT-888747 - Remove widget when on the starter island
        (Class|FortAsyncActionCurrentPlaylistReady|PlaylistDataReadyAsync))
      (:then_1
        ; Postmatch placement banner Hero Ranking icon
        (GameplayEventRouter|StartListeningToEvent (GameplayEventRouter|GetGlobalEventRouter) "/Script/FortniteGame.OnLocalPlayerPlaceChangedEvent" (CallFunction|GetOwningLocalPlayerController)
          (:OnEventReceived
            (CallFunction|SetHeroRankingPostMatchSceneActive true)))))
    (else
      (CallFunction|SetHeroRankingSceneActive true))))

(event EventEndPlay
  (params (EndPlayReason :type EEndPlayReason%20Enum :default "Destroyed"))
  (CallFunction|SetHeroRankingPostMatchSceneActive false)
  (CallFunction|SetHeroRankingSceneActive false))

(fn SetHeroRankingPostMatchSceneActive
  (params (IsActive :type Boolean))
  (bind _scene_postmatchplacement (Variables|Default|GetScene_PostMatchPlacement))
  (bind _updated_variable (Utilities|FlowControl|DeltaGateWildcard (Variables|Default|GetIsHeroRankingPostMatchSceneActive) IsActive)
    (:Success
      (if _updated_variable
        (CallFunction|AddScene _scene_postmatchplacement)
        (else
          (CallFunction|RemoveScene _scene_postmatchplacement))))))

(fn SetHeroRankingSceneActive
  (params (IsActive :type Boolean))
  (bind _returnvalue (CallFunction|GetOwningLocalPlayerController))
  (bind _updated_variable (Utilities|FlowControl|DeltaGateWildcard (Variables|Default|GetIsHeroRankingHUDSceneActive) IsActive)
    (:Success
      (if _updated_variable
        (bind _asfort_player_controller (Utilities|Casting|CastToFortPlayerController _returnvalue))
        (CallFunction|AddScene (select (PlayerController|IsinSplitscreen _asfort_player_controller) (Variables|Default|GetScene_HeroRanking_HUDSplitscreen) (Variables|Default|GetScene_HeroRankingHUD)))
        (CallFunction|RemoveScene (select (PlayerController|IsinSplitscreen _asfort_player_controller) (Variables|Default|GetScene_HeroRankingHUD) (Variables|Default|GetScene_HeroRanking_HUDSplitscreen)))
        (else
          (bind _asfort_player_controller_2 (Utilities|Casting|CastToFortPlayerController _returnvalue))
          (CallFunction|RemoveScene (select (PlayerController|IsinSplitscreen _asfort_player_controller_2) (Variables|Default|GetScene_HeroRanking_HUDSplitscreen) (Variables|Default|GetScene_HeroRankingHUD))))))))
