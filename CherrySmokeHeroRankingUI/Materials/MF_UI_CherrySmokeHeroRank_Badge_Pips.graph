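# UI material function that draws an animated pip shape, output as a mask, for the CherrySmoke hero rank badge.
# UVs rotate about the center by an angle lerped between Rotation Min and Max using a time-driven sine, which is not remapped, so it swings below Min.
# Position then offsets the UVs horizontally into MF_UI_Box, sized by Thickness and Width.
# The box SDF is inverted and smoothstepped into a tight-edged mask.
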MaterialFunction /CherrySmokeHeroRankingUI/Materials/MF_UI_CherrySmokeHeroRank_Badge_Pips
graph {nodes [
  n1 = TextureCoordinate()
  n2 = Constant() {r=0.5}
  n3 = FunctionInput() {inputName=Rotation Min, inputType=FunctionInput_Scalar, sortPriority=5, previewValue=<0.3 0 0 1>}
  n4 = FunctionInput() {inputName=Rotation Max, inputType=FunctionInput_Scalar, sortPriority=6, previewValue=<0.4 0 0 1>}
  n5 = Time()
  n6 = FunctionInput() {inputName=Speed, inputType=FunctionInput_Scalar, sortPriority=4, previewValue=<0.05 0 0 1>}
  n7 = LinearInterpolate(A=@n3, B=@n4, Alpha=Sine(None=Multiply(A=@n5, B=@n6) {constB=0.2})) {constA=0.3, constB=0.4}
  n8 = FunctionInput() {inputName=Position, inputType=FunctionInput_Scalar, sortPriority=1, previewValue=<0.19 0 0 1>}
  n9 = Constant()
  n10 = FunctionInput() {inputName=Thickness, inputType=FunctionInput_Scalar, sortPriority=2, previewValue=<0.04 0 0 1>}
  n11 = MaterialFunctionCall!(
   UVs=Add(
    A=MaterialFunctionCall(UVs=@n1, Rotation Center=@n2, "Rotation Angle (0-1)"=@n7) {materialFunction=&CustomRotator}.Rotated Values
    B=AppendVector(A=@n8, B=@n9)
   )
   Size=AppendVector(A=@n10, B=FunctionInput() {inputName=Width, inputType=FunctionInput_Scalar, sortPriority=3, previewValue=<0.1 0 0 1>})
   CornerRoundness
   OutlineThickness
   OutlineAlignment
   GlowMin
   GlowMax
   KeepAspectRatio
  ) {materialFunction=&MF_UI_Box}
  FunctionOutput(None=SmoothStep(Value=OneMinus(None=@n11.SDF)) {constMin=0.9}) {outputName=Pips}
 ]}
