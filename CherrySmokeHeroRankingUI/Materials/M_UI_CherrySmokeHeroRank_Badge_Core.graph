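# Translucent UI material that draws the hexagonal core of a Cherry Smoke hero ranking badge.
# A perspective-warped polygon fill is shaded by a radial gradient with separate inner and outer opacities.
# The LinearGradient static switch selects a vertical two-color blend, or a hexagon-SDF color blend plus scrolling segmented lines.
# Those lines come from panned, ping-ponged UVs masked by a circle SDF, and an Opacity parameter scales the combined shape.
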Material /CherrySmokeHeroRankingUI/Materials/M_UI_CherrySmokeHeroRank_Badge_Core
graph {
 properties {bAutomaticallySetUsageInEditor=false, blendMode=BLEND_Translucent, materialDomain=MD_UI, usedShadingModels=Default Lit}
 nodes [
  n1 = ScalarParameter() {defaultValue=0.5, group=Circle, parameterName=CircleSize}
  MaterialFunctionCall(Size=@n1) {materialFunction=&MF_UI_Circle}
  n2 = TextureCoordinate()
  n3 = Constant() {r=0.25}
  n4 = Constant() {r=0.1}
  n5 = Add(
   A=Multiply(
    A=MaterialFunctionCall!(UV=@n2, Perspective Rotation=@n3, Perspective Amount=@n4) {materialFunction=/FortUILibrary/MaterialFunctions/Basics/MF_UI_Perspective}.Perspective UVs
    B=Constant2Vector() {g=0.94, r=1}
   )
   B=Constant2Vector() {g=-0.016}
  )
  n6 = MaterialFunctionCall(
   GradientStart=Add(A=OneMinus(None=@n1), B=ScalarParameter() {defaultValue=0.25, parameterName=AtenuateColor1}) {constB=0.25}
   GradientEnd=OneMinus(None=ScalarParameter() {parameterName=AtenuateColor2})
  ) {materialFunction=&MF_UI_RadialGradient}
  n7 = ScalarParameter() {defaultValue=0.6, parameterName=Color2Opacity}
  n8 = OneMinus(None=@n6.Smooth)
  n9 = ScalarParameter() {defaultValue=0.8, parameterName=Color1Opacity}
  n10 = Multiply(
   A=MaterialFunctionCall!(UVs=@n5, Size=Constant() {r=0.64}, Sides=Constant() {r=6}, InnerAngle, OutlineThickness, OutlineAlignment, GlowMin, GlowMax) {materialFunction=/FortUILibrary/MaterialFunctions/Shapes/MF_UI_RegularPolygon}.Fill
   B=Add(A=Multiply(A=@n6.Smooth, B=@n7), B=Multiply(A=@n8, B=@n9))
  )
  n11 = MaterialFunctionCall() {materialFunction=&GetUserInterfaceUV}
  n12 = Multiply(A=Time(), B=ScalarParameter() {defaultValue=0.8, parameterName=LinesSpeed})
  n13 = ComponentMask(None=Multiply(A=Panner(Coordinate=@n11.Normalized UV, Time=@n12) {speedY=0.1}, B=ScalarParameter() {defaultValue=26, parameterName=SegmentCount})) {g=true}
  n14 = MaterialFunctionCall() {materialFunction=&GetUserInterfaceUV}
  n15 = ScalarParameter() {defaultValue=0.65, parameterName=CircleRadius}
  n16 = ScalarParameter() {defaultValue=0.5, parameterName=CircleDensity}
  n17 = Multiply(
   A=MaterialFunctionCall(
    DistanceField=MaterialFunctionCall!(Value=@n13, Scale, Height) {materialFunction=/FortUILibrary/MaterialFunctions/Basics/MF_UI_PingPong}.Result
    Threshold=Constant() {r=0.4}
    Smoothing=Constant()
   ) {materialFunction=&MF_UI_AntiAliasing}.Inner
   B=Multiply(
    A=MaterialFunctionCall!(UVs=@n14.Normalized UV, Radius=@n15) {materialFunction=/FortUILibrary/MaterialFunctions/Shapes/SDFs/MF_SDF_Circle}.SDF
    B=Multiply(B=@n16) {constA=-1}
   ) {constB=-60}
  )
  Multiply(
   A=StaticSwitchParameter(True=@n10, False=Add(A=@n10, B=@n17)) {parameterName=LinearGradient}
   B=ScalarParameter() {defaultValue=1, parameterName=Opacity}
  )
  n18 = VectorParameter() {group=Color, parameterName=SecondColor, defaultValue=<rgba 0.8714 0.2895 0.2274 1>}
  n19 = VectorParameter() {group=Color, parameterName=Color, defaultValue=<rgba 1 1 1 1>}
  n20 = MaterialFunctionCall(
   UVs=MaterialFunctionCall(UVs=@n5, Texture Scale=Constant2Vector() {g=0.18, r=0.18}) {materialFunction=&ScaleUVsByCenter}.UVs
   Size=Constant() {r=1}
   Sides=Constant() {r=6}
  ) {materialFunction=&MF_UI_RegularPolygon}
  StaticSwitchParameter(
   True=LinearInterpolate(
    A=@n18.RGB
    B=@n19.RGB
    Alpha=MaterialFunctionCall() {materialFunction=&LinearGradient}.VGradient
   )
   False=LinearInterpolate(A=@n19.RGB, B=@n18.RGB, Alpha=OneMinus(None=@n20.SDF))
  ) {parameterName=LinearGradient}
 ]
}
