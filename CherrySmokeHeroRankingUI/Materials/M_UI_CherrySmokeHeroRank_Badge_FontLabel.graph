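# Translucent UI material that draws the rank lettering on a CherrySmoke hero ranking badge.
# A rotated linear gradient blends dark grey toward white for the shading.
# The swappable FontTexture parameter, defaulting to the S+ rank texture, is sampled with scaled and offset UVs to position the glyph.
# Its inverted red channel is treated as a distance field and thresholded by an anti-aliasing function into a smooth-edged mask.
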Material /CherrySmokeHeroRankingUI/Materials/M_UI_CherrySmokeHeroRank_Badge_FontLabel
graph {
 properties {bAutomaticallySetUsageInEditor=false, blendMode=BLEND_Translucent, materialDomain=MD_UI, usedShadingModels=Default Lit}
 nodes [
  LinearInterpolate(
   A=Constant3Vector() {constant=<rgba 0.12 0.12 0.12 1>}
   Alpha=MaterialFunctionCall(GradientStart=Constant() {r=0.15}, GradientEnd=Constant() {r=0.65}, RotationAngle=Constant() {r=0.18}) {materialFunction=&MF_UI_LinearGradient}.Result
  )
  n1 = TextureCoordinate() {uTiling=1.5, vTiling=1.5}
  n2 = ComponentMask(None=@n1) {r=true}
  n3 = ComponentMask(None=@n1) {g=true}
  n4 = TextureSampleParameter2D(UVs=AppendVector(A=Add(A=@n2) {constB=-0.29}, B=Add(A=@n3) {constB=-0.24})) {
   parameterName=FontTexture
   samplerType=SAMPLERTYPE_Grayscale
   texture=/CherrySmokeHeroRankingUI/Textures/T_UI_CherrySmokeHeroRank_Badge_RankSPlus
  }
  MaterialFunctionCall(DistanceField=OneMinus(None=@n4.R), Threshold=Constant() {r=0.5}) {materialFunction=&MF_UI_AntiAliasing}
 ]
}
