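# Translucent UI material that draws a hexagonal ring outline for the Cherry Smoke hero rank badge.
# UVs pass through a perspective function, then a slight vertical scale and offset, before feeding a six-sided polygon function with Size and Thickness parameters.
# The inverted outline distance field runs through an anti-aliasing function and is multiplied by an Opacity parameter, with a separate Color parameter for tint.
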Material /CherrySmokeHeroRankingUI/Materials/M_UI_CherrySmokeHeroRank_Badge_Ring
graph {
 properties {bAutomaticallySetUsageInEditor=false, blendMode=BLEND_Translucent, materialDomain=MD_UI, usedShadingModels=Default Lit}
 nodes [
  n1 = MaterialFunctionCall!(UV, Perspective Rotation=Constant() {r=0.25}, Perspective Amount=Constant() {r=0.1}) {materialFunction=/FortUILibrary/MaterialFunctions/Basics/MF_UI_Perspective}
  n2 = Constant2Vector() {g=0.94, r=1}
  n3 = MaterialFunctionCall!(
   UVs=Add(A=Multiply(A=@n1.Perspective UVs, B=@n2), B=Constant2Vector() {g=-0.016})
   Size=ScalarParameter() {defaultValue=0.5, group=Circle, parameterName=Size}
   Sides=Constant() {r=6}
   InnerAngle
   OutlineThickness=ScalarParameter() {defaultValue=0.04, group=Circle, parameterName=Thickness}
   OutlineAlignment
   GlowMin
   GlowMax
  ) {materialFunction=/FortUILibrary/MaterialFunctions/Shapes/MF_UI_RegularPolygon}
  Multiply(
   A=MaterialFunctionCall(DistanceField=OneMinus(None=@n3.Outline), Threshold=Constant() {r=0.2}) {materialFunction=&MF_UI_AntiAliasing}.Inner
   B=ScalarParameter() {defaultValue=1, group=General, parameterName=Opacity}
  )
  VectorParameter() {group=Color, parameterName=Color, defaultValue=<rgba 1 0 0.9574 1>}
 ]
}
