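# Material function that draws one step of a depth streak for the CherrySmoke hero rank badge UI from an SDF texture.
# Step Index and total depth steps drive the UV rotation, directional offset, start-to-end scale and color blend.
# The transformed UVs sample the SDF, which is anti-aliased into a hard shape and faded by a power falloff.
# It outputs that faded shape both plain and tinted.
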MaterialFunction /CherrySmokeHeroRankingUI/Materials/MF_UI_CherrySmokeHeroRank_FullBadge_Streak
graph {
 properties {previewBlendMode=BLEND_Translucent, previewMaterialDomain=MD_UI}
 nodes [
  n1 = TextureSample(UVs=NamedRerouteUsage(), Tex=NamedRerouteUsage()) {samplerType=SAMPLERTYPE_Grayscale, texture=/CherrySmokeHeroRankingUI/Textures/T_UI_CherrySmokeHeroRank_Badge_RankA}
  n2 = Constant() {r=0.52}
  n3 = Constant() {r=0.01}
  n4 = Divide(A=NamedRerouteUsage(), B=NamedRerouteUsage())
  n5 = Multiply(
   A=NamedRerouteDeclaration(None=MaterialFunctionCall(DistanceField=@n1.RGB, Threshold=@n2, Smoothing=@n3) {materialFunction=&MF_UI_AntiAliasing}.Outer) {name=I_Shape, nodeColor=<rgba 0 0 0 1>}
   B=Power(Base=Subtract(B=@n4), Exp=NamedRerouteUsage())
  )
  FunctionOutput(None=Multiply(A=@n5, B=NamedRerouteUsage())) {outputName="Shape (Color)"}
  FunctionOutput(None=@n5) {outputName=Shape}
  NamedRerouteDeclaration(None=FunctionInput(Preview=Constant()) {inputName=Step Index, inputType=FunctionInput_Scalar, sortPriority=10}) {name=P_Step_Index, nodeColor=<rgba 0 0 0 1>}
  NamedRerouteDeclaration(None=FunctionInput(Preview=TextureCoordinate()) {inputName=UVs, inputType=FunctionInput_Vector2, sortPriority=30}) {name=P_UVs, nodeColor=<rgba 0 0 0 1>}
  n6 = NamedRerouteUsage()
  NamedRerouteDeclaration(None=Divide(A=NamedRerouteUsage(), B=Subtract(A=@n6))) {name=I_Step_T, nodeColor=<rgba 0 0 0 1>}
  NamedRerouteDeclaration(None=FunctionInput(Preview=Constant() {r=0.72}) {inputName=End Scale, sortPriority=120}) {name=I_Step_Scale_End, nodeColor=<rgba 0 0 0 1>}
  NamedRerouteDeclaration(None=LinearInterpolate(A=NamedRerouteUsage(), B=NamedRerouteUsage(), Alpha=NamedRerouteUsage())) {name=I_Step_Scale, nodeColor=<rgba 0 0 0 1>}
  n7 = MaterialFunctionCall(Float2=NamedRerouteUsage()) {materialFunction=&BreakOutFloat2Components}
  n8 = Multiply(A=NamedRerouteUsage(), B=NamedRerouteUsage()) {constA=1}
  n9 = Cosine(None=@n8)
  n10 = Sine(None=@n8)
  n11 = MaterialFunctionCall(X=Subtract(A=Multiply(A=@n7.R, B=@n9), B=Multiply(A=@n7.G, B=@n10)), Y=Add(A=Multiply(A=@n7.R, B=@n10), B=Multiply(A=@n7.G, B=@n9))) {materialFunction=&MakeFloat2}
  n12 = Multiply(A=NamedRerouteUsage(), B=Multiply(A=NamedRerouteUsage(), B=NamedRerouteUsage())) {constA=1}
  NamedRerouteDeclaration(None=Multiply(A=Add(A=@n11.Result, B=@n12), B=NamedRerouteUsage())) {name=I_UVs_Adjusted, nodeColor=<rgba 0 0 0 1>}
  NamedRerouteDeclaration(None=FunctionInput(Preview=TextureObject() {samplerType=SAMPLERTYPE_Grayscale, texture=/CherrySmokeHeroRankingUI/Textures/T_UI_CherrySmokeHeroRank_Badge_RankSPlus}) {inputName=SDF Texture, inputType=FunctionInput_Texture2D}) {name=P_Texture, nodeColor=<rgba 0 0 0 1>}
  NamedRerouteDeclaration(None=LinearInterpolate(A=NamedRerouteUsage(), B=NamedRerouteUsage(), Alpha=NamedRerouteUsage())) {name=I_Color, nodeColor=<rgba 0 0 0 1>}
  NamedRerouteDeclaration(None=FunctionInput(Preview=Constant()) {inputName=Rotation Per Step, inputType=FunctionInput_Scalar, sortPriority=40}) {name=P_Step_Rotation_Per, nodeColor=<rgba 0 0 0 1>}
  NamedRerouteDeclaration(None=FunctionInput(Preview=Constant() {r=4}) {inputName=Falloff Power, inputType=FunctionInput_Scalar, sortPriority=50}) {name=P_Falloff_Power, nodeColor=<rgba 0 0 0 1>}
  NamedRerouteDeclaration(None=FunctionInput(Preview=Constant() {r=16}) {inputName="Depth Steps (Total)", inputType=FunctionInput_Scalar, sortPriority=60}) {name=P_Step_Count_Total, nodeColor=<rgba 0 0 0 1>}
  NamedRerouteDeclaration(None=FunctionInput(Preview=Constant() {r=0.0125}) {inputName=Depth Offset, inputType=FunctionInput_Scalar, sortPriority=70}) {name=P_Step_Offset, nodeColor=<rgba 0 0 0 1>}
  NamedRerouteDeclaration(None=FunctionInput(Preview=Constant() {r=0.5}) {inputName=Depth Offset Dir, inputType=FunctionInput_Vector2, sortPriority=80}) {name=P_Step_Offset_Dir, nodeColor=<rgba 0 0 0 1>}
  NamedRerouteDeclaration(None=FunctionInput(Preview=Constant3Vector() {constant=<rgba 1 0 0 0>}) {inputName=Start Color, sortPriority=90}) {name=P_Color_Start, nodeColor=<rgba 0 0 0 1>}
  NamedRerouteDeclaration(None=FunctionInput(Preview=Constant3Vector() {constant=<rgba 0 1 0 0>}) {inputName=End Color, sortPriority=100}) {name=P_Color_End, nodeColor=<rgba 0 0 0 1>}
  NamedRerouteDeclaration(None=FunctionInput(Preview=Constant() {r=1}) {inputName=Start Scale, sortPriority=110}) {name=I_Step_Scale_Start, nodeColor=<rgba 0 0 0 1>}
 ]
}
