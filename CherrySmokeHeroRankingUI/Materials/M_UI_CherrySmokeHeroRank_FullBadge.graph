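# UI material that draws the full hero rank badge for the CherrySmoke hero ranking UI.
# Sixteen calls to a shared Streak function step an SDF texture through rotation, depth offset, scale and color, with Max chains merging layered results.
# A perspective-warped six-sided polygon SDF, anti-aliased outlines and a radial gradient add further shapes and shading.
# Lerp chains composite layers, with static switches selecting some UV and blend inputs.
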Material /CherrySmokeHeroRankingUI/Materials/M_UI_CherrySmokeHeroRank_FullBadge
graph {
 expressions {
  MaterialExpressionAdd_0 {
   class=MaterialExpressionAdd
   inputs [
    A {from=MaterialExpressionNamedRerouteUsage_19}
    B {from=MaterialExpressionMultiply_2}
   ]
   properties {desc=PolarUV}
  }
  MaterialExpressionAdd_1 {
   class=MaterialExpressionAdd
   inputs [
    A {from=MaterialExpressionNamedRerouteUsage_20}
    B {from=MaterialExpressionNamedRerouteUsage_21}
   ]
   properties {desc=XYUV}
  }
  MaterialExpressionAdd_100 {
   class=MaterialExpressionAdd
   inputs [
    A {from=MaterialExpressionNamedRerouteUsage_146}
    B {from=MaterialExpressionNamedRerouteUsage_45}
   ]
  }
  MaterialExpressionAdd_31 {
   class=MaterialExpressionAdd
   inputs [
    A {from=MaterialExpressionReroute_13}
    B {from=MaterialExpressionMultiply_47}
   ]
  }
  MaterialExpressionAdd_7 {
   class=MaterialExpressionAdd
   inputs [
    A {from=MaterialExpressionMultiply_1}
    B {from=MaterialExpressionConstant2Vector_6}
   ]
  }
  MaterialExpressionClamp_0 {
   class=MaterialExpressionClamp
   inputs [
    None {from=MaterialExpressionScalarParameter_0, output=""}
    Min {from="", output=""}
    Max {from="", output=""}
   ]
   properties {maxDefault=2}
  }
  MaterialExpressionComponentMask_8 {class=MaterialExpressionComponentMask, inputs [None {from=MaterialExpressionTextureSample_5, output=RGB}], properties {g=true, r=true}}
  MaterialExpressionConstant2Vector_0 {class=MaterialExpressionConstant2Vector, properties {g=0.94, r=1}}
  MaterialExpressionConstant2Vector_6 {class=MaterialExpressionConstant2Vector, properties {g=-0.016}}
  MaterialExpressionConstant2Vector_9 {class=MaterialExpressionConstant2Vector, properties {g=0.25}}
  MaterialExpressionConstant_117 {class=MaterialExpressionConstant, properties {r=2}}
  MaterialExpressionConstant_118 {class=MaterialExpressionConstant, properties {r=3}}
  MaterialExpressionConstant_119 {class=MaterialExpressionConstant, properties {r=4}}
  MaterialExpressionConstant_12 {class=MaterialExpressionConstant, properties {r=16}}
  MaterialExpressionConstant_120 {class=MaterialExpressionConstant, properties {r=5}}
  MaterialExpressionConstant_121 {class=MaterialExpressionConstant, properties {r=6}}
  MaterialExpressionConstant_122 {class=MaterialExpressionConstant, properties {r=7}}
  MaterialExpressionConstant_123 {class=MaterialExpressionConstant, properties {r=8}}
  MaterialExpressionConstant_124 {class=MaterialExpressionConstant, properties {r=13}}
  MaterialExpressionConstant_125 {class=MaterialExpressionConstant, properties {r=14}}
  MaterialExpressionConstant_126 {class=MaterialExpressionConstant, properties {r=15}}
  MaterialExpressionConstant_127 {class=MaterialExpressionConstant, properties {r=16}}
  MaterialExpressionConstant_128 {class=MaterialExpressionConstant, properties {r=9}}
  MaterialExpressionConstant_129 {class=MaterialExpressionConstant, properties {r=10}}
  MaterialExpressionConstant_130 {class=MaterialExpressionConstant, properties {r=11}}
  MaterialExpressionConstant_131 {class=MaterialExpressionConstant, properties {r=12}}
  MaterialExpressionConstant_157 {class=MaterialExpressionConstant, properties {r=0.01}}
  MaterialExpressionConstant_158 {class=MaterialExpressionConstant, properties {r=0.52}}
  MaterialExpressionConstant_17 {class=MaterialExpressionConstant}
  MaterialExpressionConstant_174 {class=MaterialExpressionConstant, properties {r=0.64}}
  MaterialExpressionConstant_2 {class=MaterialExpressionConstant, properties {r=0.25}}
  MaterialExpressionConstant_20 {class=MaterialExpressionConstant}
  MaterialExpressionConstant_22 {class=MaterialExpressionConstant}
  MaterialExpressionConstant_24 {class=MaterialExpressionConstant}
  MaterialExpressionConstant_25 {class=MaterialExpressionConstant}
  MaterialExpressionConstant_3 {class=MaterialExpressionConstant, properties {r=0.1}}
  MaterialExpressionConstant_34 {class=MaterialExpressionConstant, properties {r=1}}
  MaterialExpressionConstant_43 {class=MaterialExpressionConstant, properties {r=6}}
  MaterialExpressionConstant_44 {class=MaterialExpressionConstant, properties {r=0.6}}
  MaterialExpressionConstant_48 {class=MaterialExpressionConstant, properties {r=0.1}}
  MaterialExpressionConstant_49 {class=MaterialExpressionConstant, properties {r=0.004}}
  MaterialExpressionConstant_51 {class=MaterialExpressionConstant, properties {r=0.022}}
  MaterialExpressionConstant_54 {class=MaterialExpressionConstant, properties {r=8.2}}
  MaterialExpressionConstant_55 {class=MaterialExpressionConstant, properties {r=0.004}}
  MaterialExpressionConstant_56 {class=MaterialExpressionConstant, properties {r=0.04}}
  MaterialExpressionConstant_57 {class=MaterialExpressionConstant, properties {r=7.7}}
  MaterialExpressionConstant_58 {class=MaterialExpressionConstant, properties {r=0.012}}
  MaterialExpressionConstant_59 {class=MaterialExpressionConstant, properties {r=0.004}}
  MaterialExpressionConstant_60 {class=MaterialExpressionConstant, properties {r=70}}
  MaterialExpressionConstant_61 {class=MaterialExpressionConstant, properties {r=0.001}}
  MaterialExpressionConstant_7 {class=MaterialExpressionConstant, properties {r=0.28}}
  MaterialExpressionConvert_1 {
   class=MaterialExpressionConvert
   inputs [
    None {from=MaterialExpressionCosine_2}
    None {from=MaterialExpressionSine_2}
   ]
   properties {
    nodeName=Make Float2
    convertInputs [
     {
      type=Scalar
      expressionInput {
       expression=&self:MaterialExpressionCosine_2
       inputName=None
       mask=0
       maskA=0
       maskB=0
       maskG=0
       maskR=0
       outputIndex=0
      }
     }
     {
      type=Scalar
      expressionInput {
       expression=&self:MaterialExpressionSine_2
       inputName=None
       mask=0
       maskA=0
       maskB=0
       maskG=0
       maskR=0
       outputIndex=0
      }
     }
    ]
    convertOutputs [{type=Vector2}]
   }
  }
  MaterialExpressionCosine_0 {class=MaterialExpressionCosine, inputs [None {from=MaterialExpressionNamedRerouteUsage_9, output=""}], properties {desc=X}}
  MaterialExpressionCosine_2 {class=MaterialExpressionCosine, inputs [None {from=MaterialExpressionMaterialFunctionCall_8, output=R}], properties {}}
  MaterialExpressionLinearInterpolate_0 {
   class=MaterialExpressionLinearInterpolate
   inputs [
    A {from=""}
    B {from=MaterialExpressionMultiply_26}
    Alpha {from=MaterialExpressionMultiply_27}
   ]
  }
  MaterialExpressionLinearInterpolate_1 {
   class=MaterialExpressionLinearInterpolate
   inputs [
    A {from=MaterialExpressionReroute_5}
    B {from=MaterialExpressionMultiply_4}
    Alpha {from=MaterialExpressionStaticSwitch_7}
   ]
  }
  MaterialExpressionLinearInterpolate_11 {
   class=MaterialExpressionLinearInterpolate
   inputs [
    A {from=MaterialExpressionReroute_56}
    B {from=""}
    Alpha {from=MaterialExpressionMultiply_21}
   ]
  }
  MaterialExpressionLinearInterpolate_12 {
   class=MaterialExpressionLinearInterpolate
   inputs [
    A {from=MaterialExpressionReroute_61}
    B {from=""}
    Alpha {from=MaterialExpressionMultiply_22}
   ]
  }
  MaterialExpressionLinearInterpolate_13 {
   class=MaterialExpressionLinearInterpolate
   inputs [
    A {from=MaterialExpressionReroute_51}
    B {from=""}
    Alpha {from=MaterialExpressionMultiply_20}
   ]
  }
  MaterialExpressionLinearInterpolate_14 {
   class=MaterialExpressionLinearInterpolate
   inputs [
    A {from=MaterialExpressionReroute_46}
    B {from=""}
    Alpha {from=MaterialExpressionMultiply_19}
   ]
  }
  MaterialExpressionLinearInterpolate_15 {
   class=MaterialExpressionLinearInterpolate
   inputs [
    A {from=MaterialExpressionReroute_66}
    B {from=""}
    Alpha {from=MaterialExpressionMultiply_23}
   ]
  }
  MaterialExpressionLinearInterpolate_17 {
   class=MaterialExpressionLinearInterpolate
   inputs [
    A {from=""}
    B {from=MaterialExpressionMultiply_24}
    Alpha {from=MaterialExpressionMultiply_25}
   ]
  }
  MaterialExpressionLinearInterpolate_18 {
   class=MaterialExpressionLinearInterpolate
   inputs [
    A {from=MaterialExpressionReroute_76}
    B {from=MaterialExpressionMultiply_26}
    Alpha {from=MaterialExpressionMultiply_27}
   ]
  }
  MaterialExpressionLinearInterpolate_19 {
   class=MaterialExpressionLinearInterpolate
   inputs [
    A {from=MaterialExpressionNamedRerouteDeclaration_99}
    B {from=MaterialExpressionLinearInterpolate_18}
    Alpha {from=MaterialExpressionNamedRerouteUsage_87}
   ]
  }
  MaterialExpressionLinearInterpolate_2 {
   class=MaterialExpressionLinearInterpolate
   inputs [
    A {from=MaterialExpressionReroute_9}
    B {from=""}
    Alpha {from=MaterialExpressionStaticSwitch_8}
   ]
  }
  MaterialExpressionLinearInterpolate_20 {
   class=MaterialExpressionLinearInterpolate
   inputs [
    A {from=""}
    B {from=MaterialExpressionMultiply_28}
    Alpha {from=MaterialExpressionMultiply_29}
   ]
  }
  MaterialExpressionLinearInterpolate_21 {
   class=MaterialExpressionLinearInterpolate
   inputs [
    A {from=""}
    B {from=MaterialExpressionMultiply_30}
    Alpha {from=MaterialExpressionMultiply_31}
   ]
  }
  MaterialExpressionLinearInterpolate_22 {
   class=MaterialExpressionLinearInterpolate
   inputs [
    A {from=MaterialExpressionNamedRerouteDeclaration_102}
    B {from=""}
    Alpha {from=MaterialExpressionNamedRerouteDeclaration_103}
   ]
  }
  MaterialExpressionLinearInterpolate_23 {
   class=MaterialExpressionLinearInterpolate
   inputs [
    A {from=""}
    B {from=MaterialExpressionMultiply_32}
    Alpha {from=MaterialExpressionMultiply_33}
   ]
  }
  MaterialExpressionLinearInterpolate_24 {
   class=MaterialExpressionLinearInterpolate
   inputs [
    A {from=""}
    B {from=MaterialExpressionMultiply_34}
    Alpha {from=MaterialExpressionMultiply_35}
   ]
  }
  MaterialExpressionLinearInterpolate_25 {
   class=MaterialExpressionLinearInterpolate
   inputs [
    A {from=MaterialExpressionNamedRerouteDeclaration_104}
    B {from=""}
    Alpha {from=MaterialExpressionNamedRerouteDeclaration_105}
   ]
  }
  MaterialExpressionLinearInterpolate_26 {
   class=MaterialExpressionLinearInterpolate
   inputs [
    A {from=""}
    B {from=MaterialExpressionMultiply_36}
    Alpha {from=MaterialExpressionMultiply_37}
   ]
  }
  MaterialExpressionLinearInterpolate_27 {
   class=MaterialExpressionLinearInterpolate
   inputs [
    A {from=""}
    B {from=MaterialExpressionMultiply_38}
    Alpha {from=MaterialExpressionMultiply_39}
   ]
  }
  MaterialExpressionLinearInterpolate_28 {
   class=MaterialExpressionLinearInterpolate
   inputs [
    A {from=MaterialExpressionNamedRerouteDeclaration_108}
    B {from=""}
    Alpha {from=MaterialExpressionNamedRerouteDeclaration_109}
   ]
  }
  MaterialExpressionLinearInterpolate_29 {
   class=MaterialExpressionLinearInterpolate
   inputs [
    A {from=MaterialExpressionReroute_24}
    B {from=MaterialExpressionMultiply_43}
    Alpha {from=MaterialExpressionMultiply_44}
   ]
  }
  MaterialExpressionLinearInterpolate_3 {
   class=MaterialExpressionLinearInterpolate
   inputs [
    A {from=MaterialExpressionReroute_21}
    B {from=MaterialExpressionMultiply_41}
    Alpha {from=MaterialExpressionMultiply_42}
   ]
  }
  MaterialExpressionLinearInterpolate_30 {
   class=MaterialExpressionLinearInterpolate
   inputs [
    A {from=MaterialExpressionReroute_27}
    B {from=MaterialExpressionMultiply_45}
    Alpha {from=MaterialExpressionMultiply_46}
   ]
  }
  MaterialExpressionLinearInterpolate_31 {
   class=MaterialExpressionLinearInterpolate
   inputs [
    A {from=""}
    B {from=""}
    Alpha {from=MaterialExpressionMultiply_194}
   ]
  }
  MaterialExpressionLinearInterpolate_40 {
   class=MaterialExpressionLinearInterpolate
   inputs [
    A {from=MaterialExpressionReroute_41}
    B {from=MaterialExpressionMultiply_303}
    Alpha {from=MaterialExpressionMultiply_304}
   ]
  }
  MaterialExpressionLinearInterpolate_41 {
   class=MaterialExpressionLinearInterpolate
   inputs [
    A {from=MaterialExpressionReroute_71}
    B {from=""}
    Alpha {from=MaterialExpressionMultiply_302}
   ]
  }
  MaterialExpressionLinearInterpolate_6 {
   class=MaterialExpressionLinearInterpolate
   inputs [
    A {from=MaterialExpressionNamedRerouteUsage_2}
    B {from=MaterialExpressionNamedRerouteUsage_1}
    Alpha {from=MaterialExpressionMaterialFunctionCall_36, output=R}
   ]
  }
  MaterialExpressionLinearInterpolate_7 {
   class=MaterialExpressionLinearInterpolate
   inputs [
    A {from=""}
    B {from=""}
    Alpha {from=MaterialExpressionScalarParameter_28}
   ]
   properties {constA=0.2, constB=3}
  }
  MaterialExpressionLinearInterpolate_8 {
   class=MaterialExpressionLinearInterpolate
   inputs [
    A {from=MaterialExpressionReroute_36}
    B {from=MaterialExpressionMultiply_102}
    Alpha {from=MaterialExpressionMultiply_7}
   ]
  }
  MaterialExpressionMaterialFunctionCall_0 {
   class=MaterialExpressionMaterialFunctionCall
   inputs [
    UV {from=MaterialExpressionStaticSwitch_0}
    GradientStart {from=MaterialExpressionNamedRerouteUsage_23}
    GradientEnd {from=MaterialExpressionNamedRerouteUsage_25}
   ]
   properties {materialFunction=&MF_UI_RadialGradient}
  }
  MaterialExpressionMaterialFunctionCall_1 {
   class=MaterialExpressionMaterialFunctionCall
   inputs [
    DistanceField {from=MaterialExpressionNamedRerouteUsage_13}
    Threshold {from=MaterialExpressionConstant_7}
    Smoothing {from=""}
   ]
   properties {materialFunction=&MF_UI_AntiAliasing}
  }
  MaterialExpressionMaterialFunctionCall_128 {
   class=MaterialExpressionMaterialFunctionCall
   inputs [
    DistanceField {from=MaterialExpressionNamedRerouteUsage_58}
    Threshold {from=MaterialExpressionConstant_158}
    Smoothing {from=MaterialExpressionConstant_157}
   ]
   properties {materialFunction=&MF_UI_AntiAliasing}
  }
  MaterialExpressionMaterialFunctionCall_129 {
   class=MaterialExpressionMaterialFunctionCall
   inputs [
    UVs {from=MaterialExpressionAdd_100}
    Texture Scale {from=MaterialExpressionConstant_174}
   ]
   properties {materialFunction=&ScaleUVsByCenter}
  }
  MaterialExpressionMaterialFunctionCall_143 {
   class=MaterialExpressionMaterialFunctionCall
   inputs [
    X {from=MaterialExpressionScalarParameter_72}
    Y {from=MaterialExpressionScalarParameter_73}
   ]
   properties {materialFunction=&MakeFloat2}
  }
  MaterialExpressionMaterialFunctionCall_2 {
   class=MaterialExpressionMaterialFunctionCall
   inputs [
    X {from=MaterialExpressionCosine_0}
    Y {from=MaterialExpressionSine_0}
   ]
   properties {desc=Dir, materialFunction=&MakeFloat2}
  }
  MaterialExpressionMaterialFunctionCall_23 {
   class=MaterialExpressionMaterialFunctionCall
   inputs [
    UVs {from=MaterialExpressionAdd_7}
    Size {from=MaterialExpressionConstant_44}
    Sides {from=MaterialExpressionConstant_43}
    InnerAngle {from=""}
    OutlineThickness {from=""}
    OutlineAlignment {from=""}
    GlowMin {from=""}
    GlowMax {from=""}
   ]
   properties {materialFunction=/FortUILibrary/MaterialFunctions/Shapes/MF_UI_RegularPolygon}
  }
  MaterialExpressionMaterialFunctionCall_24 {
   class=MaterialExpressionMaterialFunctionCall
   inputs [
    DistanceField {from=MaterialExpressionNamedRerouteUsage_18}
    Threshold {from=MaterialExpressionConstant_48}
    Smoothing {from=""}
   ]
   properties {materialFunction=&MF_UI_AntiAliasing}
  }
  MaterialExpressionMaterialFunctionCall_27 {
   class=MaterialExpressionMaterialFunctionCall
   inputs [
    DistanceField {from=MaterialExpressionMaterialFunctionCall_29, output=SDFOutline}
    Threshold {from=MaterialExpressionConstant_49}
    Smoothing {from=""}
   ]
   properties {materialFunction=&MF_UI_AntiAliasing}
  }
  MaterialExpressionMaterialFunctionCall_29 {
   class=MaterialExpressionMaterialFunctionCall
   inputs [
    DistanceField {from=MaterialExpressionNamedRerouteUsage_29}
    Threshold {from=""}
    Thickness {from=MaterialExpressionConstant_51}
    Alignment {from=MaterialExpressionConstant_54}
    Smoothing {from=""}
   ]
   properties {materialFunction=/FortUILibrary/MaterialFunctions/Shapes/Modifiers/MF_UI_ShapeOutline}
  }
  MaterialExpressionMaterialFunctionCall_3 {
   class=MaterialExpressionMaterialFunctionCall
   inputs [
    X {from=MaterialExpressionScalarParameter_23}
    Y {from=MaterialExpressionScalarParameter_24}
   ]
   properties {materialFunction=&MakeFloat2}
  }
  MaterialExpressionMaterialFunctionCall_31 {
   class=MaterialExpressionMaterialFunctionCall
   inputs [
    DistanceField {from=MaterialExpressionMaterialFunctionCall_32, output=SDFOutline}
    Threshold {from=MaterialExpressionConstant_55}
    Smoothing {from=""}
   ]
   properties {materialFunction=&MF_UI_AntiAliasing}
  }
  MaterialExpressionMaterialFunctionCall_32 {
   class=MaterialExpressionMaterialFunctionCall
   inputs [
    DistanceField {from=MaterialExpressionNamedRerouteUsage_14}
    Threshold {from=""}
    Thickness {from=MaterialExpressionConstant_56}
    Alignment {from=MaterialExpressionConstant_57}
    Smoothing {from=""}
   ]
   properties {materialFunction=&MF_UI_ShapeOutline}
  }
  MaterialExpressionMaterialFunctionCall_33 {
   class=MaterialExpressionMaterialFunctionCall
   inputs [
    DistanceField {from=MaterialExpressionMaterialFunctionCall_34, output=SDFOutline}
    Threshold {from=MaterialExpressionConstant_58}
    Smoothing {from=MaterialExpressionConstant_61}
   ]
   properties {materialFunction=&MF_UI_AntiAliasing}
  }
  MaterialExpressionMaterialFunctionCall_34 {
   class=MaterialExpressionMaterialFunctionCall
   inputs [
    DistanceField {from=MaterialExpressionNamedRerouteUsage_15}
    Threshold {from=""}
    Thickness {from=MaterialExpressionConstant_59}
    Alignment {from=MaterialExpressionConstant_60}
    Smoothing {from=""}
   ]
   properties {materialFunction=&MF_UI_ShapeOutline}
  }
  MaterialExpressionMaterialFunctionCall_36 {
   class=MaterialExpressionMaterialFunctionCall
   inputs [Float2 {from=MaterialExpressionMultiply_12}]
   properties {materialFunction=&BreakOutFloat2Components}
  }
  MaterialExpressionMaterialFunctionCall_37 {
   class=MaterialExpressionMaterialFunctionCall
   inputs [
    UVs {from=MaterialExpressionNamedRerouteUsage_47}
    Rotation Center {from=""}
    "Rotation Angle (0-1)" {from=MaterialExpressionNamedRerouteUsage_10}
   ]
   properties {materialFunction=&CustomRotator}
  }
  MaterialExpressionMaterialFunctionCall_4 {
   class=MaterialExpressionMaterialFunctionCall
   inputs [
    UV {from=MaterialExpressionNamedRerouteUsage_46}
    Perspective Rotation {from=MaterialExpressionConstant_2}
    Perspective Amount {from=MaterialExpressionConstant_3}
   ]
   properties {materialFunction=/FortUILibrary/MaterialFunctions/Basics/MF_UI_Perspective}
  }
  MaterialExpressionMaterialFunctionCall_5 {
   class=MaterialExpressionMaterialFunctionCall
   inputs [
    X {from=MaterialExpressionScalarParameter_19}
    Y {from=MaterialExpressionScalarParameter_20}
   ]
   properties {materialFunction=&MakeFloat2}
  }
  MaterialExpressionMaterialFunctionCall_8 {
   class=MaterialExpressionMaterialFunctionCall
   inputs [Float2 {from=MaterialExpressionNamedRerouteUsage_150}]
   properties {materialFunction=&BreakOutFloat2Components}
  }
  MaterialExpressionMaterialFunctionCall_80 {
   class=MaterialExpressionMaterialFunctionCall
   inputs [
    SDF Texture {from=MaterialExpressionNamedRerouteUsage_271}
    Step Index {from=MaterialExpressionConstant_34}
    UVs {from=MaterialExpressionNamedRerouteUsage_207}
    Rotation Per Step {from=MaterialExpressionNamedRerouteUsage_206}
    Falloff Power {from=MaterialExpressionNamedRerouteUsage_211}
    "Depth Steps (Total)" {from=MaterialExpressionNamedRerouteUsage_209}
    Depth Offset {from=MaterialExpressionNamedRerouteUsage_208}
    Depth Offset Dir {from=MaterialExpressionNamedRerouteUsage_205}
    Start Color {from=MaterialExpressionNamedRerouteUsage_11}
    End Color {from=MaterialExpressionNamedRerouteUsage_8}
    Start Scale {from=MaterialExpressionNamedRerouteUsage_523}
    End Scale {from=MaterialExpressionNamedRerouteUsage_524}
   ]
   properties {materialFunction=&/MF_UI_CherrySmokeHeroRank_FullBadge_Streak}
  }
  MaterialExpressionMaterialFunctionCall_81 {
   class=MaterialExpressionMaterialFunctionCall
   inputs [
    SDF Texture {from=MaterialExpressionNamedRerouteUsage_248}
    Step Index {from=MaterialExpressionConstant_117}
    UVs {from=MaterialExpressionNamedRerouteUsage_244}
    Rotation Per Step {from=MaterialExpressionNamedRerouteUsage_243}
    Falloff Power {from=MaterialExpressionNamedRerouteUsage_247}
    "Depth Steps (Total)" {from=MaterialExpressionNamedRerouteUsage_246}
    Depth Offset {from=MaterialExpressionNamedRerouteUsage_245}
    Depth Offset Dir {from=MaterialExpressionNamedRerouteUsage_242}
    Start Color {from=MaterialExpressionNamedRerouteUsage_250}
    End Color {from=MaterialExpressionNamedRerouteUsage_249}
    Start Scale {from=MaterialExpressionNamedRerouteUsage_525}
    End Scale {from=MaterialExpressionNamedRerouteUsage_526}
   ]
   properties {materialFunction=&MF_UI_CherrySmokeHeroRank_FullBadge_Streak}
  }
  MaterialExpressionMaterialFunctionCall_82 {
   class=MaterialExpressionMaterialFunctionCall
   inputs [
    SDF Texture {from=MaterialExpressionNamedRerouteUsage_257}
    Step Index {from=MaterialExpressionConstant_118}
    UVs {from=MaterialExpressionNamedRerouteUsage_253}
    Rotation Per Step {from=MaterialExpressionNamedRerouteUsage_252}
    Falloff Power {from=MaterialExpressionNamedRerouteUsage_256}
    "Depth Steps (Total)" {from=MaterialExpressionNamedRerouteUsage_255}
    Depth Offset {from=MaterialExpressionNamedRerouteUsage_254}
    Depth Offset Dir {from=MaterialExpressionNamedRerouteUsage_251}
    Start Color {from=MaterialExpressionNamedRerouteUsage_259}
    End Color {from=MaterialExpressionNamedRerouteUsage_258}
    Start Scale {from=MaterialExpressionNamedRerouteUsage_527}
    End Scale {from=MaterialExpressionNamedRerouteUsage_528}
   ]
   properties {materialFunction=&MF_UI_CherrySmokeHeroRank_FullBadge_Streak}
  }
  MaterialExpressionMaterialFunctionCall_83 {
   class=MaterialExpressionMaterialFunctionCall
   inputs [
    SDF Texture {from=MaterialExpressionNamedRerouteUsage_266}
    Step Index {from=MaterialExpressionConstant_119}
    UVs {from=MaterialExpressionNamedRerouteUsage_262}
    Rotation Per Step {from=MaterialExpressionNamedRerouteUsage_261}
    Falloff Power {from=MaterialExpressionNamedRerouteUsage_265}
    "Depth Steps (Total)" {from=MaterialExpressionNamedRerouteUsage_264}
    Depth Offset {from=MaterialExpressionNamedRerouteUsage_263}
    Depth Offset Dir {from=MaterialExpressionNamedRerouteUsage_260}
    Start Color {from=MaterialExpressionNamedRerouteUsage_268}
    End Color {from=MaterialExpressionNamedRerouteUsage_267}
    Start Scale {from=MaterialExpressionNamedRerouteUsage_529}
    End Scale {from=MaterialExpressionNamedRerouteUsage_530}
   ]
   properties {materialFunction=&MF_UI_CherrySmokeHeroRank_FullBadge_Streak}
  }
  MaterialExpressionMaterialFunctionCall_84 {
   class=MaterialExpressionMaterialFunctionCall
   inputs [
    SDF Texture {from=MaterialExpressionNamedRerouteUsage_398}
    Step Index {from=MaterialExpressionConstant_120}
    UVs {from=MaterialExpressionNamedRerouteUsage_272}
    Rotation Per Step {from=MaterialExpressionNamedRerouteUsage_270}
    Falloff Power {from=MaterialExpressionNamedRerouteUsage_397}
    "Depth Steps (Total)" {from=MaterialExpressionNamedRerouteUsage_396}
    Depth Offset {from=MaterialExpressionNamedRerouteUsage_395}
    Depth Offset Dir {from=MaterialExpressionNamedRerouteUsage_269}
    Start Color {from=MaterialExpressionNamedRerouteUsage_400}
    End Color {from=MaterialExpressionNamedRerouteUsage_399}
    Start Scale {from=MaterialExpressionNamedRerouteUsage_531}
    End Scale {from=MaterialExpressionNamedRerouteUsage_532}
   ]
   properties {materialFunction=&MF_UI_CherrySmokeHeroRank_FullBadge_Streak}
  }
  MaterialExpressionMaterialFunctionCall_85 {
   class=MaterialExpressionMaterialFunctionCall
   inputs [
    SDF Texture {from=MaterialExpressionNamedRerouteUsage_407}
    Step Index {from=MaterialExpressionConstant_121}
    UVs {from=MaterialExpressionNamedRerouteUsage_403}
    Rotation Per Step {from=MaterialExpressionNamedRerouteUsage_402}
    Falloff Power {from=MaterialExpressionNamedRerouteUsage_406}
    "Depth Steps (Total)" {from=MaterialExpressionNamedRerouteUsage_405}
    Depth Offset {from=MaterialExpressionNamedRerouteUsage_404}
    Depth Offset Dir {from=MaterialExpressionNamedRerouteUsage_401}
    Start Color {from=MaterialExpressionNamedRerouteUsage_409}
    End Color {from=MaterialExpressionNamedRerouteUsage_408}
    Start Scale {from=MaterialExpressionNamedRerouteUsage_534}
    End Scale {from=MaterialExpressionNamedRerouteUsage_533}
   ]
   properties {materialFunction=&MF_UI_CherrySmokeHeroRank_FullBadge_Streak}
  }
  MaterialExpressionMaterialFunctionCall_86 {
   class=MaterialExpressionMaterialFunctionCall
   inputs [
    SDF Texture {from=MaterialExpressionNamedRerouteUsage_416}
    Step Index {from=MaterialExpressionConstant_122}
    UVs {from=MaterialExpressionNamedRerouteUsage_412}
    Rotation Per Step {from=MaterialExpressionNamedRerouteUsage_411}
    Falloff Power {from=MaterialExpressionNamedRerouteUsage_415}
    "Depth Steps (Total)" {from=MaterialExpressionNamedRerouteUsage_414}
    Depth Offset {from=MaterialExpressionNamedRerouteUsage_413}
    Depth Offset Dir {from=MaterialExpressionNamedRerouteUsage_410}
    Start Color {from=MaterialExpressionNamedRerouteUsage_418}
    End Color {from=MaterialExpressionNamedRerouteUsage_417}
    Start Scale {from=MaterialExpressionNamedRerouteUsage_536}
    End Scale {from=MaterialExpressionNamedRerouteUsage_535}
   ]
   properties {materialFunction=&MF_UI_CherrySmokeHeroRank_FullBadge_Streak}
  }
  MaterialExpressionMaterialFunctionCall_87 {
   class=MaterialExpressionMaterialFunctionCall
   inputs [
    SDF Texture {from=MaterialExpressionNamedRerouteUsage_425}
    Step Index {from=MaterialExpressionConstant_123}
    UVs {from=MaterialExpressionNamedRerouteUsage_421}
    Rotation Per Step {from=MaterialExpressionNamedRerouteUsage_420}
    Falloff Power {from=MaterialExpressionNamedRerouteUsage_424}
    "Depth Steps (Total)" {from=MaterialExpressionNamedRerouteUsage_423}
    Depth Offset {from=MaterialExpressionNamedRerouteUsage_422}
    Depth Offset Dir {from=MaterialExpressionNamedRerouteUsage_419}
    Start Color {from=MaterialExpressionNamedRerouteUsage_427}
    End Color {from=MaterialExpressionNamedRerouteUsage_426}
    Start Scale {from=MaterialExpressionNamedRerouteUsage_538}
    End Scale {from=MaterialExpressionNamedRerouteUsage_537}
   ]
   properties {materialFunction=&MF_UI_CherrySmokeHeroRank_FullBadge_Streak}
  }
  MaterialExpressionMaterialFunctionCall_88 {
   class=MaterialExpressionMaterialFunctionCall
   inputs [
    SDF Texture {from=MaterialExpressionNamedRerouteUsage_434}
    Step Index {from=MaterialExpressionConstant_124}
    UVs {from=MaterialExpressionNamedRerouteUsage_430}
    Rotation Per Step {from=MaterialExpressionNamedRerouteUsage_429}
    Falloff Power {from=MaterialExpressionNamedRerouteUsage_433}
    "Depth Steps (Total)" {from=MaterialExpressionNamedRerouteUsage_432}
    Depth Offset {from=MaterialExpressionNamedRerouteUsage_431}
    Depth Offset Dir {from=MaterialExpressionNamedRerouteUsage_428}
    Start Color {from=MaterialExpressionNamedRerouteUsage_436}
    End Color {from=MaterialExpressionNamedRerouteUsage_435}
    Start Scale {from=MaterialExpressionNamedRerouteUsage_547}
    End Scale {from=MaterialExpressionNamedRerouteUsage_548}
   ]
   properties {materialFunction=&MF_UI_CherrySmokeHeroRank_FullBadge_Streak}
  }
  MaterialExpressionMaterialFunctionCall_89 {
   class=MaterialExpressionMaterialFunctionCall
   inputs [
    SDF Texture {from=MaterialExpressionNamedRerouteUsage_443}
    Step Index {from=MaterialExpressionConstant_125}
    UVs {from=MaterialExpressionNamedRerouteUsage_439}
    Rotation Per Step {from=MaterialExpressionNamedRerouteUsage_438}
    Falloff Power {from=MaterialExpressionNamedRerouteUsage_442}
    "Depth Steps (Total)" {from=MaterialExpressionNamedRerouteUsage_441}
    Depth Offset {from=MaterialExpressionNamedRerouteUsage_440}
    Depth Offset Dir {from=MaterialExpressionNamedRerouteUsage_437}
    Start Color {from=MaterialExpressionNamedRerouteUsage_445}
    End Color {from=MaterialExpressionNamedRerouteUsage_444}
    Start Scale {from=MaterialExpressionNamedRerouteUsage_550}
    End Scale {from=MaterialExpressionNamedRerouteUsage_549}
   ]
   properties {materialFunction=&MF_UI_CherrySmokeHeroRank_FullBadge_Streak}
  }
  MaterialExpressionMaterialFunctionCall_90 {
   class=MaterialExpressionMaterialFunctionCall
   inputs [
    SDF Texture {from=MaterialExpressionNamedRerouteUsage_452}
    Step Index {from=MaterialExpressionConstant_126}
    UVs {from=MaterialExpressionNamedRerouteUsage_448}
    Rotation Per Step {from=MaterialExpressionNamedRerouteUsage_447}
    Falloff Power {from=MaterialExpressionNamedRerouteUsage_451}
    "Depth Steps (Total)" {from=MaterialExpressionNamedRerouteUsage_450}
    Depth Offset {from=MaterialExpressionNamedRerouteUsage_449}
    Depth Offset Dir {from=MaterialExpressionNamedRerouteUsage_446}
    Start Color {from=MaterialExpressionNamedRerouteUsage_454}
    End Color {from=MaterialExpressionNamedRerouteUsage_453}
    Start Scale {from=MaterialExpressionNamedRerouteUsage_552}
    End Scale {from=MaterialExpressionNamedRerouteUsage_551}
   ]
   properties {materialFunction=&MF_UI_CherrySmokeHeroRank_FullBadge_Streak}
  }
  MaterialExpressionMaterialFunctionCall_91 {
   class=MaterialExpressionMaterialFunctionCall
   inputs [
    SDF Texture {from=MaterialExpressionNamedRerouteUsage_461}
    Step Index {from=MaterialExpressionConstant_127}
    UVs {from=MaterialExpressionNamedRerouteUsage_457}
    Rotation Per Step {from=MaterialExpressionNamedRerouteUsage_456}
    Falloff Power {from=MaterialExpressionNamedRerouteUsage_460}
    "Depth Steps (Total)" {from=MaterialExpressionNamedRerouteUsage_459}
    Depth Offset {from=MaterialExpressionNamedRerouteUsage_458}
    Depth Offset Dir {from=MaterialExpressionNamedRerouteUsage_455}
    Start Color {from=MaterialExpressionNamedRerouteUsage_463}
    End Color {from=MaterialExpressionNamedRerouteUsage_462}
    Start Scale {from=MaterialExpressionNamedRerouteUsage_554}
    End Scale {from=MaterialExpressionNamedRerouteUsage_553}
   ]
   properties {materialFunction=&MF_UI_CherrySmokeHeroRank_FullBadge_Streak}
  }
  MaterialExpressionMaterialFunctionCall_92 {
   class=MaterialExpressionMaterialFunctionCall
   inputs [
    SDF Texture {from=MaterialExpressionNamedRerouteUsage_470}
    Step Index {from=MaterialExpressionConstant_128}
    UVs {from=MaterialExpressionNamedRerouteUsage_466}
    Rotation Per Step {from=MaterialExpressionNamedRerouteUsage_465}
    Falloff Power {from=MaterialExpressionNamedRerouteUsage_469}
    "Depth Steps (Total)" {from=MaterialExpressionNamedRerouteUsage_468}
    Depth Offset {from=MaterialExpressionNamedRerouteUsage_467}
    Depth Offset Dir {from=MaterialExpressionNamedRerouteUsage_464}
    Start Color {from=MaterialExpressionNamedRerouteUsage_472}
    End Color {from=MaterialExpressionNamedRerouteUsage_471}
    Start Scale {from=MaterialExpressionNamedRerouteUsage_539}
    End Scale {from=MaterialExpressionNamedRerouteUsage_540}
   ]
   properties {materialFunction=&MF_UI_CherrySmokeHeroRank_FullBadge_Streak}
  }
  MaterialExpressionMaterialFunctionCall_93 {
   class=MaterialExpressionMaterialFunctionCall
   inputs [
    SDF Texture {from=MaterialExpressionNamedRerouteUsage_479}
    Step Index {from=MaterialExpressionConstant_129}
    UVs {from=MaterialExpressionNamedRerouteUsage_475}
    Rotation Per Step {from=MaterialExpressionNamedRerouteUsage_474}
    Falloff Power {from=MaterialExpressionNamedRerouteUsage_478}
    "Depth Steps (Total)" {from=MaterialExpressionNamedRerouteUsage_477}
    Depth Offset {from=MaterialExpressionNamedRerouteUsage_476}
    Depth Offset Dir {from=MaterialExpressionNamedRerouteUsage_473}
    Start Color {from=MaterialExpressionNamedRerouteUsage_481}
    End Color {from=MaterialExpressionNamedRerouteUsage_480}
    Start Scale {from=MaterialExpressionNamedRerouteUsage_542}
    End Scale {from=MaterialExpressionNamedRerouteUsage_541}
   ]
   properties {materialFunction=&MF_UI_CherrySmokeHeroRank_FullBadge_Streak}
  }
  MaterialExpressionMaterialFunctionCall_94 {
   class=MaterialExpressionMaterialFunctionCall
   inputs [
    SDF Texture {from=MaterialExpressionNamedRerouteUsage_488}
    Step Index {from=MaterialExpressionConstant_130}
    UVs {from=MaterialExpressionNamedRerouteUsage_484}
    Rotation Per Step {from=MaterialExpressionNamedRerouteUsage_483}
    Falloff Power {from=MaterialExpressionNamedRerouteUsage_487}
    "Depth Steps (Total)" {from=MaterialExpressionNamedRerouteUsage_486}
    Depth Offset {from=MaterialExpressionNamedRerouteUsage_485}
    Depth Offset Dir {from=MaterialExpressionNamedRerouteUsage_482}
    Start Color {from=MaterialExpressionNamedRerouteUsage_490}
    End Color {from=MaterialExpressionNamedRerouteUsage_489}
    Start Scale {from=MaterialExpressionNamedRerouteUsage_544}
    End Scale {from=MaterialExpressionNamedRerouteUsage_543}
   ]
   properties {materialFunction=&MF_UI_CherrySmokeHeroRank_FullBadge_Streak}
  }
  MaterialExpressionMaterialFunctionCall_95 {
   class=MaterialExpressionMaterialFunctionCall
   inputs [
    SDF Texture {from=MaterialExpressionNamedRerouteUsage_497}
    Step Index {from=MaterialExpressionConstant_131}
    UVs {from=MaterialExpressionNamedRerouteUsage_493}
    Rotation Per Step {from=MaterialExpressionNamedRerouteUsage_492}
    Falloff Power {from=MaterialExpressionNamedRerouteUsage_496}
    "Depth Steps (Total)" {from=MaterialExpressionNamedRerouteUsage_495}
    Depth Offset {from=MaterialExpressionNamedRerouteUsage_494}
    Depth Offset Dir {from=MaterialExpressionNamedRerouteUsage_491}
    Start Color {from=MaterialExpressionNamedRerouteUsage_499}
    End Color {from=MaterialExpressionNamedRerouteUsage_498}
    Start Scale {from=MaterialExpressionNamedRerouteUsage_546}
    End Scale {from=MaterialExpressionNamedRerouteUsage_545}
   ]
   properties {materialFunction=&MF_UI_CherrySmokeHeroRank_FullBadge_Streak}
  }
  MaterialExpressionMax_1 {
   class=MaterialExpressionMax
   inputs [
    A {from=MaterialExpressionNamedRerouteUsage_322}
    B {from=MaterialExpressionNamedRerouteUsage_323}
   ]
  }
  MaterialExpressionMax_10 {
   class=MaterialExpressionMax
   inputs [
    A {from=MaterialExpressionMax_9}
    B {from=MaterialExpressionNamedRerouteUsage_389}
   ]
  }
  MaterialExpressionMax_11 {
   class=MaterialExpressionMax
   inputs [
    A {from=MaterialExpressionMax_10}
    B {from=MaterialExpressionNamedRerouteUsage_390}
   ]
  }
  MaterialExpressionMax_12 {
   class=MaterialExpressionMax
   inputs [
    A {from=MaterialExpressionMax_11}
    B {from=MaterialExpressionNamedRerouteUsage_391}
   ]
  }
  MaterialExpressionMax_13 {
   class=MaterialExpressionMax
   inputs [
    A {from=MaterialExpressionMax_12}
    B {from=MaterialExpressionNamedRerouteUsage_392}
   ]
  }
  MaterialExpressionMax_14 {
   class=MaterialExpressionMax
   inputs [
    A {from=MaterialExpressionMax_13}
    B {from=MaterialExpressionNamedRerouteUsage_393}
   ]
  }
  MaterialExpressionMax_15 {
   class=MaterialExpressionMax
   inputs [
    A {from=MaterialExpressionMax_29}
    B {from=MaterialExpressionNamedRerouteUsage_502}
   ]
  }
  MaterialExpressionMax_16 {
   class=MaterialExpressionMax
   inputs [
    A {from=MaterialExpressionMax_15}
    B {from=MaterialExpressionNamedRerouteUsage_504}
   ]
  }
  MaterialExpressionMax_17 {
   class=MaterialExpressionMax
   inputs [
    A {from=MaterialExpressionMax_16}
    B {from=MaterialExpressionNamedRerouteUsage_503}
   ]
  }
  MaterialExpressionMax_18 {
   class=MaterialExpressionMax
   inputs [
    A {from=MaterialExpressionMax_17}
    B {from=MaterialExpressionNamedRerouteUsage_505}
   ]
  }
  MaterialExpressionMax_19 {
   class=MaterialExpressionMax
   inputs [
    A {from=MaterialExpressionMax_18}
    B {from=MaterialExpressionNamedRerouteUsage_506}
   ]
  }
  MaterialExpressionMax_2 {
   class=MaterialExpressionMax
   inputs [
    A {from=MaterialExpressionMax_1}
    B {from=MaterialExpressionNamedRerouteUsage_324}
   ]
  }
  MaterialExpressionMax_20 {
   class=MaterialExpressionMax
   inputs [
    A {from=MaterialExpressionMax_19}
    B {from=MaterialExpressionNamedRerouteUsage_507}
   ]
  }
  MaterialExpressionMax_21 {
   class=MaterialExpressionMax
   inputs [
    A {from=MaterialExpressionMax_20}
    B {from=MaterialExpressionNamedRerouteUsage_508}
   ]
  }
  MaterialExpressionMax_22 {
   class=MaterialExpressionMax
   inputs [
    A {from=MaterialExpressionMax_21}
    B {from=MaterialExpressionNamedRerouteUsage_509}
   ]
  }
  MaterialExpressionMax_23 {
   class=MaterialExpressionMax
   inputs [
    A {from=MaterialExpressionMax_22}
    B {from=MaterialExpressionNamedRerouteUsage_510}
   ]
  }
  MaterialExpressionMax_24 {
   class=MaterialExpressionMax
   inputs [
    A {from=MaterialExpressionMax_23}
    B {from=MaterialExpressionNamedRerouteUsage_511}
   ]
  }
  MaterialExpressionMax_25 {
   class=MaterialExpressionMax
   inputs [
    A {from=MaterialExpressionMax_24}
    B {from=MaterialExpressionNamedRerouteUsage_512}
   ]
  }
  MaterialExpressionMax_26 {
   class=MaterialExpressionMax
   inputs [
    A {from=MaterialExpressionMax_25}
    B {from=MaterialExpressionNamedRerouteUsage_513}
   ]
  }
  MaterialExpressionMax_27 {
   class=MaterialExpressionMax
   inputs [
    A {from=MaterialExpressionMax_26}
    B {from=MaterialExpressionNamedRerouteUsage_514}
   ]
  }
  MaterialExpressionMax_28 {
   class=MaterialExpressionMax
   inputs [
    A {from=MaterialExpressionMax_27}
    B {from=MaterialExpressionNamedRerouteUsage_515}
   ]
  }
  MaterialExpressionMax_29 {
   class=MaterialExpressionMax
   inputs [
    A {from=MaterialExpressionNamedRerouteUsage_500}
    B {from=MaterialExpressionNamedRerouteUsage_501}
   ]
  }
  MaterialExpressionMax_3 {
   class=MaterialExpressionMax
   inputs [
    A {from=MaterialExpressionMax_2}
    B {from=MaterialExpressionNamedRerouteUsage_325}
   ]
  }
  MaterialExpressionMax_30 {
   class=MaterialExpressionMax
   inputs [
    A {from=MaterialExpressionMax_14}
    B {from=MaterialExpressionNamedRerouteUsage_394}
   ]
  }
  MaterialExpressionMax_4 {
   class=MaterialExpressionMax
   inputs [
    A {from=MaterialExpressionMax_3}
    B {from=MaterialExpressionNamedRerouteUsage_326}
   ]
  }
  MaterialExpressionMax_5 {
   class=MaterialExpressionMax
   inputs [
    A {from=MaterialExpressionMax_4}
    B {from=MaterialExpressionNamedRerouteUsage_327}
   ]
  }
  MaterialExpressionMax_6 {
   class=MaterialExpressionMax
   inputs [
    A {from=MaterialExpressionMax_5}
    B {from=MaterialExpressionNamedRerouteUsage_328}
   ]
  }
  MaterialExpressionMax_7 {
   class=MaterialExpressionMax
   inputs [
    A {from=MaterialExpressionMax_6}
    B {from=MaterialExpressionNamedRerouteUsage_329}
   ]
  }
  MaterialExpressionMax_8 {
   class=MaterialExpressionMax
   inputs [
    A {from=MaterialExpressionMax_7}
    B {from=MaterialExpressionNamedRerouteUsage_387}
   ]
  }
  MaterialExpressionMax_9 {
   class=MaterialExpressionMax
   inputs [
    A {from=MaterialExpressionMax_8}
    B {from=MaterialExpressionNamedRerouteUsage_388}
   ]
  }
  MaterialExpressionMultiply_0 {
   class=MaterialExpressionMultiply
   inputs [
    A {from=MaterialExpressionMaterialFunctionCall_0, output=Smooth}
    B {from=""}
   ]
  }
  MaterialExpressionMultiply_1 {
   class=MaterialExpressionMultiply
   inputs [
    A {from=MaterialExpressionMaterialFunctionCall_4, output=Perspective UVs}
    B {from=MaterialExpressionConstant2Vector_0}
   ]
  }
  MaterialExpressionMultiply_10 {
   class=MaterialExpressionMultiply
   inputs [
    A {from=MaterialExpressionNamedRerouteDeclaration_34}
    B {from=MaterialExpressionNamedRerouteUsage_4}
   ]
   properties {constB=0}
  }
  MaterialExpressionMultiply_102 {
   class=MaterialExpressionMultiply
   inputs [
    A {from=MaterialExpressionNamedRerouteUsage_59}
    B {from=MaterialExpressionNamedRerouteUsage_63}
   ]
  }
  MaterialExpressionMultiply_11 {
   class=MaterialExpressionMultiply
   inputs [
    A {from=MaterialExpressionNamedRerouteDeclaration_36}
    B {from=MaterialExpressionNamedRerouteUsage_7}
   ]
  }
  MaterialExpressionMultiply_12 {
   class=MaterialExpressionMultiply
   inputs [
    A {from=MaterialExpressionMaterialFunctionCall_37, output=Rotated Values}
    B {from=MaterialExpressionNamedRerouteUsage_17}
   ]
  }
  MaterialExpressionMultiply_13 {
   class=MaterialExpressionMultiply
   inputs [
    A {from=MaterialExpressionMultiply_5}
    B {from=MaterialExpressionNamedRerouteUsage_69}
   ]
  }
  MaterialExpressionMultiply_15 {
   class=MaterialExpressionMultiply
   inputs [
    A {from=MaterialExpressionNamedRerouteUsage_72}
    B {from=MaterialExpressionNamedRerouteUsage_73}
   ]
   properties {constA=1}
  }
  MaterialExpressionMultiply_16 {
   class=MaterialExpressionMultiply
   inputs [
    A {from=MaterialExpressionMultiply_15}
    B {from=MaterialExpressionNamedRerouteUsage_117}
   ]
  }
  MaterialExpressionMultiply_19 {
   class=MaterialExpressionMultiply
   inputs [
    A {from=MaterialExpressionNamedRerouteUsage_55}
    B {from=MaterialExpressionNamedRerouteUsage_67}
   ]
  }
  MaterialExpressionMultiply_194 {
   class=MaterialExpressionMultiply
   inputs [
    A {from=MaterialExpressionNamedRerouteUsage_330}
    B {from=MaterialExpressionNamedRerouteUsage_68}
   ]
  }
  MaterialExpressionMultiply_2 {
   class=MaterialExpressionMultiply
   inputs [
    A {from=MaterialExpressionMaterialFunctionCall_2, output=Result}
    B {from=MaterialExpressionNamedRerouteUsage_16}
   ]
   properties {desc=PolarOffset}
  }
  MaterialExpressionMultiply_20 {
   class=MaterialExpressionMultiply
   inputs [
    A {from=MaterialExpressionNamedRerouteUsage_52}
    B {from=MaterialExpressionNamedRerouteUsage_65}
   ]
  }
  MaterialExpressionMultiply_21 {
   class=MaterialExpressionMultiply
   inputs [
    A {from=MaterialExpressionNamedRerouteUsage_53}
    B {from=MaterialExpressionNamedRerouteUsage_66}
   ]
  }
  MaterialExpressionMultiply_22 {
   class=MaterialExpressionMultiply
   inputs [
    A {from=MaterialExpressionNamedRerouteUsage_54}
    B {from=MaterialExpressionNamedRerouteUsage_64}
   ]
  }
  MaterialExpressionMultiply_23 {
   class=MaterialExpressionMultiply
   inputs [
    A {from=MaterialExpressionNamedRerouteUsage_70}
    B {from=MaterialExpressionNamedRerouteUsage_71}
   ]
  }
  MaterialExpressionMultiply_24 {
   class=MaterialExpressionMultiply
   inputs [
    A {from=MaterialExpressionNamedRerouteUsage_76}
    B {from=MaterialExpressionNamedRerouteUsage_77}
   ]
  }
  MaterialExpressionMultiply_25 {
   class=MaterialExpressionMultiply
   inputs [
    A {from=MaterialExpressionNamedRerouteUsage_110}
    B {from=MaterialExpressionNamedRerouteUsage_77}
   ]
  }
  MaterialExpressionMultiply_26 {
   class=MaterialExpressionMultiply
   inputs [
    A {from=MaterialExpressionNamedRerouteUsage_81}
    B {from=MaterialExpressionNamedRerouteUsage_82}
   ]
  }
  MaterialExpressionMultiply_27 {
   class=MaterialExpressionMultiply
   inputs [
    A {from=MaterialExpressionNamedRerouteUsage_83}
    B {from=MaterialExpressionNamedRerouteUsage_82}
   ]
  }
  MaterialExpressionMultiply_28 {
   class=MaterialExpressionMultiply
   inputs [
    A {from=MaterialExpressionNamedRerouteUsage_90}
    B {from=MaterialExpressionNamedRerouteUsage_86}
   ]
  }
  MaterialExpressionMultiply_29 {
   class=MaterialExpressionMultiply
   inputs [
    A {from=MaterialExpressionNamedRerouteUsage_116}
    B {from=MaterialExpressionNamedRerouteUsage_86}
   ]
  }
  MaterialExpressionMultiply_3 {
   class=MaterialExpressionMultiply
   inputs [
    A {from=MaterialExpressionNamedRerouteDeclaration_119}
    B {from=MaterialExpressionNamedRerouteUsage_30}
   ]
  }
  MaterialExpressionMultiply_30 {
   class=MaterialExpressionMultiply
   inputs [
    A {from=MaterialExpressionNamedRerouteUsage_90}
    B {from=MaterialExpressionNamedRerouteUsage_88}
   ]
  }
  MaterialExpressionMultiply_302 {
   class=MaterialExpressionMultiply
   inputs [
    A {from=MaterialExpressionNamedRerouteUsage_519}
    B {from=MaterialExpressionNamedRerouteUsage_520}
   ]
   properties {constA=1}
  }
  MaterialExpressionMultiply_303 {
   class=MaterialExpressionMultiply
   inputs [
    A {from=MaterialExpressionNamedRerouteUsage_517}
    B {from=MaterialExpressionNamedRerouteUsage_516}
   ]
  }
  MaterialExpressionMultiply_304 {
   class=MaterialExpressionMultiply
   inputs [
    A {from=MaterialExpressionMultiply_303}
    B {from=MaterialExpressionNamedRerouteUsage_522}
   ]
  }
  MaterialExpressionMultiply_31 {
   class=MaterialExpressionMultiply
   inputs [
    A {from=MaterialExpressionNamedRerouteUsage_89}
    B {from=MaterialExpressionNamedRerouteUsage_88}
   ]
  }
  MaterialExpressionMultiply_32 {
   class=MaterialExpressionMultiply
   inputs [
    A {from=MaterialExpressionNamedRerouteUsage_98}
    B {from=MaterialExpressionNamedRerouteUsage_94}
   ]
  }
  MaterialExpressionMultiply_33 {
   class=MaterialExpressionMultiply
   inputs [
    A {from=MaterialExpressionNamedRerouteUsage_115}
    B {from=MaterialExpressionNamedRerouteUsage_94}
   ]
  }
  MaterialExpressionMultiply_34 {
   class=MaterialExpressionMultiply
   inputs [
    A {from=MaterialExpressionNamedRerouteUsage_95}
    B {from=MaterialExpressionNamedRerouteUsage_96}
   ]
  }
  MaterialExpressionMultiply_35 {
   class=MaterialExpressionMultiply
   inputs [
    A {from=MaterialExpressionNamedRerouteUsage_97}
    B {from=MaterialExpressionNamedRerouteUsage_96}
   ]
  }
  MaterialExpressionMultiply_36 {
   class=MaterialExpressionMultiply
   inputs [
    A {from=MaterialExpressionNamedRerouteUsage_107}
    B {from=MaterialExpressionNamedRerouteUsage_101}
   ]
  }
  MaterialExpressionMultiply_37 {
   class=MaterialExpressionMultiply
   inputs [
    A {from=MaterialExpressionNamedRerouteUsage_112}
    B {from=MaterialExpressionNamedRerouteUsage_101}
   ]
  }
  MaterialExpressionMultiply_38 {
   class=MaterialExpressionMultiply
   inputs [
    A {from=MaterialExpressionNamedRerouteUsage_105}
    B {from=MaterialExpressionNamedRerouteUsage_103}
   ]
  }
  MaterialExpressionMultiply_39 {
   class=MaterialExpressionMultiply
   inputs [
    A {from=MaterialExpressionNamedRerouteUsage_104}
    B {from=MaterialExpressionNamedRerouteUsage_103}
   ]
  }
  MaterialExpressionMultiply_4 {
   class=MaterialExpressionMultiply
   inputs [
    A {from=MaterialExpressionNamedRerouteUsage_31}
    B {from=MaterialExpressionNamedRerouteUsage_41}
   ]
  }
  MaterialExpressionMultiply_40 {
   class=MaterialExpressionMultiply
   inputs [
    A {from=MaterialExpressionTime_0}
    B {from=MaterialExpressionLinearInterpolate_7}
   ]
  }
  MaterialExpressionMultiply_41 {
   class=MaterialExpressionMultiply
   inputs [
    A {from=MaterialExpressionNamedRerouteUsage_28}
    B {from=MaterialExpressionNamedRerouteUsage_121}
   ]
  }
  MaterialExpressionMultiply_42 {
   class=MaterialExpressionMultiply
   inputs [
    A {from=MaterialExpressionNamedRerouteUsage_27}
    B {from=MaterialExpressionNamedRerouteUsage_121}
   ]
  }
  MaterialExpressionMultiply_43 {
   class=MaterialExpressionMultiply
   inputs [
    A {from=MaterialExpressionNamedRerouteUsage_123}
    B {from=MaterialExpressionNamedRerouteUsage_122}
   ]
  }
  MaterialExpressionMultiply_44 {
   class=MaterialExpressionMultiply
   inputs [
    A {from=MaterialExpressionNamedRerouteUsage_124}
    B {from=MaterialExpressionNamedRerouteUsage_122}
   ]
  }
  MaterialExpressionMultiply_45 {
   class=MaterialExpressionMultiply
   inputs [
    A {from=MaterialExpressionNamedRerouteUsage_37}
    B {from=MaterialExpressionNamedRerouteUsage_125}
   ]
  }
  MaterialExpressionMultiply_46 {
   class=MaterialExpressionMultiply
   inputs [
    A {from=MaterialExpressionNamedRerouteUsage_38}
    B {from=MaterialExpressionNamedRerouteUsage_125}
   ]
  }
  MaterialExpressionMultiply_47 {
   class=MaterialExpressionMultiply
   inputs [
    A {from=MaterialExpressionSubtract_1}
    B {from=MaterialExpressionMultiply_48}
   ]
  }
  MaterialExpressionMultiply_48 {
   class=MaterialExpressionMultiply
   inputs [
    A {from=MaterialExpressionScalarParameter_27}
    B {from=""}
   ]
   properties {constB=0.1}
  }
  MaterialExpressionMultiply_5 {
   class=MaterialExpressionMultiply
   inputs [
    A {from=MaterialExpressionNamedRerouteUsage_62}
    B {from=MaterialExpressionNamedRerouteUsage_41}
   ]
  }
  MaterialExpressionMultiply_7 {
   class=MaterialExpressionMultiply
   inputs [
    A {from=MaterialExpressionNamedRerouteUsage_142}
    B {from=MaterialExpressionNamedRerouteUsage_63}
   ]
  }
  MaterialExpressionMultiply_9 {
   class=MaterialExpressionMultiply
   inputs [
    A {from=MaterialExpressionNamedRerouteDeclaration_32}
    B {from=MaterialExpressionNamedRerouteUsage_3}
   ]
  }
  MaterialExpressionNamedRerouteDeclaration_0 {
   class=MaterialExpressionNamedRerouteDeclaration
   inputs [None {from=MaterialExpressionStaticSwitch_16}]
   properties {name=Final_Color, nodeColor=<rgba 0 1 0 1>}
  }
  MaterialExpressionNamedRerouteDeclaration_1 {
   class=MaterialExpressionNamedRerouteDeclaration
   inputs [None {from=MaterialExpressionTextureSample_1, output=RGB}]
   properties {name=P_Rank_Glyph, nodeColor=<rgba 0 0 0 1>}
  }
  MaterialExpressionNamedRerouteDeclaration_10 {
   class=MaterialExpressionNamedRerouteDeclaration
   inputs [None {from=MaterialExpressionScalarParameter_4}]
   properties {name=P_ColorGradient_BG_Rotation, nodeColor=<rgba 0 0 0 1>}
  }
  MaterialExpressionNamedRerouteDeclaration_100 {
   class=MaterialExpressionNamedRerouteDeclaration
   inputs [None {from=MaterialExpressionLinearInterpolate_19}]
   properties {name=I_Color_Solid_Both, nodeColor=<rgba 0 0 0 1>}
  }
  MaterialExpressionNamedRerouteDeclaration_101 {
   class=MaterialExpressionNamedRerouteDeclaration
   inputs [None {from=MaterialExpressionReroute_98}]
   properties {name=I_Color_i0_Both, nodeColor=<rgba 0 0 0 1>}
  }
  MaterialExpressionNamedRerouteDeclaration_102 {
   class=MaterialExpressionNamedRerouteDeclaration
   inputs [None {from=MaterialExpressionLinearInterpolate_20}]
   properties {name=I_Color_i0_Gradient_Outer, nodeColor=<rgba 0 0 0 1>}
  }
  MaterialExpressionNamedRerouteDeclaration_103 {
   class=MaterialExpressionNamedRerouteDeclaration
   inputs [None {from=MaterialExpressionLinearInterpolate_21}]
   properties {name=I_Color_i0_Gradient_Inner, nodeColor=<rgba 0 0 0 1>}
  }
  MaterialExpressionNamedRerouteDeclaration_104 {
   class=MaterialExpressionNamedRerouteDeclaration
   inputs [None {from=MaterialExpressionLinearInterpolate_23}]
   properties {name=I_Color_i1_Gradient_Outer, nodeColor=<rgba 0 0 0 1>}
  }
  MaterialExpressionNamedRerouteDeclaration_105 {
   class=MaterialExpressionNamedRerouteDeclaration
   inputs [None {from=MaterialExpressionLinearInterpolate_24}]
   properties {name=I_Color_i1_Solid_Inner, nodeColor=<rgba 0 0 0 1>}
  }
  MaterialExpressionNamedRerouteDeclaration_106 {
   class=MaterialExpressionNamedRerouteDeclaration
   inputs [None {from=MaterialExpressionLinearInterpolate_25}]
   properties {name=I_Color_i1_Both, nodeColor=<rgba 0 0 0 1>}
  }
  MaterialExpressionNamedRerouteDeclaration_108 {
   class=MaterialExpressionNamedRerouteDeclaration
   inputs [None {from=MaterialExpressionLinearInterpolate_26}]
   properties {name=I_Color_i2_Solid_Outer, nodeColor=<rgba 0 0 0 1>}
  }
  MaterialExpressionNamedRerouteDeclaration_109 {
   class=MaterialExpressionNamedRerouteDeclaration
   inputs [None {from=MaterialExpressionLinearInterpolate_27}]
   properties {name=I_Color_i2_Gradient_Inner, nodeColor=<rgba 0 0 0 1>}
  }
  MaterialExpressionNamedRerouteDeclaration_11 {
   class=MaterialExpressionNamedRerouteDeclaration
   inputs [None {from=MaterialExpressionScalarParameter_5}]
   properties {name=P_ColorGradient_BG_Scale, nodeColor=<rgba 0 0 0 1>}
  }
  MaterialExpressionNamedRerouteDeclaration_110 {
   class=MaterialExpressionNamedRerouteDeclaration
   inputs [None {from=MaterialExpressionLinearInterpolate_28}]
   properties {name=I_Color_i2_Both, nodeColor=<rgba 0 0 0 1>}
  }
  MaterialExpressionNamedRerouteDeclaration_111 {
   class=MaterialExpressionNamedRerouteDeclaration
   inputs [None {from=MaterialExpressionStaticBoolParameter_11}]
   properties {name=P_Rank_Radial_Glow_Position_Polar, nodeColor=<rgba 0 0 0 1>}
  }
  MaterialExpressionNamedRerouteDeclaration_112 {
   class=MaterialExpressionNamedRerouteDeclaration
   inputs [None {from=MaterialExpressionMaterialFunctionCall_5, output=Result}]
   properties {name=P_Rank_Radial_Glow_Position, nodeColor=<rgba 0 0 0 1>}
  }
  MaterialExpressionNamedRerouteDeclaration_114 {
   class=MaterialExpressionNamedRerouteDeclaration
   inputs [None {from=MaterialExpressionScalarParameter_21}]
   properties {name=P_Rank_Radial_Glow_Position_Distance, nodeColor=<rgba 0 0 0 1>}
  }
  MaterialExpressionNamedRerouteDeclaration_115 {
   class=MaterialExpressionNamedRerouteDeclaration
   inputs [None {from=MaterialExpressionScalarParameter_22}]
   properties {name=P_Rank_Radial_Glow_Position_Angle, nodeColor=<rgba 0 0 0 1>}
  }
  MaterialExpressionNamedRerouteDeclaration_116 {
   class=MaterialExpressionNamedRerouteDeclaration
   inputs [None {from=MaterialExpressionMaterialFunctionCall_3, output=Result}]
   properties {name=P_Rank_Radial_Glow_Position_Center, nodeColor=<rgba 0 0 0 1>}
  }
  MaterialExpressionNamedRerouteDeclaration_118 {
   class=MaterialExpressionNamedRerouteDeclaration
   inputs [None {from=MaterialExpressionStaticBoolParameter_12}]
   properties {name=P_Rank_Radial_Glow_Clamp, nodeColor=<rgba 0 0 0 1>}
  }
  MaterialExpressionNamedRerouteDeclaration_119 {
   class=MaterialExpressionNamedRerouteDeclaration
   inputs [None {from=MaterialExpressionMultiply_0}]
   properties {name=SDF_AA_Shape_Glow, nodeColor=<rgba 0 0 0 1>}
  }
  MaterialExpressionNamedRerouteDeclaration_12 {
   class=MaterialExpressionNamedRerouteDeclaration
   inputs [None {from=MaterialExpressionSaturate_4}]
   properties {name=P_Opacity_Outline_Thick_Outer, nodeColor=<rgba 0 0 0 1>}
  }
  MaterialExpressionNamedRerouteDeclaration_120 {
   class=MaterialExpressionNamedRerouteDeclaration
   inputs [None {from=MaterialExpressionVectorParameter_6, output=RGB}]
   properties {name=P_Rank_Radial_Glow_Color, nodeColor=<rgba 0 0 0 1>}
  }
  MaterialExpressionNamedRerouteDeclaration_121 {
   class=MaterialExpressionNamedRerouteDeclaration
   inputs [None {from=MaterialExpressionMultiply_3}]
   properties {name=SDF_Color_Shape_Glow, nodeColor=<rgba 0 0 0 1>}
  }
  MaterialExpressionNamedRerouteDeclaration_122 {
   class=MaterialExpressionNamedRerouteDeclaration
   inputs [None {from=MaterialExpressionStaticSwitch_1}]
   properties {name=I_Final_Opacity_Pre_Glow, nodeColor=<rgba 0 0 0 1>}
  }
  MaterialExpressionNamedRerouteDeclaration_124 {
   class=MaterialExpressionNamedRerouteDeclaration
   inputs [None {from=MaterialExpressionAdd_31}]
   properties {name=I_Glitch_UVs, nodeColor=<rgba 0 0 0 1>}
  }
  MaterialExpressionNamedRerouteDeclaration_125 {
   class=MaterialExpressionNamedRerouteDeclaration
   inputs [None {from=MaterialExpressionAdd_31}]
   properties {name=I_Glitch_UVs_Rank, nodeColor=<rgba 0 0 0 1>}
  }
  MaterialExpressionNamedRerouteDeclaration_126 {
   class=MaterialExpressionNamedRerouteDeclaration
   inputs [None {from=MaterialExpressionTextureCoordinate_1}]
   properties {name=I_Raw_UVs, nodeColor=<rgba 0 0 0 1>}
  }
  MaterialExpressionNamedRerouteDeclaration_13 {
   class=MaterialExpressionNamedRerouteDeclaration
   inputs [None {from=MaterialExpressionSaturate_6}]
   properties {name=P_Opacity_Outline_Thick_Inner, nodeColor=<rgba 0 0 0 1>}
  }
  MaterialExpressionNamedRerouteDeclaration_14 {
   class=MaterialExpressionNamedRerouteDeclaration
   inputs [None {from=MaterialExpressionScalarParameter_8}]
   properties {name=P_Rank_Radial_Glow_Start, nodeColor=<rgba 0 0 0 1>}
  }
  MaterialExpressionNamedRerouteDeclaration_15 {
   class=MaterialExpressionNamedRerouteDeclaration
   inputs [None {from=MaterialExpressionMaterialFunctionCall_23, output=SDF}]
   properties {name=SDF_Raw_Shape_Inner, nodeColor=<rgba 0 0 0 1>}
  }
  MaterialExpressionNamedRerouteDeclaration_16 {
   class=MaterialExpressionNamedRerouteDeclaration
   inputs [None {from=MaterialExpressionScalarParameter_16}]
   properties {name=P_Rank_Radial_Glow_End, nodeColor=<rgba 0 0 0 1>}
  }
  MaterialExpressionNamedRerouteDeclaration_17 {
   class=MaterialExpressionNamedRerouteDeclaration
   inputs [None {from=MaterialExpressionMaterialFunctionCall_1, output=Inner}]
   properties {name=SDF_AA_Shape_Outer, nodeColor=<rgba 0 0 0 1>}
  }
  MaterialExpressionNamedRerouteDeclaration_18 {
   class=MaterialExpressionNamedRerouteDeclaration
   inputs [None {from=MaterialExpressionMaterialFunctionCall_24, output=Inner}]
   properties {name=SDF_AA_Shape_Inner, nodeColor=<rgba 0 0 0 1>}
  }
  MaterialExpressionNamedRerouteDeclaration_19 {
   class=MaterialExpressionNamedRerouteDeclaration
   inputs [None {from=MaterialExpressionSaturate_5}]
   properties {name=P_Opacity_Outline_Thin_Outer, nodeColor=<rgba 0 0 0 1>}
  }
  MaterialExpressionNamedRerouteDeclaration_2 {
   class=MaterialExpressionNamedRerouteDeclaration
   inputs [None {from=MaterialExpressionVectorParameter_0, output=RGB}]
   properties {name=P_Color_BG_Gradient_Start, nodeColor=<rgba 0 0 0 1>}
  }
  MaterialExpressionNamedRerouteDeclaration_20 {
   class=MaterialExpressionNamedRerouteDeclaration
   inputs [None {from=MaterialExpressionSaturate_1}]
   properties {name=P_Opacity_BG_Outer, nodeColor=<rgba 0 0 0 1>}
  }
  MaterialExpressionNamedRerouteDeclaration_21 {
   class=MaterialExpressionNamedRerouteDeclaration
   inputs [None {from=MaterialExpressionSaturate_0}]
   properties {name=P_Opacity_BG_Inner, nodeColor=<rgba 0 0 0 1>}
  }
  MaterialExpressionNamedRerouteDeclaration_22 {
   class=MaterialExpressionNamedRerouteDeclaration
   inputs [None {from=MaterialExpressionMaterialFunctionCall_143, output=Result}]
   properties {name=P_Rank_Offset, nodeColor=<rgba 0 0 0 1>}
  }
  MaterialExpressionNamedRerouteDeclaration_23 {
   class=MaterialExpressionNamedRerouteDeclaration
   inputs [None {from=MaterialExpressionNamedRerouteUsage_145}]
   properties {name=I_Standard_UVs, nodeColor=<rgba 0 0 0 1>}
  }
  MaterialExpressionNamedRerouteDeclaration_24 {
   class=MaterialExpressionNamedRerouteDeclaration
   inputs [None {from=MaterialExpressionVectorParameter_7, output=RGB}]
   properties {name=P_Color_Rank, nodeColor=<rgba 0 0 0 1>}
  }
  MaterialExpressionNamedRerouteDeclaration_25 {
   class=MaterialExpressionNamedRerouteDeclaration
   inputs [None {from=MaterialExpressionVectorParameter_8, output=RGB}]
   properties {name=P_Color_BG_Inner, nodeColor=<rgba 0 0 0 1>}
  }
  MaterialExpressionNamedRerouteDeclaration_26 {
   class=MaterialExpressionNamedRerouteDeclaration
   inputs [None {from=MaterialExpressionStaticBoolParameter_0}]
   properties {name=P_Rank_Streaks_Enabled, nodeColor=<rgba 0 0 0 1>}
  }
  MaterialExpressionNamedRerouteDeclaration_27 {
   class=MaterialExpressionNamedRerouteDeclaration
   inputs [None {from=MaterialExpressionStaticSwitch_28}]
   properties {name=SDF_Color_Shape_NoGradient, nodeColor=<rgba 0 0 0 1>}
  }
  MaterialExpressionNamedRerouteDeclaration_28 {
   class=MaterialExpressionNamedRerouteDeclaration
   inputs [None {from=MaterialExpressionReroute_100}]
   properties {name=SDF_Color_Shape_Switch_0_Whole, nodeColor=<rgba 0 0 0 1>}
  }
  MaterialExpressionNamedRerouteDeclaration_29 {
   class=MaterialExpressionNamedRerouteDeclaration
   inputs [None {from=MaterialExpressionReroute_106}]
   properties {name=SDF_Color_Shape_Switch_1_Outer, nodeColor=<rgba 0 0 0 1>}
  }
  MaterialExpressionNamedRerouteDeclaration_3 {
   class=MaterialExpressionNamedRerouteDeclaration
   inputs [None {from=MaterialExpressionStaticBoolParameter_9}]
   properties {name=P_Rank_Radial_Glow_Enabled, nodeColor=<rgba 0 0 0 1>}
  }
  MaterialExpressionNamedRerouteDeclaration_30 {
   class=MaterialExpressionNamedRerouteDeclaration
   inputs [None {from=MaterialExpressionMultiply_9}]
   properties {name=SDF_Color_Outline_Thick_Inner, nodeColor=<rgba 0 0 0 1>}
  }
  MaterialExpressionNamedRerouteDeclaration_31 {
   class=MaterialExpressionNamedRerouteDeclaration
   inputs [None {from=MaterialExpressionReroute_109}]
   properties {name=SDF_Color_Shape_Switch_2_Inner, nodeColor=<rgba 0 0 0 1>}
  }
  MaterialExpressionNamedRerouteDeclaration_32 {
   class=MaterialExpressionNamedRerouteDeclaration
   inputs [None {from=MaterialExpressionMaterialFunctionCall_27, output=Inner}]
   properties {name=SDF_AA_Outline_Thick_Inner, nodeColor=<rgba 0 0 0 1>}
  }
  MaterialExpressionNamedRerouteDeclaration_33 {
   class=MaterialExpressionNamedRerouteDeclaration
   inputs [None {from=MaterialExpressionMultiply_10}]
   properties {name=SDF_Color_Outline_Thick_Outer, nodeColor=<rgba 0 0 0 1>}
  }
  MaterialExpressionNamedRerouteDeclaration_34 {
   class=MaterialExpressionNamedRerouteDeclaration
   inputs [None {from=MaterialExpressionMaterialFunctionCall_31, output=Inner}]
   properties {name=SDF_AA_Outline_Thick_Outer, nodeColor=<rgba 0 0 0 1>}
  }
  MaterialExpressionNamedRerouteDeclaration_35 {
   class=MaterialExpressionNamedRerouteDeclaration
   inputs [None {from=MaterialExpressionMultiply_11}]
   properties {name=SDF_Color_Outline_Thin, nodeColor=<rgba 0 0 0 1>}
  }
  MaterialExpressionNamedRerouteDeclaration_36 {
   class=MaterialExpressionNamedRerouteDeclaration
   inputs [None {from=MaterialExpressionMaterialFunctionCall_33, output=Inner}]
   properties {name=SDF_AA_Outline_Thin, nodeColor=<rgba 0 0 0 1>}
  }
  MaterialExpressionNamedRerouteDeclaration_37 {
   class=MaterialExpressionNamedRerouteDeclaration
   inputs [None {from=MaterialExpressionStaticBoolParameter_8}]
   properties {name=P_Shape_BG_Overlap, nodeColor=<rgba 0 0 0 1>}
  }
  MaterialExpressionNamedRerouteDeclaration_38 {
   class=MaterialExpressionNamedRerouteDeclaration
   inputs [None {from=MaterialExpressionStaticSwitch_26}]
   properties {name=SDF_AA_Shape_Outer_Overlap_Affected, nodeColor=<rgba 0 0 0 1>}
  }
  MaterialExpressionNamedRerouteDeclaration_39 {
   class=MaterialExpressionNamedRerouteDeclaration
   inputs [None {from=MaterialExpressionConstant_12}]
   properties {name=P_Rank_Streaks_Steps, nodeColor=<rgba 0 0 0 1>}
  }
  MaterialExpressionNamedRerouteDeclaration_4 {
   class=MaterialExpressionNamedRerouteDeclaration
   inputs [None {from=MaterialExpressionVectorParameter_1, output=RGB}]
   properties {name=P_Color_BG_Gradient_End, nodeColor=<rgba 0 0 0 1>}
  }
  MaterialExpressionNamedRerouteDeclaration_40 {
   class=MaterialExpressionNamedRerouteDeclaration
   inputs [None {from=MaterialExpressionScalarParameter_10}]
   properties {name=P_Rank_Streaks_Offset, nodeColor=<rgba 0 0 0 1>}
  }
  MaterialExpressionNamedRerouteDeclaration_41 {
   class=MaterialExpressionNamedRerouteDeclaration
   inputs [None {from=MaterialExpressionScalarParameter_11}]
   properties {name=P_Rank_Streaks_Angle, nodeColor=<rgba 0 0 0 1>}
  }
  MaterialExpressionNamedRerouteDeclaration_42 {
   class=MaterialExpressionNamedRerouteDeclaration
   inputs [None {from=MaterialExpressionRound_2}]
   properties {name=P_Color_Gradient_Shape, nodeColor=<rgba 0 0 0 1>}
  }
  MaterialExpressionNamedRerouteDeclaration_43 {
   class=MaterialExpressionNamedRerouteDeclaration
   inputs [None {from=MaterialExpressionScalarParameter_12}]
   properties {name=P_Rank_Streaks_Falloff_Power, nodeColor=<rgba 0 0 0 1>}
  }
  MaterialExpressionNamedRerouteDeclaration_44 {
   class=MaterialExpressionNamedRerouteDeclaration
   inputs [None {from=MaterialExpressionScalarParameter_13}]
   properties {name=P_Rank_Streaks_Rotation_Per_Step, nodeColor=<rgba 0 0 0 1>}
  }
  MaterialExpressionNamedRerouteDeclaration_45 {
   class=MaterialExpressionNamedRerouteDeclaration
   inputs [None {from=MaterialExpressionLinearInterpolate_6}]
   properties {name=I_Color_Fill_BG_Gradient, nodeColor=<rgba 0 0 0 1>}
  }
  MaterialExpressionNamedRerouteDeclaration_46 {
   class=MaterialExpressionNamedRerouteDeclaration
   inputs [None {from=MaterialExpressionStaticBoolParameter_1}]
   properties {name=P_BG_Inner_Enabled, nodeColor=<rgba 0 0 0 1>}
  }
  MaterialExpressionNamedRerouteDeclaration_47 {
   class=MaterialExpressionNamedRerouteDeclaration
   inputs [None {from=MaterialExpressionStaticBoolParameter_2}]
   properties {name=P_BG_Outer_Enabled, nodeColor=<rgba 0 0 0 1>}
  }
  MaterialExpressionNamedRerouteDeclaration_48 {
   class=MaterialExpressionNamedRerouteDeclaration
   inputs [None {from=MaterialExpressionStaticBoolParameter_3}]
   properties {name=P_BG_Gradient_Enabled, nodeColor=<rgba 0 0 0 1>}
  }
  MaterialExpressionNamedRerouteDeclaration_49 {
   class=MaterialExpressionNamedRerouteDeclaration
   inputs [None {from=MaterialExpressionStaticBoolParameter_4}]
   properties {name=P_Outline_Thick_Inner_Enabled, nodeColor=<rgba 0 0 0 1>}
  }
  MaterialExpressionNamedRerouteDeclaration_5 {
   class=MaterialExpressionNamedRerouteDeclaration
   inputs [None {from=MaterialExpressionVectorParameter_3, output=RGB}]
   properties {name=P_Color_Outline_Thick_Inner, nodeColor=<rgba 0 0 0 1>}
  }
  MaterialExpressionNamedRerouteDeclaration_50 {
   class=MaterialExpressionNamedRerouteDeclaration
   inputs [None {from=MaterialExpressionMaterialFunctionCall_80, output=Shape}]
   properties {name=I_SDF_AA_Glyph_Streak_Step_Mask_1, nodeColor=<rgba 0 0 0 1>}
  }
  MaterialExpressionNamedRerouteDeclaration_51 {
   class=MaterialExpressionNamedRerouteDeclaration
   inputs [None {from=MaterialExpressionTextureObjectParameter_0}]
   properties {name=P_Rank_Glyph_Pre, nodeColor=<rgba 0 0 0 1>}
  }
  MaterialExpressionNamedRerouteDeclaration_52 {
   class=MaterialExpressionNamedRerouteDeclaration
   inputs [None {from=MaterialExpressionConvert_1}]
   properties {name=I_Rank_Depth_OffsetDir, nodeColor=<rgba 0 0 0 1>}
  }
  MaterialExpressionNamedRerouteDeclaration_53 {
   class=MaterialExpressionNamedRerouteDeclaration
   inputs [None {from=MaterialExpressionMaterialFunctionCall_80, output="Shape (Color)"}]
   properties {name=I_SDF_AA_Glyph_Streak_Step1, nodeColor=<rgba 0 0 0 1>}
  }
  MaterialExpressionNamedRerouteDeclaration_54 {
   class=MaterialExpressionNamedRerouteDeclaration
   inputs [None {from=MaterialExpressionMaterialFunctionCall_81, output="Shape (Color)"}]
   properties {name=I_SDF_AA_Glyph_Streak_Step2, nodeColor=<rgba 0 0 0 1>}
  }
  MaterialExpressionNamedRerouteDeclaration_55 {
   class=MaterialExpressionNamedRerouteDeclaration
   inputs [None {from=MaterialExpressionMaterialFunctionCall_82, output="Shape (Color)"}]
   properties {name=I_SDF_AA_Glyph_Streak_Step3, nodeColor=<rgba 0 0 0 1>}
  }
  MaterialExpressionNamedRerouteDeclaration_56 {
   class=MaterialExpressionNamedRerouteDeclaration
   inputs [None {from=MaterialExpressionMaterialFunctionCall_83, output="Shape (Color)"}]
   properties {name=I_SDF_AA_Glyph_Streak_Step4, nodeColor=<rgba 0 0 0 1>}
  }
  MaterialExpressionNamedRerouteDeclaration_57 {
   class=MaterialExpressionNamedRerouteDeclaration
   inputs [None {from=MaterialExpressionMaterialFunctionCall_81, output=Shape}]
   properties {name=I_SDF_AA_Glyph_Streak_Step_Mask_2, nodeColor=<rgba 0 0 0 1>}
  }
  MaterialExpressionNamedRerouteDeclaration_58 {
   class=MaterialExpressionNamedRerouteDeclaration
   inputs [None {from=MaterialExpressionMaterialFunctionCall_82, output=Shape}]
   properties {name=I_SDF_AA_Glyph_Streak_Step_Mask_3, nodeColor=<rgba 0 0 0 1>}
  }
  MaterialExpressionNamedRerouteDeclaration_59 {
   class=MaterialExpressionNamedRerouteDeclaration
   inputs [None {from=MaterialExpressionMaterialFunctionCall_83, output=Shape}]
   properties {name=I_SDF_AA_Glyph_Streak_Step_Mask_4, nodeColor=<rgba 0 0 0 1>}
  }
  MaterialExpressionNamedRerouteDeclaration_6 {
   class=MaterialExpressionNamedRerouteDeclaration
   inputs [None {from=MaterialExpressionVectorParameter_4, output=RGB}]
   properties {name=P_Color_Outline_Thick_Outer, nodeColor=<rgba 0 0 0 1>}
  }
  MaterialExpressionNamedRerouteDeclaration_60 {
   class=MaterialExpressionNamedRerouteDeclaration
   inputs [None {from=MaterialExpressionMaterialFunctionCall_84, output=Shape}]
   properties {name=I_SDF_AA_Glyph_Streak_Step_Mask_5, nodeColor=<rgba 0 0 0 1>}
  }
  MaterialExpressionNamedRerouteDeclaration_61 {
   class=MaterialExpressionNamedRerouteDeclaration
   inputs [None {from=MaterialExpressionMaterialFunctionCall_84, output="Shape (Color)"}]
   properties {name=I_SDF_AA_Glyph_Streak_Step5, nodeColor=<rgba 0 0 0 1>}
  }
  MaterialExpressionNamedRerouteDeclaration_62 {
   class=MaterialExpressionNamedRerouteDeclaration
   inputs [None {from=MaterialExpressionMaterialFunctionCall_85, output="Shape (Color)"}]
   properties {name=I_SDF_AA_Glyph_Streak_Step6, nodeColor=<rgba 0 0 0 1>}
  }
  MaterialExpressionNamedRerouteDeclaration_63 {
   class=MaterialExpressionNamedRerouteDeclaration
   inputs [None {from=MaterialExpressionMaterialFunctionCall_128, output=Outer}]
   properties {name=SDF_AA_Rank_Glyph, nodeColor=<rgba 0 0 0 1>}
  }
  MaterialExpressionNamedRerouteDeclaration_64 {
   class=MaterialExpressionNamedRerouteDeclaration
   inputs [None {from=MaterialExpressionMaterialFunctionCall_86, output="Shape (Color)"}]
   properties {name=I_SDF_AA_Glyph_Streak_Step7, nodeColor=<rgba 0 0 0 1>}
  }
  MaterialExpressionNamedRerouteDeclaration_65 {
   class=MaterialExpressionNamedRerouteDeclaration
   inputs [None {from=MaterialExpressionMaterialFunctionCall_87, output="Shape (Color)"}]
   properties {name=I_SDF_AA_Glyph_Streak_Step8, nodeColor=<rgba 0 0 0 1>}
  }
  MaterialExpressionNamedRerouteDeclaration_66 {
   class=MaterialExpressionNamedRerouteDeclaration
   inputs [None {from=MaterialExpressionMaterialFunctionCall_92, output="Shape (Color)"}]
   properties {name=I_SDF_AA_Glyph_Streak_Step9, nodeColor=<rgba 0 0 0 1>}
  }
  MaterialExpressionNamedRerouteDeclaration_67 {
   class=MaterialExpressionNamedRerouteDeclaration
   inputs [None {from=MaterialExpressionMaterialFunctionCall_93, output="Shape (Color)"}]
   properties {name=I_SDF_AA_Glyph_Streak_Step10, nodeColor=<rgba 0 0 0 1>}
  }
  MaterialExpressionNamedRerouteDeclaration_68 {
   class=MaterialExpressionNamedRerouteDeclaration
   inputs [None {from=MaterialExpressionMaterialFunctionCall_94, output="Shape (Color)"}]
   properties {name=I_SDF_AA_Glyph_Streak_Step11, nodeColor=<rgba 0 0 0 1>}
  }
  MaterialExpressionNamedRerouteDeclaration_69 {
   class=MaterialExpressionNamedRerouteDeclaration
   inputs [None {from=MaterialExpressionMaterialFunctionCall_85, output=Shape}]
   properties {name=I_SDF_AA_Glyph_Streak_Step_Mask_6, nodeColor=<rgba 0 0 0 1>}
  }
  MaterialExpressionNamedRerouteDeclaration_7 {
   class=MaterialExpressionNamedRerouteDeclaration
   inputs [None {from=MaterialExpressionVectorParameter_2, output=RGB}]
   properties {name=P_Color_BG_Outer, nodeColor=<rgba 0 0 0 1>}
  }
  MaterialExpressionNamedRerouteDeclaration_70 {
   class=MaterialExpressionNamedRerouteDeclaration
   inputs [None {from=MaterialExpressionMaterialFunctionCall_95, output="Shape (Color)"}]
   properties {name=I_SDF_AA_Glyph_Streak_Step12, nodeColor=<rgba 0 0 0 1>}
  }
  MaterialExpressionNamedRerouteDeclaration_71 {
   class=MaterialExpressionNamedRerouteDeclaration
   inputs [None {from=MaterialExpressionMaterialFunctionCall_88, output="Shape (Color)"}]
   properties {name=I_SDF_AA_Glyph_Streak_Step13, nodeColor=<rgba 0 0 0 1>}
  }
  MaterialExpressionNamedRerouteDeclaration_72 {
   class=MaterialExpressionNamedRerouteDeclaration
   inputs [None {from=MaterialExpressionMaterialFunctionCall_89, output="Shape (Color)"}]
   properties {name=I_SDF_AA_Glyph_Streak_Step14, nodeColor=<rgba 0 0 0 1>}
  }
  MaterialExpressionNamedRerouteDeclaration_73 {
   class=MaterialExpressionNamedRerouteDeclaration
   inputs [None {from=MaterialExpressionMaterialFunctionCall_90, output="Shape (Color)"}]
   properties {name=I_SDF_AA_Glyph_Streak_Step15, nodeColor=<rgba 0 0 0 1>}
  }
  MaterialExpressionNamedRerouteDeclaration_74 {
   class=MaterialExpressionNamedRerouteDeclaration
   inputs [None {from=MaterialExpressionSaturate_2}]
   properties {name=P_Rank_Radial_Glow_Opacity, nodeColor=<rgba 0 0 0 1>}
  }
  MaterialExpressionNamedRerouteDeclaration_75 {
   class=MaterialExpressionNamedRerouteDeclaration
   inputs [None {from=MaterialExpressionMaterialFunctionCall_91, output="Shape (Color)"}]
   properties {name=I_SDF_AA_Glyph_Streak_Step16, nodeColor=<rgba 0 0 0 1>}
  }
  MaterialExpressionNamedRerouteDeclaration_76 {
   class=MaterialExpressionNamedRerouteDeclaration
   inputs [None {from=MaterialExpressionSaturate_3}]
   properties {name=P_Rank_Streaks_Opacity, nodeColor=<rgba 0 0 0 1>}
  }
  MaterialExpressionNamedRerouteDeclaration_77 {
   class=MaterialExpressionNamedRerouteDeclaration
   inputs [None {from=MaterialExpressionVectorParameter_9, output=RGB}]
   properties {name=P_Color_Rank_Streak_Start, nodeColor=<rgba 0 0 0 1>}
  }
  MaterialExpressionNamedRerouteDeclaration_78 {
   class=MaterialExpressionNamedRerouteDeclaration
   inputs [None {from=MaterialExpressionVectorParameter_10, output=RGB}]
   properties {name=P_Color_Rank_Streak_End, nodeColor=<rgba 0 0 0 1>}
  }
  MaterialExpressionNamedRerouteDeclaration_79 {
   class=MaterialExpressionNamedRerouteDeclaration
   inputs [None {from=MaterialExpressionMaterialFunctionCall_86, output=Shape}]
   properties {name=I_SDF_AA_Glyph_Streak_Step_Mask_7, nodeColor=<rgba 0 0 0 1>}
  }
  MaterialExpressionNamedRerouteDeclaration_8 {
   class=MaterialExpressionNamedRerouteDeclaration
   inputs [None {from=MaterialExpressionVectorParameter_5, output=RGB}]
   properties {name=P_Color_Outline_Thin_Outer, nodeColor=<rgba 0 0 0 1>}
  }
  MaterialExpressionNamedRerouteDeclaration_80 {
   class=MaterialExpressionNamedRerouteDeclaration
   inputs [None {from=MaterialExpressionMaterialFunctionCall_87, output=Shape}]
   properties {name=I_SDF_AA_Glyph_Streak_Step_Mask_8, nodeColor=<rgba 0 0 0 1>}
  }
  MaterialExpressionNamedRerouteDeclaration_81 {
   class=MaterialExpressionNamedRerouteDeclaration
   inputs [None {from=MaterialExpressionMaterialFunctionCall_92, output=Shape}]
   properties {name=I_SDF_AA_Glyph_Streak_Step_Mask_9, nodeColor=<rgba 0 0 0 1>}
  }
  MaterialExpressionNamedRerouteDeclaration_82 {
   class=MaterialExpressionNamedRerouteDeclaration
   inputs [None {from=MaterialExpressionMaterialFunctionCall_93, output=Shape}]
   properties {name=I_SDF_AA_Glyph_Streak_Step_Mask_10, nodeColor=<rgba 0 0 0 1>}
  }
  MaterialExpressionNamedRerouteDeclaration_83 {
   class=MaterialExpressionNamedRerouteDeclaration
   inputs [None {from=MaterialExpressionMaterialFunctionCall_94, output=Shape}]
   properties {name=I_SDF_AA_Glyph_Streak_Step_Mask_11, nodeColor=<rgba 0 0 0 1>}
  }
  MaterialExpressionNamedRerouteDeclaration_84 {
   class=MaterialExpressionNamedRerouteDeclaration
   inputs [None {from=MaterialExpressionMaterialFunctionCall_95, output=Shape}]
   properties {name=I_SDF_AA_Glyph_Streak_Step_Mask_12, nodeColor=<rgba 0 0 0 1>}
  }
  MaterialExpressionNamedRerouteDeclaration_85 {
   class=MaterialExpressionNamedRerouteDeclaration
   inputs [None {from=MaterialExpressionMaterialFunctionCall_88, output=Shape}]
   properties {name=I_SDF_AA_Glyph_Streak_Step_Mask_13, nodeColor=<rgba 0 0 0 1>}
  }
  MaterialExpressionNamedRerouteDeclaration_86 {
   class=MaterialExpressionNamedRerouteDeclaration
   inputs [None {from=MaterialExpressionStaticSwitch_1}]
   properties {name=Final_Opacity, nodeColor=<rgba 0 1 0 1>}
  }
  MaterialExpressionNamedRerouteDeclaration_87 {
   class=MaterialExpressionNamedRerouteDeclaration
   inputs [None {from=MaterialExpressionMaterialFunctionCall_89, output=Shape}]
   properties {name=I_SDF_AA_Glyph_Streak_Step_Mask_14, nodeColor=<rgba 0 0 0 1>}
  }
  MaterialExpressionNamedRerouteDeclaration_88 {
   class=MaterialExpressionNamedRerouteDeclaration
   inputs [None {from=MaterialExpressionMaterialFunctionCall_90, output=Shape}]
   properties {name=I_SDF_AA_Glyph_Streak_Step_Mask_15, nodeColor=<rgba 0 0 0 1>}
  }
  MaterialExpressionNamedRerouteDeclaration_89 {
   class=MaterialExpressionNamedRerouteDeclaration
   inputs [None {from=MaterialExpressionSaturate_7}]
   properties {name=P_Rank_Opacity, nodeColor=<rgba 0 0 0 1>}
  }
  MaterialExpressionNamedRerouteDeclaration_9 {
   class=MaterialExpressionNamedRerouteDeclaration
   inputs [None {from=MaterialExpressionMax_30}]
   properties {name=I_SDF_AA_Streaks, nodeColor=<rgba 0 0 0 1>}
  }
  MaterialExpressionNamedRerouteDeclaration_90 {
   class=MaterialExpressionNamedRerouteDeclaration
   inputs [None {from=MaterialExpressionMaterialFunctionCall_91, output=Shape}]
   properties {name=I_SDF_AA_Glyph_Streak_Step_Mask_16, nodeColor=<rgba 0 0 0 1>}
  }
  MaterialExpressionNamedRerouteDeclaration_91 {
   class=MaterialExpressionNamedRerouteDeclaration
   inputs [None {from=MaterialExpressionMax_28}]
   properties {name=I_SDF_AA_Streaks_Mask, nodeColor=<rgba 0 0 0 1>}
  }
  MaterialExpressionNamedRerouteDeclaration_92 {
   class=MaterialExpressionNamedRerouteDeclaration
   inputs [None {from=MaterialExpressionScalarParameter_9}]
   properties {name=P_Rank_Streaks_Scale_Start, nodeColor=<rgba 0 0 0 1>}
  }
  MaterialExpressionNamedRerouteDeclaration_93 {
   class=MaterialExpressionNamedRerouteDeclaration
   inputs [None {from=MaterialExpressionScalarParameter_15}]
   properties {name=P_Rank_Streaks_Scale_End, nodeColor=<rgba 0 0 0 1>}
  }
  MaterialExpressionNamedRerouteDeclaration_94 {
   class=MaterialExpressionNamedRerouteDeclaration
   inputs [None {from=MaterialExpressionStaticBoolParameter_5}]
   properties {name=P_Outline_Thin_Inner_Enabled, nodeColor=<rgba 0 0 0 1>}
  }
  MaterialExpressionNamedRerouteDeclaration_95 {
   class=MaterialExpressionNamedRerouteDeclaration
   inputs [None {from=MaterialExpressionStaticBoolParameter_6}]
   properties {name=P_Outline_Thick_Outer_Enabled, nodeColor=<rgba 0 0 0 1>}
  }
  MaterialExpressionNamedRerouteDeclaration_96 {
   class=MaterialExpressionNamedRerouteDeclaration
   inputs [None {from=MaterialExpressionStaticBoolParameter_7}]
   properties {name=P_Rank_Enabled, nodeColor=<rgba 0 0 0 1>}
  }
  MaterialExpressionNamedRerouteDeclaration_97 {
   class=MaterialExpressionNamedRerouteDeclaration
   inputs [None {from=MaterialExpressionMaterialFunctionCall_129, output=UVs}]
   properties {name=I_Rank_UVs, nodeColor=<rgba 0 0 0 1>}
  }
  MaterialExpressionNamedRerouteDeclaration_98 {
   class=MaterialExpressionNamedRerouteDeclaration
   inputs [None {from=MaterialExpressionLinearInterpolate_0}]
   properties {name=I_Color_Solid_Inner, nodeColor=<rgba 0 0 0 1>}
  }
  MaterialExpressionNamedRerouteDeclaration_99 {
   class=MaterialExpressionNamedRerouteDeclaration
   inputs [None {from=MaterialExpressionReroute_95}]
   properties {name=I_Color_Solid_Outer, nodeColor=<rgba 0 0 0 1>}
  }
  MaterialExpressionNamedRerouteUsage_0 {class=MaterialExpressionNamedRerouteUsage}
  MaterialExpressionNamedRerouteUsage_1 {class=MaterialExpressionNamedRerouteUsage}
  MaterialExpressionNamedRerouteUsage_10 {class=MaterialExpressionNamedRerouteUsage}
  MaterialExpressionNamedRerouteUsage_100 {class=MaterialExpressionNamedRerouteUsage}
  MaterialExpressionNamedRerouteUsage_101 {class=MaterialExpressionNamedRerouteUsage}
  MaterialExpressionNamedRerouteUsage_102 {class=MaterialExpressionNamedRerouteUsage}
  MaterialExpressionNamedRerouteUsage_103 {class=MaterialExpressionNamedRerouteUsage}
  MaterialExpressionNamedRerouteUsage_104 {class=MaterialExpressionNamedRerouteUsage}
  MaterialExpressionNamedRerouteUsage_105 {class=MaterialExpressionNamedRerouteUsage}
  MaterialExpressionNamedRerouteUsage_106 {class=MaterialExpressionNamedRerouteUsage}
  MaterialExpressionNamedRerouteUsage_107 {class=MaterialExpressionNamedRerouteUsage}
  MaterialExpressionNamedRerouteUsage_108 {class=MaterialExpressionNamedRerouteUsage}
  MaterialExpressionNamedRerouteUsage_109 {class=MaterialExpressionNamedRerouteUsage}
  MaterialExpressionNamedRerouteUsage_11 {class=MaterialExpressionNamedRerouteUsage}
  MaterialExpressionNamedRerouteUsage_110 {class=MaterialExpressionNamedRerouteUsage}
  MaterialExpressionNamedRerouteUsage_111 {class=MaterialExpressionNamedRerouteUsage}
  MaterialExpressionNamedRerouteUsage_112 {class=MaterialExpressionNamedRerouteUsage}
  MaterialExpressionNamedRerouteUsage_113 {class=MaterialExpressionNamedRerouteUsage}
  MaterialExpressionNamedRerouteUsage_114 {class=MaterialExpressionNamedRerouteUsage}
  MaterialExpressionNamedRerouteUsage_115 {class=MaterialExpressionNamedRerouteUsage}
  MaterialExpressionNamedRerouteUsage_116 {class=MaterialExpressionNamedRerouteUsage}
  MaterialExpressionNamedRerouteUsage_117 {class=MaterialExpressionNamedRerouteUsage}
  MaterialExpressionNamedRerouteUsage_118 {class=MaterialExpressionNamedRerouteUsage}
  MaterialExpressionNamedRerouteUsage_119 {class=MaterialExpressionNamedRerouteUsage}
  MaterialExpressionNamedRerouteUsage_12 {class=MaterialExpressionNamedRerouteUsage}
  MaterialExpressionNamedRerouteUsage_120 {class=MaterialExpressionNamedRerouteUsage}
  MaterialExpressionNamedRerouteUsage_121 {class=MaterialExpressionNamedRerouteUsage}
  MaterialExpressionNamedRerouteUsage_122 {class=MaterialExpressionNamedRerouteUsage}
  MaterialExpressionNamedRerouteUsage_123 {class=MaterialExpressionNamedRerouteUsage}
  MaterialExpressionNamedRerouteUsage_124 {class=MaterialExpressionNamedRerouteUsage}
  MaterialExpressionNamedRerouteUsage_125 {class=MaterialExpressionNamedRerouteUsage}
  MaterialExpressionNamedRerouteUsage_126 {class=MaterialExpressionNamedRerouteUsage}
  MaterialExpressionNamedRerouteUsage_127 {class=MaterialExpressionNamedRerouteUsage}
  MaterialExpressionNamedRerouteUsage_128 {class=MaterialExpressionNamedRerouteUsage}
  MaterialExpressionNamedRerouteUsage_129 {class=MaterialExpressionNamedRerouteUsage}
  MaterialExpressionNamedRerouteUsage_13 {class=MaterialExpressionNamedRerouteUsage}
  MaterialExpressionNamedRerouteUsage_130 {class=MaterialExpressionNamedRerouteUsage}
  MaterialExpressionNamedRerouteUsage_131 {class=MaterialExpressionNamedRerouteUsage}
  MaterialExpressionNamedRerouteUsage_132 {class=MaterialExpressionNamedRerouteUsage}
  MaterialExpressionNamedRerouteUsage_133 {class=MaterialExpressionNamedRerouteUsage}
  MaterialExpressionNamedRerouteUsage_134 {class=MaterialExpressionNamedRerouteUsage}
  MaterialExpressionNamedRerouteUsage_135 {class=MaterialExpressionNamedRerouteUsage}
  MaterialExpressionNamedRerouteUsage_136 {class=MaterialExpressionNamedRerouteUsage}
  MaterialExpressionNamedRerouteUsage_137 {class=MaterialExpressionNamedRerouteUsage}
  MaterialExpressionNamedRerouteUsage_138 {class=MaterialExpressionNamedRerouteUsage}
  MaterialExpressionNamedRerouteUsage_14 {class=MaterialExpressionNamedRerouteUsage}
  MaterialExpressionNamedRerouteUsage_142 {class=MaterialExpressionNamedRerouteUsage}
  MaterialExpressionNamedRerouteUsage_144 {class=MaterialExpressionNamedRerouteUsage}
  MaterialExpressionNamedRerouteUsage_145 {class=MaterialExpressionNamedRerouteUsage}
  MaterialExpressionNamedRerouteUsage_146 {class=MaterialExpressionNamedRerouteUsage}
  MaterialExpressionNamedRerouteUsage_15 {class=MaterialExpressionNamedRerouteUsage}
  MaterialExpressionNamedRerouteUsage_150 {class=MaterialExpressionNamedRerouteUsage}
  MaterialExpressionNamedRerouteUsage_16 {class=MaterialExpressionNamedRerouteUsage}
  MaterialExpressionNamedRerouteUsage_17 {class=MaterialExpressionNamedRerouteUsage}
  MaterialExpressionNamedRerouteUsage_18 {class=MaterialExpressionNamedRerouteUsage}
  MaterialExpressionNamedRerouteUsage_19 {class=MaterialExpressionNamedRerouteUsage}
  MaterialExpressionNamedRerouteUsage_2 {class=MaterialExpressionNamedRerouteUsage}
  MaterialExpressionNamedRerouteUsage_20 {class=MaterialExpressionNamedRerouteUsage}
  MaterialExpressionNamedRerouteUsage_205 {class=MaterialExpressionNamedRerouteUsage}
  MaterialExpressionNamedRerouteUsage_206 {class=MaterialExpressionNamedRerouteUsage}
  MaterialExpressionNamedRerouteUsage_207 {class=MaterialExpressionNamedRerouteUsage}
  MaterialExpressionNamedRerouteUsage_208 {class=MaterialExpressionNamedRerouteUsage}
  MaterialExpressionNamedRerouteUsage_209 {class=MaterialExpressionNamedRerouteUsage}
  MaterialExpressionNamedRerouteUsage_21 {class=MaterialExpressionNamedRerouteUsage}
  MaterialExpressionNamedRerouteUsage_211 {class=MaterialExpressionNamedRerouteUsage}
  MaterialExpressionNamedRerouteUsage_22 {class=MaterialExpressionNamedRerouteUsage}
  MaterialExpressionNamedRerouteUsage_23 {class=MaterialExpressionNamedRerouteUsage}
  MaterialExpressionNamedRerouteUsage_24 {class=MaterialExpressionNamedRerouteUsage}
  MaterialExpressionNamedRerouteUsage_242 {class=MaterialExpressionNamedRerouteUsage}
  MaterialExpressionNamedRerouteUsage_243 {class=MaterialExpressionNamedRerouteUsage}
  MaterialExpressionNamedRerouteUsage_244 {class=MaterialExpressionNamedRerouteUsage}
  MaterialExpressionNamedRerouteUsage_245 {class=MaterialExpressionNamedRerouteUsage}
  MaterialExpressionNamedRerouteUsage_246 {class=MaterialExpressionNamedRerouteUsage}
  MaterialExpressionNamedRerouteUsage_247 {class=MaterialExpressionNamedRerouteUsage}
  MaterialExpressionNamedRerouteUsage_248 {class=MaterialExpressionNamedRerouteUsage}
  MaterialExpressionNamedRerouteUsage_249 {class=MaterialExpressionNamedRerouteUsage}
  MaterialExpressionNamedRerouteUsage_25 {class=MaterialExpressionNamedRerouteUsage}
  MaterialExpressionNamedRerouteUsage_250 {class=MaterialExpressionNamedRerouteUsage}
  MaterialExpressionNamedRerouteUsage_251 {class=MaterialExpressionNamedRerouteUsage}
  MaterialExpressionNamedRerouteUsage_252 {class=MaterialExpressionNamedRerouteUsage}
  MaterialExpressionNamedRerouteUsage_253 {class=MaterialExpressionNamedRerouteUsage}
  MaterialExpressionNamedRerouteUsage_254 {class=MaterialExpressionNamedRerouteUsage}
  MaterialExpressionNamedRerouteUsage_255 {class=MaterialExpressionNamedRerouteUsage}
  MaterialExpressionNamedRerouteUsage_256 {class=MaterialExpressionNamedRerouteUsage}
  MaterialExpressionNamedRerouteUsage_257 {class=MaterialExpressionNamedRerouteUsage}
  MaterialExpressionNamedRerouteUsage_258 {class=MaterialExpressionNamedRerouteUsage}
  MaterialExpressionNamedRerouteUsage_259 {class=MaterialExpressionNamedRerouteUsage}
  MaterialExpressionNamedRerouteUsage_26 {class=MaterialExpressionNamedRerouteUsage}
  MaterialExpressionNamedRerouteUsage_260 {class=MaterialExpressionNamedRerouteUsage}
  MaterialExpressionNamedRerouteUsage_261 {class=MaterialExpressionNamedRerouteUsage}
  MaterialExpressionNamedRerouteUsage_262 {class=MaterialExpressionNamedRerouteUsage}
  MaterialExpressionNamedRerouteUsage_263 {class=MaterialExpressionNamedRerouteUsage}
  MaterialExpressionNamedRerouteUsage_264 {class=MaterialExpressionNamedRerouteUsage}
  MaterialExpressionNamedRerouteUsage_265 {class=MaterialExpressionNamedRerouteUsage}
  MaterialExpressionNamedRerouteUsage_266 {class=MaterialExpressionNamedRerouteUsage}
  MaterialExpressionNamedRerouteUsage_267 {class=MaterialExpressionNamedRerouteUsage}
  MaterialExpressionNamedRerouteUsage_268 {class=MaterialExpressionNamedRerouteUsage}
  MaterialExpressionNamedRerouteUsage_269 {class=MaterialExpressionNamedRerouteUsage}
  MaterialExpressionNamedRerouteUsage_27 {class=MaterialExpressionNamedRerouteUsage}
  MaterialExpressionNamedRerouteUsage_270 {class=MaterialExpressionNamedRerouteUsage}
  MaterialExpressionNamedRerouteUsage_271 {class=MaterialExpressionNamedRerouteUsage}
  MaterialExpressionNamedRerouteUsage_272 {class=MaterialExpressionNamedRerouteUsage}
  MaterialExpressionNamedRerouteUsage_28 {class=MaterialExpressionNamedRerouteUsage}
  MaterialExpressionNamedRerouteUsage_29 {class=MaterialExpressionNamedRerouteUsage}
  MaterialExpressionNamedRerouteUsage_3 {class=MaterialExpressionNamedRerouteUsage}
  MaterialExpressionNamedRerouteUsage_30 {class=MaterialExpressionNamedRerouteUsage}
  MaterialExpressionNamedRerouteUsage_31 {class=MaterialExpressionNamedRerouteUsage}
  MaterialExpressionNamedRerouteUsage_32 {class=MaterialExpressionNamedRerouteUsage}
  MaterialExpressionNamedRerouteUsage_322 {class=MaterialExpressionNamedRerouteUsage}
  MaterialExpressionNamedRerouteUsage_323 {class=MaterialExpressionNamedRerouteUsage}
  MaterialExpressionNamedRerouteUsage_324 {class=MaterialExpressionNamedRerouteUsage}
  MaterialExpressionNamedRerouteUsage_325 {class=MaterialExpressionNamedRerouteUsage}
  MaterialExpressionNamedRerouteUsage_326 {class=MaterialExpressionNamedRerouteUsage}
  MaterialExpressionNamedRerouteUsage_327 {class=MaterialExpressionNamedRerouteUsage}
  MaterialExpressionNamedRerouteUsage_328 {class=MaterialExpressionNamedRerouteUsage}
  MaterialExpressionNamedRerouteUsage_329 {class=MaterialExpressionNamedRerouteUsage}
  MaterialExpressionNamedRerouteUsage_33 {class=MaterialExpressionNamedRerouteUsage}
  MaterialExpressionNamedRerouteUsage_330 {class=MaterialExpressionNamedRerouteUsage}
  MaterialExpressionNamedRerouteUsage_34 {class=MaterialExpressionNamedRerouteUsage}
  MaterialExpressionNamedRerouteUsage_35 {class=MaterialExpressionNamedRerouteUsage}
  MaterialExpressionNamedRerouteUsage_37 {class=MaterialExpressionNamedRerouteUsage}
  MaterialExpressionNamedRerouteUsage_38 {class=MaterialExpressionNamedRerouteUsage}
  MaterialExpressionNamedRerouteUsage_387 {class=MaterialExpressionNamedRerouteUsage}
  MaterialExpressionNamedRerouteUsage_388 {class=MaterialExpressionNamedRerouteUsage}
  MaterialExpressionNamedRerouteUsage_389 {class=MaterialExpressionNamedRerouteUsage}
  MaterialExpressionNamedRerouteUsage_39 {class=MaterialExpressionNamedRerouteUsage}
  MaterialExpressionNamedRerouteUsage_390 {class=MaterialExpressionNamedRerouteUsage}
  MaterialExpressionNamedRerouteUsage_391 {class=MaterialExpressionNamedRerouteUsage}
  MaterialExpressionNamedRerouteUsage_392 {class=MaterialExpressionNamedRerouteUsage}
  MaterialExpressionNamedRerouteUsage_393 {class=MaterialExpressionNamedRerouteUsage}
  MaterialExpressionNamedRerouteUsage_394 {class=MaterialExpressionNamedRerouteUsage}
  MaterialExpressionNamedRerouteUsage_395 {class=MaterialExpressionNamedRerouteUsage}
  MaterialExpressionNamedRerouteUsage_396 {class=MaterialExpressionNamedRerouteUsage}
  MaterialExpressionNamedRerouteUsage_397 {class=MaterialExpressionNamedRerouteUsage}
  MaterialExpressionNamedRerouteUsage_398 {class=MaterialExpressionNamedRerouteUsage}
  MaterialExpressionNamedRerouteUsage_399 {class=MaterialExpressionNamedRerouteUsage}
  MaterialExpressionNamedRerouteUsage_4 {class=MaterialExpressionNamedRerouteUsage}
  MaterialExpressionNamedRerouteUsage_400 {class=MaterialExpressionNamedRerouteUsage}
  MaterialExpressionNamedRerouteUsage_401 {class=MaterialExpressionNamedRerouteUsage}
  MaterialExpressionNamedRerouteUsage_402 {class=MaterialExpressionNamedRerouteUsage}
  MaterialExpressionNamedRerouteUsage_403 {class=MaterialExpressionNamedRerouteUsage}
  MaterialExpressionNamedRerouteUsage_404 {class=MaterialExpressionNamedRerouteUsage}
  MaterialExpressionNamedRerouteUsage_405 {class=MaterialExpressionNamedRerouteUsage}
  MaterialExpressionNamedRerouteUsage_406 {class=MaterialExpressionNamedRerouteUsage}
  MaterialExpressionNamedRerouteUsage_407 {class=MaterialExpressionNamedRerouteUsage}
  MaterialExpressionNamedRerouteUsage_408 {class=MaterialExpressionNamedRerouteUsage}
  MaterialExpressionNamedRerouteUsage_409 {class=MaterialExpressionNamedRerouteUsage}
  MaterialExpressionNamedRerouteUsage_41 {class=MaterialExpressionNamedRerouteUsage}
  MaterialExpressionNamedRerouteUsage_410 {class=MaterialExpressionNamedRerouteUsage}
  MaterialExpressionNamedRerouteUsage_411 {class=MaterialExpressionNamedRerouteUsage}
  MaterialExpressionNamedRerouteUsage_412 {class=MaterialExpressionNamedRerouteUsage}
  MaterialExpressionNamedRerouteUsage_413 {class=MaterialExpressionNamedRerouteUsage}
  MaterialExpressionNamedRerouteUsage_414 {class=MaterialExpressionNamedRerouteUsage}
  MaterialExpressionNamedRerouteUsage_415 {class=MaterialExpressionNamedRerouteUsage}
  MaterialExpressionNamedRerouteUsage_416 {class=MaterialExpressionNamedRerouteUsage}
  MaterialExpressionNamedRerouteUsage_417 {class=MaterialExpressionNamedRerouteUsage}
  MaterialExpressionNamedRerouteUsage_418 {class=MaterialExpressionNamedRerouteUsage}
  MaterialExpressionNamedRerouteUsage_419 {class=MaterialExpressionNamedRerouteUsage}
  MaterialExpressionNamedRerouteUsage_42 {class=MaterialExpressionNamedRerouteUsage}
  MaterialExpressionNamedRerouteUsage_420 {class=MaterialExpressionNamedRerouteUsage}
  MaterialExpressionNamedRerouteUsage_421 {class=MaterialExpressionNamedRerouteUsage}
  MaterialExpressionNamedRerouteUsage_422 {class=MaterialExpressionNamedRerouteUsage}
  MaterialExpressionNamedRerouteUsage_423 {class=MaterialExpressionNamedRerouteUsage}
  MaterialExpressionNamedRerouteUsage_424 {class=MaterialExpressionNamedRerouteUsage}
  MaterialExpressionNamedRerouteUsage_425 {class=MaterialExpressionNamedRerouteUsage}
  MaterialExpressionNamedRerouteUsage_426 {class=MaterialExpressionNamedRerouteUsage}
  MaterialExpressionNamedRerouteUsage_427 {class=MaterialExpressionNamedRerouteUsage}
  MaterialExpressionNamedRerouteUsage_428 {class=MaterialExpressionNamedRerouteUsage}
  MaterialExpressionNamedRerouteUsage_429 {class=MaterialExpressionNamedRerouteUsage}
  MaterialExpressionNamedRerouteUsage_43 {class=MaterialExpressionNamedRerouteUsage}
  MaterialExpressionNamedRerouteUsage_430 {class=MaterialExpressionNamedRerouteUsage}
  MaterialExpressionNamedRerouteUsage_431 {class=MaterialExpressionNamedRerouteUsage}
  MaterialExpressionNamedRerouteUsage_432 {class=MaterialExpressionNamedRerouteUsage}
  MaterialExpressionNamedRerouteUsage_433 {class=MaterialExpressionNamedRerouteUsage}
  MaterialExpressionNamedRerouteUsage_434 {class=MaterialExpressionNamedRerouteUsage}
  MaterialExpressionNamedRerouteUsage_435 {class=MaterialExpressionNamedRerouteUsage}
  MaterialExpressionNamedRerouteUsage_436 {class=MaterialExpressionNamedRerouteUsage}
  MaterialExpressionNamedRerouteUsage_437 {class=MaterialExpressionNamedRerouteUsage}
  MaterialExpressionNamedRerouteUsage_438 {class=MaterialExpressionNamedRerouteUsage}
  MaterialExpressionNamedRerouteUsage_439 {class=MaterialExpressionNamedRerouteUsage}
  MaterialExpressionNamedRerouteUsage_44 {class=MaterialExpressionNamedRerouteUsage}
  MaterialExpressionNamedRerouteUsage_440 {class=MaterialExpressionNamedRerouteUsage}
  MaterialExpressionNamedRerouteUsage_441 {class=MaterialExpressionNamedRerouteUsage}
  MaterialExpressionNamedRerouteUsage_442 {class=MaterialExpressionNamedRerouteUsage}
  MaterialExpressionNamedRerouteUsage_443 {class=MaterialExpressionNamedRerouteUsage}
  MaterialExpressionNamedRerouteUsage_444 {class=MaterialExpressionNamedRerouteUsage}
  MaterialExpressionNamedRerouteUsage_445 {class=MaterialExpressionNamedRerouteUsage}
  MaterialExpressionNamedRerouteUsage_446 {class=MaterialExpressionNamedRerouteUsage}
  MaterialExpressionNamedRerouteUsage_447 {class=MaterialExpressionNamedRerouteUsage}
  MaterialExpressionNamedRerouteUsage_448 {class=MaterialExpressionNamedRerouteUsage}
  MaterialExpressionNamedRerouteUsage_449 {class=MaterialExpressionNamedRerouteUsage}
  MaterialExpressionNamedRerouteUsage_45 {class=MaterialExpressionNamedRerouteUsage}
  MaterialExpressionNamedRerouteUsage_450 {class=MaterialExpressionNamedRerouteUsage}
  MaterialExpressionNamedRerouteUsage_451 {class=MaterialExpressionNamedRerouteUsage}
  MaterialExpressionNamedRerouteUsage_452 {class=MaterialExpressionNamedRerouteUsage}
  MaterialExpressionNamedRerouteUsage_453 {class=MaterialExpressionNamedRerouteUsage}
  MaterialExpressionNamedRerouteUsage_454 {class=MaterialExpressionNamedRerouteUsage}
  MaterialExpressionNamedRerouteUsage_455 {class=MaterialExpressionNamedRerouteUsage}
  MaterialExpressionNamedRerouteUsage_456 {class=MaterialExpressionNamedRerouteUsage}
  MaterialExpressionNamedRerouteUsage_457 {class=MaterialExpressionNamedRerouteUsage}
  MaterialExpressionNamedRerouteUsage_458 {class=MaterialExpressionNamedRerouteUsage}
  MaterialExpressionNamedRerouteUsage_459 {class=MaterialExpressionNamedRerouteUsage}
  MaterialExpressionNamedRerouteUsage_46 {class=MaterialExpressionNamedRerouteUsage}
  MaterialExpressionNamedRerouteUsage_460 {class=MaterialExpressionNamedRerouteUsage}
  MaterialExpressionNamedRerouteUsage_461 {class=MaterialExpressionNamedRerouteUsage}
  MaterialExpressionNamedRerouteUsage_462 {class=MaterialExpressionNamedRerouteUsage}
  MaterialExpressionNamedRerouteUsage_463 {class=MaterialExpressionNamedRerouteUsage}
  MaterialExpressionNamedRerouteUsage_464 {class=MaterialExpressionNamedRerouteUsage}
  MaterialExpressionNamedRerouteUsage_465 {class=MaterialExpressionNamedRerouteUsage}
  MaterialExpressionNamedRerouteUsage_466 {class=MaterialExpressionNamedRerouteUsage}
  MaterialExpressionNamedRerouteUsage_467 {class=MaterialExpressionNamedRerouteUsage}
  MaterialExpressionNamedRerouteUsage_468 {class=MaterialExpressionNamedRerouteUsage}
  MaterialExpressionNamedRerouteUsage_469 {class=MaterialExpressionNamedRerouteUsage}
  MaterialExpressionNamedRerouteUsage_47 {class=MaterialExpressionNamedRerouteUsage}
  MaterialExpressionNamedRerouteUsage_470 {class=MaterialExpressionNamedRerouteUsage}
  MaterialExpressionNamedRerouteUsage_471 {class=MaterialExpressionNamedRerouteUsage}
  MaterialExpressionNamedRerouteUsage_472 {class=MaterialExpressionNamedRerouteUsage}
  MaterialExpressionNamedRerouteUsage_473 {class=MaterialExpressionNamedRerouteUsage}
  MaterialExpressionNamedRerouteUsage_474 {class=MaterialExpressionNamedRerouteUsage}
  MaterialExpressionNamedRerouteUsage_475 {class=MaterialExpressionNamedRerouteUsage}
  MaterialExpressionNamedRerouteUsage_476 {class=MaterialExpressionNamedRerouteUsage}
  MaterialExpressionNamedRerouteUsage_477 {class=MaterialExpressionNamedRerouteUsage}
  MaterialExpressionNamedRerouteUsage_478 {class=MaterialExpressionNamedRerouteUsage}
  MaterialExpressionNamedRerouteUsage_479 {class=MaterialExpressionNamedRerouteUsage}
  MaterialExpressionNamedRerouteUsage_48 {class=MaterialExpressionNamedRerouteUsage}
  MaterialExpressionNamedRerouteUsage_480 {class=MaterialExpressionNamedRerouteUsage}
  MaterialExpressionNamedRerouteUsage_481 {class=MaterialExpressionNamedRerouteUsage}
  MaterialExpressionNamedRerouteUsage_482 {class=MaterialExpressionNamedRerouteUsage}
  MaterialExpressionNamedRerouteUsage_483 {class=MaterialExpressionNamedRerouteUsage}
  MaterialExpressionNamedRerouteUsage_484 {class=MaterialExpressionNamedRerouteUsage}
  MaterialExpressionNamedRerouteUsage_485 {class=MaterialExpressionNamedRerouteUsage}
  MaterialExpressionNamedRerouteUsage_486 {class=MaterialExpressionNamedRerouteUsage}
  MaterialExpressionNamedRerouteUsage_487 {class=MaterialExpressionNamedRerouteUsage}
  MaterialExpressionNamedRerouteUsage_488 {class=MaterialExpressionNamedRerouteUsage}
  MaterialExpressionNamedRerouteUsage_489 {class=MaterialExpressionNamedRerouteUsage}
  MaterialExpressionNamedRerouteUsage_490 {class=MaterialExpressionNamedRerouteUsage}
  MaterialExpressionNamedRerouteUsage_491 {class=MaterialExpressionNamedRerouteUsage}
  MaterialExpressionNamedRerouteUsage_492 {class=MaterialExpressionNamedRerouteUsage}
  MaterialExpressionNamedRerouteUsage_493 {class=MaterialExpressionNamedRerouteUsage}
  MaterialExpressionNamedRerouteUsage_494 {class=MaterialExpressionNamedRerouteUsage}
  MaterialExpressionNamedRerouteUsage_495 {class=MaterialExpressionNamedRerouteUsage}
  MaterialExpressionNamedRerouteUsage_496 {class=MaterialExpressionNamedRerouteUsage}
  MaterialExpressionNamedRerouteUsage_497 {class=MaterialExpressionNamedRerouteUsage}
  MaterialExpressionNamedRerouteUsage_498 {class=MaterialExpressionNamedRerouteUsage}
  MaterialExpressionNamedRerouteUsage_499 {class=MaterialExpressionNamedRerouteUsage}
  MaterialExpressionNamedRerouteUsage_5 {class=MaterialExpressionNamedRerouteUsage}
  MaterialExpressionNamedRerouteUsage_500 {class=MaterialExpressionNamedRerouteUsage}
  MaterialExpressionNamedRerouteUsage_501 {class=MaterialExpressionNamedRerouteUsage}
  MaterialExpressionNamedRerouteUsage_502 {class=MaterialExpressionNamedRerouteUsage}
  MaterialExpressionNamedRerouteUsage_503 {class=MaterialExpressionNamedRerouteUsage}
  MaterialExpressionNamedRerouteUsage_504 {class=MaterialExpressionNamedRerouteUsage}
  MaterialExpressionNamedRerouteUsage_505 {class=MaterialExpressionNamedRerouteUsage}
  MaterialExpressionNamedRerouteUsage_506 {class=MaterialExpressionNamedRerouteUsage}
  MaterialExpressionNamedRerouteUsage_507 {class=MaterialExpressionNamedRerouteUsage}
  MaterialExpressionNamedRerouteUsage_508 {class=MaterialExpressionNamedRerouteUsage}
  MaterialExpressionNamedRerouteUsage_509 {class=MaterialExpressionNamedRerouteUsage}
  MaterialExpressionNamedRerouteUsage_51 {class=MaterialExpressionNamedRerouteUsage}
  MaterialExpressionNamedRerouteUsage_510 {class=MaterialExpressionNamedRerouteUsage}
  MaterialExpressionNamedRerouteUsage_511 {class=MaterialExpressionNamedRerouteUsage}
  MaterialExpressionNamedRerouteUsage_512 {class=MaterialExpressionNamedRerouteUsage}
  MaterialExpressionNamedRerouteUsage_513 {class=MaterialExpressionNamedRerouteUsage}
  MaterialExpressionNamedRerouteUsage_514 {class=MaterialExpressionNamedRerouteUsage}
  MaterialExpressionNamedRerouteUsage_515 {class=MaterialExpressionNamedRerouteUsage}
  MaterialExpressionNamedRerouteUsage_516 {class=MaterialExpressionNamedRerouteUsage}
  MaterialExpressionNamedRerouteUsage_517 {class=MaterialExpressionNamedRerouteUsage}
  MaterialExpressionNamedRerouteUsage_519 {class=MaterialExpressionNamedRerouteUsage}
  MaterialExpressionNamedRerouteUsage_52 {class=MaterialExpressionNamedRerouteUsage}
  MaterialExpressionNamedRerouteUsage_520 {class=MaterialExpressionNamedRerouteUsage}
  MaterialExpressionNamedRerouteUsage_522 {class=MaterialExpressionNamedRerouteUsage}
  MaterialExpressionNamedRerouteUsage_523 {class=MaterialExpressionNamedRerouteUsage}
  MaterialExpressionNamedRerouteUsage_524 {class=MaterialExpressionNamedRerouteUsage}
  MaterialExpressionNamedRerouteUsage_525 {class=MaterialExpressionNamedRerouteUsage}
  MaterialExpressionNamedRerouteUsage_526 {class=MaterialExpressionNamedRerouteUsage}
  MaterialExpressionNamedRerouteUsage_527 {class=MaterialExpressionNamedRerouteUsage}
  MaterialExpressionNamedRerouteUsage_528 {class=MaterialExpressionNamedRerouteUsage}
  MaterialExpressionNamedRerouteUsage_529 {class=MaterialExpressionNamedRerouteUsage}
  MaterialExpressionNamedRerouteUsage_53 {class=MaterialExpressionNamedRerouteUsage}
  MaterialExpressionNamedRerouteUsage_530 {class=MaterialExpressionNamedRerouteUsage}
  MaterialExpressionNamedRerouteUsage_531 {class=MaterialExpressionNamedRerouteUsage}
  MaterialExpressionNamedRerouteUsage_532 {class=MaterialExpressionNamedRerouteUsage}
  MaterialExpressionNamedRerouteUsage_533 {class=MaterialExpressionNamedRerouteUsage}
  MaterialExpressionNamedRerouteUsage_534 {class=MaterialExpressionNamedRerouteUsage}
  MaterialExpressionNamedRerouteUsage_535 {class=MaterialExpressionNamedRerouteUsage}
  MaterialExpressionNamedRerouteUsage_536 {class=MaterialExpressionNamedRerouteUsage}
  MaterialExpressionNamedRerouteUsage_537 {class=MaterialExpressionNamedRerouteUsage}
  MaterialExpressionNamedRerouteUsage_538 {class=MaterialExpressionNamedRerouteUsage}
  MaterialExpressionNamedRerouteUsage_539 {class=MaterialExpressionNamedRerouteUsage}
  MaterialExpressionNamedRerouteUsage_54 {class=MaterialExpressionNamedRerouteUsage}
  MaterialExpressionNamedRerouteUsage_540 {class=MaterialExpressionNamedRerouteUsage}
  MaterialExpressionNamedRerouteUsage_541 {class=MaterialExpressionNamedRerouteUsage}
  MaterialExpressionNamedRerouteUsage_542 {class=MaterialExpressionNamedRerouteUsage}
  MaterialExpressionNamedRerouteUsage_543 {class=MaterialExpressionNamedRerouteUsage}
  MaterialExpressionNamedRerouteUsage_544 {class=MaterialExpressionNamedRerouteUsage}
  MaterialExpressionNamedRerouteUsage_545 {class=MaterialExpressionNamedRerouteUsage}
  MaterialExpressionNamedRerouteUsage_546 {class=MaterialExpressionNamedRerouteUsage}
  MaterialExpressionNamedRerouteUsage_547 {class=MaterialExpressionNamedRerouteUsage}
  MaterialExpressionNamedRerouteUsage_548 {class=MaterialExpressionNamedRerouteUsage}
  MaterialExpressionNamedRerouteUsage_549 {class=MaterialExpressionNamedRerouteUsage}
  MaterialExpressionNamedRerouteUsage_55 {class=MaterialExpressionNamedRerouteUsage}
  MaterialExpressionNamedRerouteUsage_550 {class=MaterialExpressionNamedRerouteUsage}
  MaterialExpressionNamedRerouteUsage_551 {class=MaterialExpressionNamedRerouteUsage}
  MaterialExpressionNamedRerouteUsage_552 {class=MaterialExpressionNamedRerouteUsage}
  MaterialExpressionNamedRerouteUsage_553 {class=MaterialExpressionNamedRerouteUsage}
  MaterialExpressionNamedRerouteUsage_554 {class=MaterialExpressionNamedRerouteUsage}
  MaterialExpressionNamedRerouteUsage_56 {class=MaterialExpressionNamedRerouteUsage}
  MaterialExpressionNamedRerouteUsage_57 {class=MaterialExpressionNamedRerouteUsage}
  MaterialExpressionNamedRerouteUsage_58 {class=MaterialExpressionNamedRerouteUsage}
  MaterialExpressionNamedRerouteUsage_59 {class=MaterialExpressionNamedRerouteUsage}
  MaterialExpressionNamedRerouteUsage_60 {class=MaterialExpressionNamedRerouteUsage}
  MaterialExpressionNamedRerouteUsage_62 {class=MaterialExpressionNamedRerouteUsage}
  MaterialExpressionNamedRerouteUsage_63 {class=MaterialExpressionNamedRerouteUsage}
  MaterialExpressionNamedRerouteUsage_64 {class=MaterialExpressionNamedRerouteUsage}
  MaterialExpressionNamedRerouteUsage_65 {class=MaterialExpressionNamedRerouteUsage}
  MaterialExpressionNamedRerouteUsage_66 {class=MaterialExpressionNamedRerouteUsage}
  MaterialExpressionNamedRerouteUsage_67 {class=MaterialExpressionNamedRerouteUsage}
  MaterialExpressionNamedRerouteUsage_68 {class=MaterialExpressionNamedRerouteUsage}
  MaterialExpressionNamedRerouteUsage_69 {class=MaterialExpressionNamedRerouteUsage}
  MaterialExpressionNamedRerouteUsage_7 {class=MaterialExpressionNamedRerouteUsage}
  MaterialExpressionNamedRerouteUsage_70 {class=MaterialExpressionNamedRerouteUsage}
  MaterialExpressionNamedRerouteUsage_71 {class=MaterialExpressionNamedRerouteUsage}
  MaterialExpressionNamedRerouteUsage_72 {class=MaterialExpressionNamedRerouteUsage}
  MaterialExpressionNamedRerouteUsage_73 {class=MaterialExpressionNamedRerouteUsage}
  MaterialExpressionNamedRerouteUsage_74 {class=MaterialExpressionNamedRerouteUsage}
  MaterialExpressionNamedRerouteUsage_75 {class=MaterialExpressionNamedRerouteUsage}
  MaterialExpressionNamedRerouteUsage_76 {class=MaterialExpressionNamedRerouteUsage}
  MaterialExpressionNamedRerouteUsage_77 {class=MaterialExpressionNamedRerouteUsage}
  MaterialExpressionNamedRerouteUsage_79 {class=MaterialExpressionNamedRerouteUsage}
  MaterialExpressionNamedRerouteUsage_8 {class=MaterialExpressionNamedRerouteUsage}
  MaterialExpressionNamedRerouteUsage_80 {class=MaterialExpressionNamedRerouteUsage}
  MaterialExpressionNamedRerouteUsage_81 {class=MaterialExpressionNamedRerouteUsage}
  MaterialExpressionNamedRerouteUsage_82 {class=MaterialExpressionNamedRerouteUsage}
  MaterialExpressionNamedRerouteUsage_83 {class=MaterialExpressionNamedRerouteUsage}
  MaterialExpressionNamedRerouteUsage_84 {class=MaterialExpressionNamedRerouteUsage}
  MaterialExpressionNamedRerouteUsage_85 {class=MaterialExpressionNamedRerouteUsage}
  MaterialExpressionNamedRerouteUsage_86 {class=MaterialExpressionNamedRerouteUsage}
  MaterialExpressionNamedRerouteUsage_87 {class=MaterialExpressionNamedRerouteUsage}
  MaterialExpressionNamedRerouteUsage_88 {class=MaterialExpressionNamedRerouteUsage}
  MaterialExpressionNamedRerouteUsage_89 {class=MaterialExpressionNamedRerouteUsage}
  MaterialExpressionNamedRerouteUsage_9 {class=MaterialExpressionNamedRerouteUsage}
  MaterialExpressionNamedRerouteUsage_90 {class=MaterialExpressionNamedRerouteUsage}
  MaterialExpressionNamedRerouteUsage_92 {class=MaterialExpressionNamedRerouteUsage}
  MaterialExpressionNamedRerouteUsage_93 {class=MaterialExpressionNamedRerouteUsage}
  MaterialExpressionNamedRerouteUsage_94 {class=MaterialExpressionNamedRerouteUsage}
  MaterialExpressionNamedRerouteUsage_95 {class=MaterialExpressionNamedRerouteUsage}
  MaterialExpressionNamedRerouteUsage_96 {class=MaterialExpressionNamedRerouteUsage}
  MaterialExpressionNamedRerouteUsage_97 {class=MaterialExpressionNamedRerouteUsage}
  MaterialExpressionNamedRerouteUsage_98 {class=MaterialExpressionNamedRerouteUsage}
  MaterialExpressionNamedRerouteUsage_99 {class=MaterialExpressionNamedRerouteUsage}
  MaterialExpressionPanner_0 {
   class=MaterialExpressionPanner
   inputs [
    Coordinate {from=MaterialExpressionTextureCoordinate_2}
    Time {from=MaterialExpressionMultiply_40}
    Speed {from=MaterialExpressionConstant2Vector_9}
   ]
  }
  MaterialExpressionReroute_0 {class=MaterialExpressionReroute, inputs [None {from=MaterialExpressionAdd_1}]}
  MaterialExpressionReroute_1 {class=MaterialExpressionReroute, inputs [None {from=MaterialExpressionReroute_2}]}
  MaterialExpressionReroute_10 {class=MaterialExpressionReroute, inputs [None {from=MaterialExpressionReroute_8}]}
  MaterialExpressionReroute_100 {class=MaterialExpressionReroute, inputs [None {from=MaterialExpressionReroute_99}]}
  MaterialExpressionReroute_104 {class=MaterialExpressionReroute, inputs [None {from=MaterialExpressionStaticSwitch_38}]}
  MaterialExpressionReroute_105 {class=MaterialExpressionReroute, inputs [None {from=MaterialExpressionStaticSwitch_39}]}
  MaterialExpressionReroute_106 {class=MaterialExpressionReroute, inputs [None {from=MaterialExpressionReroute_105}]}
  MaterialExpressionReroute_107 {class=MaterialExpressionReroute, inputs [None {from=MaterialExpressionStaticSwitch_40}]}
  MaterialExpressionReroute_108 {class=MaterialExpressionReroute, inputs [None {from=MaterialExpressionStaticSwitch_42}]}
  MaterialExpressionReroute_109 {class=MaterialExpressionReroute, inputs [None {from=MaterialExpressionReroute_108}]}
  MaterialExpressionReroute_11 {class=MaterialExpressionReroute, inputs [None {from=MaterialExpressionNamedRerouteUsage_20}]}
  MaterialExpressionReroute_12 {class=MaterialExpressionReroute, inputs [None {from=MaterialExpressionReroute_11}]}
  MaterialExpressionReroute_13 {class=MaterialExpressionReroute, inputs [None {from=MaterialExpressionNamedRerouteUsage_144}]}
  MaterialExpressionReroute_14 {class=MaterialExpressionReroute, inputs [None {from=MaterialExpressionReroute_15}]}
  MaterialExpressionReroute_15 {class=MaterialExpressionReroute, inputs [None {from=MaterialExpressionReroute_16}]}
  MaterialExpressionReroute_16 {class=MaterialExpressionReroute, inputs [None {from=MaterialExpressionReroute_13}]}
  MaterialExpressionReroute_17 {class=MaterialExpressionReroute, inputs [None {from=MaterialExpressionNamedRerouteUsage_144}]}
  MaterialExpressionReroute_19 {class=MaterialExpressionReroute, inputs [None {from=MaterialExpressionStaticSwitch_11}]}
  MaterialExpressionReroute_2 {class=MaterialExpressionReroute, inputs [None {from=MaterialExpressionReroute_3}]}
  MaterialExpressionReroute_20 {class=MaterialExpressionReroute, inputs [None {from=MaterialExpressionReroute_19}]}
  MaterialExpressionReroute_21 {class=MaterialExpressionReroute, inputs [None {from=MaterialExpressionReroute_20}]}
  MaterialExpressionReroute_22 {class=MaterialExpressionReroute, inputs [None {from=MaterialExpressionStaticSwitch_15}]}
  MaterialExpressionReroute_23 {class=MaterialExpressionReroute, inputs [None {from=MaterialExpressionReroute_22}]}
  MaterialExpressionReroute_24 {class=MaterialExpressionReroute, inputs [None {from=MaterialExpressionReroute_23}]}
  MaterialExpressionReroute_25 {class=MaterialExpressionReroute, inputs [None {from=MaterialExpressionStaticSwitch_10}]}
  MaterialExpressionReroute_26 {class=MaterialExpressionReroute, inputs [None {from=MaterialExpressionReroute_25}]}
  MaterialExpressionReroute_27 {class=MaterialExpressionReroute, inputs [None {from=MaterialExpressionReroute_26}]}
  MaterialExpressionReroute_28 {class=MaterialExpressionReroute, inputs [None {from=MaterialExpressionReroute_26}]}
  MaterialExpressionReroute_29 {class=MaterialExpressionReroute, inputs [None {from=MaterialExpressionReroute_28}]}
  MaterialExpressionReroute_3 {class=MaterialExpressionReroute, inputs [None {from=MaterialExpressionReroute_4}]}
  MaterialExpressionReroute_30 {class=MaterialExpressionReroute, inputs [None {from=MaterialExpressionReroute_20}]}
  MaterialExpressionReroute_31 {class=MaterialExpressionReroute, inputs [None {from=MaterialExpressionReroute_30}]}
  MaterialExpressionReroute_32 {class=MaterialExpressionReroute, inputs [None {from=MaterialExpressionReroute_23}]}
  MaterialExpressionReroute_33 {class=MaterialExpressionReroute, inputs [None {from=MaterialExpressionReroute_32}]}
  MaterialExpressionReroute_34 {class=MaterialExpressionReroute, inputs [None {from=MaterialExpressionStaticSwitch_14}]}
  MaterialExpressionReroute_35 {class=MaterialExpressionReroute, inputs [None {from=MaterialExpressionReroute_34}]}
  MaterialExpressionReroute_36 {class=MaterialExpressionReroute, inputs [None {from=MaterialExpressionReroute_35}]}
  MaterialExpressionReroute_37 {class=MaterialExpressionReroute, inputs [None {from=MaterialExpressionReroute_35}]}
  MaterialExpressionReroute_38 {class=MaterialExpressionReroute, inputs [None {from=MaterialExpressionReroute_37}]}
  MaterialExpressionReroute_39 {class=MaterialExpressionReroute, inputs [None {from=MaterialExpressionStaticSwitch_16}]}
  MaterialExpressionReroute_4 {class=MaterialExpressionReroute, inputs [None {from=MaterialExpressionStaticSwitch_2}]}
  MaterialExpressionReroute_40 {class=MaterialExpressionReroute, inputs [None {from=MaterialExpressionReroute_39}]}
  MaterialExpressionReroute_41 {class=MaterialExpressionReroute, inputs [None {from=MaterialExpressionReroute_40}]}
  MaterialExpressionReroute_42 {class=MaterialExpressionReroute, inputs [None {from=MaterialExpressionReroute_40}]}
  MaterialExpressionReroute_43 {class=MaterialExpressionReroute, inputs [None {from=MaterialExpressionReroute_42}]}
  MaterialExpressionReroute_44 {class=MaterialExpressionReroute, inputs [None {from=MaterialExpressionStaticSwitch_17}]}
  MaterialExpressionReroute_45 {class=MaterialExpressionReroute, inputs [None {from=MaterialExpressionReroute_44}]}
  MaterialExpressionReroute_46 {class=MaterialExpressionReroute, inputs [None {from=MaterialExpressionReroute_45}]}
  MaterialExpressionReroute_47 {class=MaterialExpressionReroute, inputs [None {from=MaterialExpressionReroute_45}]}
  MaterialExpressionReroute_48 {class=MaterialExpressionReroute, inputs [None {from=MaterialExpressionReroute_47}]}
  MaterialExpressionReroute_49 {class=MaterialExpressionReroute, inputs [None {from=MaterialExpressionReroute_50}]}
  MaterialExpressionReroute_5 {class=MaterialExpressionReroute, inputs [None {from=MaterialExpressionReroute_3}]}
  MaterialExpressionReroute_50 {class=MaterialExpressionReroute, inputs [None {from=MaterialExpressionStaticSwitch_18}]}
  MaterialExpressionReroute_51 {class=MaterialExpressionReroute, inputs [None {from=MaterialExpressionReroute_49}]}
  MaterialExpressionReroute_52 {class=MaterialExpressionReroute, inputs [None {from=MaterialExpressionReroute_49}]}
  MaterialExpressionReroute_53 {class=MaterialExpressionReroute, inputs [None {from=MaterialExpressionReroute_52}]}
  MaterialExpressionReroute_54 {class=MaterialExpressionReroute, inputs [None {from=MaterialExpressionStaticSwitch_20}]}
  MaterialExpressionReroute_55 {class=MaterialExpressionReroute, inputs [None {from=MaterialExpressionReroute_54}]}
  MaterialExpressionReroute_56 {class=MaterialExpressionReroute, inputs [None {from=MaterialExpressionReroute_55}]}
  MaterialExpressionReroute_57 {class=MaterialExpressionReroute, inputs [None {from=MaterialExpressionReroute_55}]}
  MaterialExpressionReroute_58 {class=MaterialExpressionReroute, inputs [None {from=MaterialExpressionReroute_57}]}
  MaterialExpressionReroute_59 {class=MaterialExpressionReroute, inputs [None {from=MaterialExpressionStaticSwitch_21}]}
  MaterialExpressionReroute_6 {class=MaterialExpressionReroute, inputs [None {from=MaterialExpressionReroute_7}]}
  MaterialExpressionReroute_60 {class=MaterialExpressionReroute, inputs [None {from=MaterialExpressionReroute_59}]}
  MaterialExpressionReroute_61 {class=MaterialExpressionReroute, inputs [None {from=MaterialExpressionReroute_60}]}
  MaterialExpressionReroute_62 {class=MaterialExpressionReroute, inputs [None {from=MaterialExpressionReroute_60}]}
  MaterialExpressionReroute_63 {class=MaterialExpressionReroute, inputs [None {from=MaterialExpressionReroute_62}]}
  MaterialExpressionReroute_64 {class=MaterialExpressionReroute, inputs [None {from=MaterialExpressionStaticSwitch_22}]}
  MaterialExpressionReroute_65 {class=MaterialExpressionReroute, inputs [None {from=MaterialExpressionReroute_64}]}
  MaterialExpressionReroute_66 {class=MaterialExpressionReroute, inputs [None {from=MaterialExpressionReroute_65}]}
  MaterialExpressionReroute_67 {class=MaterialExpressionReroute, inputs [None {from=MaterialExpressionReroute_65}]}
  MaterialExpressionReroute_68 {class=MaterialExpressionReroute, inputs [None {from=MaterialExpressionReroute_67}]}
  MaterialExpressionReroute_69 {class=MaterialExpressionReroute, inputs [None {from=MaterialExpressionStaticSwitch_22}]}
  MaterialExpressionReroute_7 {class=MaterialExpressionReroute, inputs [None {from=MaterialExpressionStaticSwitch_1}]}
  MaterialExpressionReroute_70 {class=MaterialExpressionReroute, inputs [None {from=MaterialExpressionReroute_69}]}
  MaterialExpressionReroute_71 {class=MaterialExpressionReroute, inputs [None {from=MaterialExpressionReroute_70}]}
  MaterialExpressionReroute_72 {class=MaterialExpressionReroute, inputs [None {from=MaterialExpressionReroute_70}]}
  MaterialExpressionReroute_73 {class=MaterialExpressionReroute, inputs [None {from=MaterialExpressionReroute_72}]}
  MaterialExpressionReroute_76 {class=MaterialExpressionReroute, inputs [None {from=MaterialExpressionReroute_77}]}
  MaterialExpressionReroute_77 {class=MaterialExpressionReroute, inputs [None {from=MaterialExpressionLinearInterpolate_17}]}
  MaterialExpressionReroute_8 {class=MaterialExpressionReroute, inputs [None {from=MaterialExpressionReroute_6}]}
  MaterialExpressionReroute_9 {class=MaterialExpressionReroute, inputs [None {from=MaterialExpressionReroute_6}]}
  MaterialExpressionReroute_94 {class=MaterialExpressionReroute, inputs [None {from=MaterialExpressionStaticSwitch_29}]}
  MaterialExpressionReroute_95 {class=MaterialExpressionReroute, inputs [None {from=MaterialExpressionReroute_76}]}
  MaterialExpressionReroute_96 {class=MaterialExpressionReroute, inputs [None {from=MaterialExpressionStaticSwitch_33}]}
  MaterialExpressionReroute_97 {class=MaterialExpressionReroute, inputs [None {from=MaterialExpressionLinearInterpolate_22}]}
  MaterialExpressionReroute_98 {class=MaterialExpressionReroute, inputs [None {from=MaterialExpressionReroute_97}]}
  MaterialExpressionReroute_99 {class=MaterialExpressionReroute, inputs [None {from=MaterialExpressionStaticSwitch_32}]}
  MaterialExpressionRound_2 {class=MaterialExpressionRound, inputs [None {from=MaterialExpressionClamp_0, output=""}], properties {}}
  MaterialExpressionSaturate_0 {class=MaterialExpressionSaturate, inputs [None {from=MaterialExpressionScalarParameter_7}]}
  MaterialExpressionSaturate_1 {class=MaterialExpressionSaturate, inputs [None {from=MaterialExpressionScalarParameter_6}]}
  MaterialExpressionSaturate_2 {class=MaterialExpressionSaturate, inputs [None {from=MaterialExpressionScalarParameter_17}]}
  MaterialExpressionSaturate_3 {class=MaterialExpressionSaturate, inputs [None {from=MaterialExpressionScalarParameter_14}]}
  MaterialExpressionSaturate_4 {class=MaterialExpressionSaturate, inputs [None {from=MaterialExpressionScalarParameter_1}]}
  MaterialExpressionSaturate_5 {class=MaterialExpressionSaturate, inputs [None {from=MaterialExpressionScalarParameter_3}]}
  MaterialExpressionSaturate_6 {class=MaterialExpressionSaturate, inputs [None {from=MaterialExpressionScalarParameter_2}]}
  MaterialExpressionSaturate_7 {class=MaterialExpressionSaturate, inputs [None {from=MaterialExpressionScalarParameter_71}]}
  MaterialExpressionScalarParameter_0 {
   class=MaterialExpressionScalarParameter
   properties {
    controlType=Enumeration
    enumeration=&/M_UI_CherrySmokeHeroRank_FullBadge_Gradient_Selection
    group=Background
    parameterName=Background - Gradient Shape
    sliderMax=3
    sortPriority=45
   }
  }
  MaterialExpressionScalarParameter_1 {
   class=MaterialExpressionScalarParameter
   properties {defaultValue=0.88, group=Outlines, parameterName="Outer [Thick] - Opacity", sliderMax=1, sortPriority=50}
  }
  MaterialExpressionScalarParameter_10 {
   class=MaterialExpressionScalarParameter
   properties {defaultValue=0.028, group=Rank Streaks, parameterName=Streaks - Step Distance, sliderMax=0.028, sortPriority=5}
  }
  MaterialExpressionScalarParameter_11 {
   class=MaterialExpressionScalarParameter
   properties {defaultValue=0.5, group=Rank Streaks, parameterName=Streaks - Angle, sliderMax=1, sortPriority=20}
  }
  MaterialExpressionScalarParameter_12 {
   class=MaterialExpressionScalarParameter
   properties {defaultValue=1, group=Rank Streaks, parameterName=Streaks - Falloff Power, sliderMax=4, sortPriority=70}
  }
  MaterialExpressionScalarParameter_13 {
   class=MaterialExpressionScalarParameter
   properties {group=Rank Streaks, parameterName=Streaks - Rotation Per Step, sliderMax=0.02, sortPriority=30}
  }
  MaterialExpressionScalarParameter_14 {
   class=MaterialExpressionScalarParameter
   properties {defaultValue=0.5, group=Rank Streaks, parameterName=Streaks - Opacity, sliderMax=1, sortPriority=60}
  }
  MaterialExpressionScalarParameter_15 {
   class=MaterialExpressionScalarParameter
   properties {defaultValue=1.2, group=Rank Streaks, parameterName=Streaks - Scale End, sliderMax=10, sortPriority=90}
  }
  MaterialExpressionScalarParameter_16 {
   class=MaterialExpressionScalarParameter
   properties {defaultValue=1, group=Rank Glow, parameterName=Radial Glow - Gradient End, sortPriority=30}
  }
  MaterialExpressionScalarParameter_17 {class=MaterialExpressionScalarParameter, properties {defaultValue=1, group=Rank Glow, parameterName=Radial Glow - Opacity, sortPriority=50}}
  MaterialExpressionScalarParameter_19 {
   class=MaterialExpressionScalarParameter
   properties {defaultValue=0.5, group=Rank Glow, parameterName=Radial Glow Position - X, sortPriority=70}
  }
  MaterialExpressionScalarParameter_2 {
   class=MaterialExpressionScalarParameter
   properties {defaultValue=0.88, group=Outlines, parameterName="Inner [Thick] - Opacity", sliderMax=1, sortPriority=10}
  }
  MaterialExpressionScalarParameter_20 {
   class=MaterialExpressionScalarParameter
   properties {defaultValue=0.5, group=Rank Glow, parameterName=Radial Glow Position - Y, sortPriority=80}
  }
  MaterialExpressionScalarParameter_21 {
   class=MaterialExpressionScalarParameter
   properties {defaultValue=0.05, group=Rank Glow, parameterName=Radial Glow - Distance, sortPriority=90}
  }
  MaterialExpressionScalarParameter_22 {class=MaterialExpressionScalarParameter, properties {group=Rank Glow, parameterName=Radial Glow - Angle, sortPriority=100}}
  MaterialExpressionScalarParameter_23 {class=MaterialExpressionScalarParameter, properties {group=Rank Glow, parameterName=Radial Glow - Center Point X, sortPriority=110}}
  MaterialExpressionScalarParameter_24 {class=MaterialExpressionScalarParameter, properties {group=Rank Glow, parameterName=Radial Glow - Center Point Y, sortPriority=120}}
  MaterialExpressionScalarParameter_27 {class=MaterialExpressionScalarParameter, properties {group=Glitch, parameterName=Glitch - Strength, sortPriority=20}}
  MaterialExpressionScalarParameter_28 {class=MaterialExpressionScalarParameter, properties {group=Glitch, parameterName=Glitch - Speed Multiplier, sortPriority=30}}
  MaterialExpressionScalarParameter_3 {
   class=MaterialExpressionScalarParameter
   properties {defaultValue=0.88, group=Outlines, parameterName="Outer [Thin] - Opacity", sliderMax=1, sortPriority=30}
  }
  MaterialExpressionScalarParameter_4 {
   class=MaterialExpressionScalarParameter
   properties {defaultValue=0.635, group=Background, parameterName=Background Gradient - Rotation, sortPriority=80}
  }
  MaterialExpressionScalarParameter_5 {
   class=MaterialExpressionScalarParameter
   properties {defaultValue=0.125, group=Background, parameterName=Background Gradient - Scale, sliderMax=1, sortPriority=90}
  }
  MaterialExpressionScalarParameter_6 {
   class=MaterialExpressionScalarParameter
   properties {defaultValue=0.88, group=Background, parameterName=Outer Background - Opacity, sliderMax=1, sortPriority=30}
  }
  MaterialExpressionScalarParameter_7 {
   class=MaterialExpressionScalarParameter
   properties {defaultValue=0.88, group=Background, parameterName=Inner Background - Opacity, sliderMax=1, sortPriority=10}
  }
  MaterialExpressionScalarParameter_71 {class=MaterialExpressionScalarParameter, properties {group=Rank, parameterName=Rank - Opacity, sliderMax=1, sortPriority=20}}
  MaterialExpressionScalarParameter_72 {class=MaterialExpressionScalarParameter, properties {group=Rank, parameterName=Rank Offset - X, sortPriority=30}}
  MaterialExpressionScalarParameter_73 {class=MaterialExpressionScalarParameter, properties {group=Rank, parameterName=Rank Offset - Y, sortPriority=40}}
  MaterialExpressionScalarParameter_8 {class=MaterialExpressionScalarParameter, properties {group=Rank Glow, parameterName=Radial Glow - Gradient Start, sortPriority=20}}
  MaterialExpressionScalarParameter_9 {
   class=MaterialExpressionScalarParameter
   properties {defaultValue=1, group=Rank Streaks, parameterName=Streaks - Scale Start, sliderMax=10, sortPriority=80}
  }
  MaterialExpressionSine_0 {class=MaterialExpressionSine, inputs [None {from=MaterialExpressionNamedRerouteUsage_9}], properties {desc=Y}}
  MaterialExpressionSine_2 {class=MaterialExpressionSine, inputs [None {from=MaterialExpressionMaterialFunctionCall_8, output=G}]}
  MaterialExpressionStaticBoolParameter_0 {class=MaterialExpressionStaticBoolParameter, inputs [], properties {group=Rank Streaks, parameterName=Streaks - Enabled, sortPriority=0}}
  MaterialExpressionStaticBoolParameter_1 {
   class=MaterialExpressionStaticBoolParameter
   inputs []
   properties {defaultValue=true, group=Background, parameterName=Inner Background - Enabled, sortPriority=5}
  }
  MaterialExpressionStaticBoolParameter_11 {
   class=MaterialExpressionStaticBoolParameter
   inputs []
   properties {defaultValue=true, group=Rank Glow, parameterName=Radial Glow Position - Switch to Polar, sortPriority=60}
  }
  MaterialExpressionStaticBoolParameter_12 {
   class=MaterialExpressionStaticBoolParameter
   inputs []
   properties {defaultValue=true, group=Rank Glow, parameterName=Radial Glow - Clamp to Badge, sortPriority=10}
  }
  MaterialExpressionStaticBoolParameter_13 {class=MaterialExpressionStaticBoolParameter, inputs [], properties {group=Glitch, parameterName=Glitch - Enabled, sortPriority=0}}
  MaterialExpressionStaticBoolParameter_14 {class=MaterialExpressionStaticBoolParameter, inputs [], properties {group=Glitch, parameterName=Glitch Rank Glyph - Enabled, sortPriority=5}}
  MaterialExpressionStaticBoolParameter_2 {
   class=MaterialExpressionStaticBoolParameter
   inputs []
   properties {defaultValue=true, group=Background, parameterName=Outer Background - Enabled, sortPriority=20}
  }
  MaterialExpressionStaticBoolParameter_3 {
   class=MaterialExpressionStaticBoolParameter
   inputs []
   properties {defaultValue=true, group=Background, parameterName=Background - Gradient Enabled, sortPriority=40}
  }
  MaterialExpressionStaticBoolParameter_4 {
   class=MaterialExpressionStaticBoolParameter
   inputs []
   properties {defaultValue=true, group=Outlines, parameterName="Inner [Thick] - Enabled", sortPriority=0}
  }
  MaterialExpressionStaticBoolParameter_5 {
   class=MaterialExpressionStaticBoolParameter
   inputs []
   properties {defaultValue=true, group=Outlines, parameterName="Outer [Thin] - Enabled", sortPriority=20}
  }
  MaterialExpressionStaticBoolParameter_6 {
   class=MaterialExpressionStaticBoolParameter
   inputs []
   properties {defaultValue=true, group=Outlines, parameterName="Outer [Thick] - Enabled", sortPriority=40}
  }
  MaterialExpressionStaticBoolParameter_7 {
   class=MaterialExpressionStaticBoolParameter
   inputs []
   properties {defaultValue=true, group=Rank, parameterName=Rank - Glyph Enabled, sortPriority=0}
  }
  MaterialExpressionStaticBoolParameter_8 {
   class=MaterialExpressionStaticBoolParameter
   inputs []
   properties {group=Background, parameterName=Background - No Inner/Outer Overlap, sortPriority=0}
  }
  MaterialExpressionStaticBoolParameter_9 {class=MaterialExpressionStaticBoolParameter, inputs [], properties {group=Rank Glow, parameterName=Radial Glow - Enabled, sortPriority=0}}
  MaterialExpressionStaticSwitch_0 {
   class=MaterialExpressionStaticSwitch
   inputs [
    True {from=MaterialExpressionSubtract_0, output=""}
    False {from=MaterialExpressionReroute_0, output=""}
    Value {from=MaterialExpressionNamedRerouteUsage_22, output=""}
   ]
   properties {}
  }
  MaterialExpressionStaticSwitch_1 {
   class=MaterialExpressionStaticSwitch
   inputs [
    True {from=MaterialExpressionLinearInterpolate_41, output=""}
    False {from=MaterialExpressionReroute_73, output=""}
    Value {from=MaterialExpressionNamedRerouteUsage_0, output=""}
   ]
   properties {}
  }
  MaterialExpressionStaticSwitch_10 {
   class=MaterialExpressionStaticSwitch
   inputs [
    True {from=MaterialExpressionLinearInterpolate_29, output=""}
    False {from=MaterialExpressionReroute_33, output=""}
    Value {from=MaterialExpressionNamedRerouteUsage_44, output=""}
   ]
   properties {}
  }
  MaterialExpressionStaticSwitch_11 {
   class=MaterialExpressionStaticSwitch
   inputs [
    True {from=MaterialExpressionSwitch_0, output=""}
    False {from=MaterialExpressionNamedRerouteUsage_26, output=""}
    Value {from=MaterialExpressionNamedRerouteUsage_24, output=""}
   ]
   properties {}
  }
  MaterialExpressionStaticSwitch_12 {
   class=MaterialExpressionStaticSwitch
   inputs [
    True {from=MaterialExpressionAdd_31, output=""}
    False {from=MaterialExpressionReroute_17, output=""}
    Value {from=MaterialExpressionStaticBoolParameter_14, output=""}
   ]
   properties {}
  }
  MaterialExpressionStaticSwitch_14 {
   class=MaterialExpressionStaticSwitch
   inputs [
    True {from=MaterialExpressionLinearInterpolate_30, output=""}
    False {from=MaterialExpressionReroute_29, output=""}
    Value {from=MaterialExpressionNamedRerouteUsage_32, output=""}
   ]
   properties {}
  }
  MaterialExpressionStaticSwitch_15 {
   class=MaterialExpressionStaticSwitch
   inputs [
    True {from=MaterialExpressionLinearInterpolate_3, output=""}
    False {from=MaterialExpressionReroute_31, output=""}
    Value {from=MaterialExpressionNamedRerouteUsage_33, output=""}
   ]
   properties {}
  }
  MaterialExpressionStaticSwitch_16 {
   class=MaterialExpressionStaticSwitch
   inputs [
    True {from=MaterialExpressionLinearInterpolate_8, output=""}
    False {from=MaterialExpressionReroute_38, output=""}
    Value {from=MaterialExpressionNamedRerouteUsage_34, output=""}
   ]
   properties {}
  }
  MaterialExpressionStaticSwitch_17 {
   class=MaterialExpressionStaticSwitch
   inputs [
    True {from=MaterialExpressionLinearInterpolate_31, output=""}
    False {from=MaterialExpressionConstant_17, output=""}
    Value {from=MaterialExpressionNamedRerouteUsage_35, output=""}
   ]
   properties {}
  }
  MaterialExpressionStaticSwitch_18 {
   class=MaterialExpressionStaticSwitch
   inputs [
    True {from=MaterialExpressionLinearInterpolate_14, output=""}
    False {from=MaterialExpressionReroute_48, output=""}
    Value {from=MaterialExpressionNamedRerouteUsage_42, output=""}
   ]
   properties {}
  }
  MaterialExpressionStaticSwitch_2 {
   class=MaterialExpressionStaticSwitch
   inputs [
    True {from=MaterialExpressionLinearInterpolate_40, output=""}
    False {from=MaterialExpressionReroute_43, output=""}
    Value {from=MaterialExpressionNamedRerouteUsage_12, output=""}
   ]
   properties {}
  }
  MaterialExpressionStaticSwitch_20 {
   class=MaterialExpressionStaticSwitch
   inputs [
    True {from=MaterialExpressionLinearInterpolate_13, output=""}
    False {from=MaterialExpressionReroute_53, output=""}
    Value {from=MaterialExpressionNamedRerouteUsage_43, output=""}
   ]
   properties {}
  }
  MaterialExpressionStaticSwitch_21 {
   class=MaterialExpressionStaticSwitch
   inputs [
    True {from=MaterialExpressionLinearInterpolate_11, output=""}
    False {from=MaterialExpressionReroute_58, output=""}
    Value {from=MaterialExpressionNamedRerouteUsage_48, output=""}
   ]
   properties {}
  }
  MaterialExpressionStaticSwitch_22 {
   class=MaterialExpressionStaticSwitch
   inputs [
    True {from=MaterialExpressionLinearInterpolate_12, output=""}
    False {from=MaterialExpressionReroute_63, output=""}
    Value {from=MaterialExpressionNamedRerouteUsage_51, output=""}
   ]
   properties {}
  }
  MaterialExpressionStaticSwitch_23 {
   class=MaterialExpressionStaticSwitch
   inputs [
    True {from=MaterialExpressionLinearInterpolate_15, output=""}
    False {from=MaterialExpressionReroute_68, output=""}
    Value {from=MaterialExpressionNamedRerouteUsage_56, output=""}
   ]
   properties {}
  }
  MaterialExpressionStaticSwitch_26 {
   class=MaterialExpressionStaticSwitch
   inputs [
    True {from=MaterialExpressionSubtract_2, output=""}
    False {from=MaterialExpressionNamedRerouteDeclaration_17, output=""}
    Value {from=MaterialExpressionNamedRerouteUsage_108, output=""}
   ]
   properties {}
  }
  MaterialExpressionStaticSwitch_27 {
   class=MaterialExpressionStaticSwitch
   inputs [
    True {from=MaterialExpressionNamedRerouteUsage_84, output=""}
    False {from=MaterialExpressionNamedRerouteUsage_80, output=""}
    Value {from=MaterialExpressionNamedRerouteUsage_74, output=""}
   ]
   properties {}
  }
  MaterialExpressionStaticSwitch_28 {
   class=MaterialExpressionStaticSwitch
   inputs [
    True {from=MaterialExpressionStaticSwitch_27, output=""}
    False {from=MaterialExpressionReroute_94, output=""}
    Value {from=MaterialExpressionNamedRerouteUsage_75, output=""}
   ]
   properties {}
  }
  MaterialExpressionStaticSwitch_29 {
   class=MaterialExpressionStaticSwitch
   inputs [
    True {from=MaterialExpressionNamedRerouteUsage_85, output=""}
    False {from=MaterialExpressionConstant_20, output=""}
    Value {from=MaterialExpressionNamedRerouteUsage_79, output=""}
   ]
   properties {}
  }
  MaterialExpressionStaticSwitch_3 {
   class=MaterialExpressionStaticSwitch
   inputs [
    True {from=MaterialExpressionLinearInterpolate_1, output=""}
    False {from=MaterialExpressionReroute_1, output=""}
    Value {from=MaterialExpressionNamedRerouteUsage_111, output=""}
   ]
   properties {}
  }
  MaterialExpressionStaticSwitch_31 {
   class=MaterialExpressionStaticSwitch
   inputs [
    True {from=MaterialExpressionNamedRerouteUsage_100, output=""}
    False {from=MaterialExpressionNamedRerouteUsage_106, output=""}
    Value {from=MaterialExpressionNamedRerouteUsage_92, output=""}
   ]
   properties {}
  }
  MaterialExpressionStaticSwitch_32 {
   class=MaterialExpressionStaticSwitch
   inputs [
    True {from=MaterialExpressionStaticSwitch_31, output=""}
    False {from=MaterialExpressionReroute_96, output=""}
    Value {from=MaterialExpressionNamedRerouteUsage_93, output=""}
   ]
   properties {}
  }
  MaterialExpressionStaticSwitch_33 {
   class=MaterialExpressionStaticSwitch
   inputs [
    True {from=MaterialExpressionNamedRerouteUsage_102, output=""}
    False {from=MaterialExpressionConstant_22, output=""}
    Value {from=MaterialExpressionNamedRerouteUsage_99, output=""}
   ]
   properties {}
  }
  MaterialExpressionStaticSwitch_37 {
   class=MaterialExpressionStaticSwitch
   inputs [
    True {from=MaterialExpressionNamedRerouteUsage_129, output=""}
    False {from=MaterialExpressionNamedRerouteUsage_131, output=""}
    Value {from=MaterialExpressionNamedRerouteUsage_126, output=""}
   ]
   properties {}
  }
  MaterialExpressionStaticSwitch_38 {
   class=MaterialExpressionStaticSwitch
   inputs [
    True {from=MaterialExpressionNamedRerouteUsage_130, output=""}
    False {from=MaterialExpressionConstant_24, output=""}
    Value {from=MaterialExpressionNamedRerouteUsage_127, output=""}
   ]
   properties {}
  }
  MaterialExpressionStaticSwitch_39 {
   class=MaterialExpressionStaticSwitch
   inputs [
    True {from=MaterialExpressionStaticSwitch_37, output=""}
    False {from=MaterialExpressionReroute_104, output=""}
    Value {from=MaterialExpressionNamedRerouteUsage_128, output=""}
   ]
   properties {}
  }
  MaterialExpressionStaticSwitch_40 {
   class=MaterialExpressionStaticSwitch
   inputs [
    True {from=MaterialExpressionNamedRerouteUsage_137, output=""}
    False {from=MaterialExpressionConstant_25, output=""}
    Value {from=MaterialExpressionNamedRerouteUsage_132, output=""}
   ]
   properties {}
  }
  MaterialExpressionStaticSwitch_41 {
   class=MaterialExpressionStaticSwitch
   inputs [
    True {from=MaterialExpressionNamedRerouteUsage_135, output=""}
    False {from=MaterialExpressionNamedRerouteUsage_136, output=""}
    Value {from=MaterialExpressionNamedRerouteUsage_133, output=""}
   ]
   properties {}
  }
  MaterialExpressionStaticSwitch_42 {
   class=MaterialExpressionStaticSwitch
   inputs [
    True {from=MaterialExpressionStaticSwitch_41, output=""}
    False {from=MaterialExpressionReroute_107, output=""}
    Value {from=MaterialExpressionNamedRerouteUsage_134, output=""}
   ]
   properties {}
  }
  MaterialExpressionStaticSwitch_6 {
   class=MaterialExpressionStaticSwitch
   inputs [
    True {from=MaterialExpressionLinearInterpolate_2, output=""}
    False {from=MaterialExpressionReroute_10, output=""}
    Value {from=MaterialExpressionNamedRerouteUsage_113, output=""}
   ]
   properties {}
  }
  MaterialExpressionStaticSwitch_7 {
   class=MaterialExpressionStaticSwitch
   inputs [
    True {from=MaterialExpressionMultiply_13, output=""}
    False {from=MaterialExpressionMultiply_5, output=""}
    Value {from=MaterialExpressionNamedRerouteUsage_114, output=""}
   ]
   properties {}
  }
  MaterialExpressionStaticSwitch_8 {
   class=MaterialExpressionStaticSwitch
   inputs [
    True {from=MaterialExpressionMultiply_16, output=""}
    False {from=MaterialExpressionMultiply_15, output=""}
    Value {from=MaterialExpressionNamedRerouteUsage_138, output=""}
   ]
   properties {}
  }
  MaterialExpressionStaticSwitch_9 {
   class=MaterialExpressionStaticSwitch
   inputs [
    True {from=MaterialExpressionAdd_31, output=""}
    False {from=MaterialExpressionReroute_14, output=""}
    Value {from=MaterialExpressionStaticBoolParameter_13, output=""}
   ]
   properties {}
  }
  MaterialExpressionSubtract_0 {
   class=MaterialExpressionSubtract
   inputs [
    A {from=MaterialExpressionReroute_12}
    B {from=MaterialExpressionAdd_0}
   ]
  }
  MaterialExpressionSubtract_1 {
   class=MaterialExpressionSubtract
   inputs [
    A {from=MaterialExpressionComponentMask_8}
    B {from=""}
   ]
   properties {constB=0.5}
  }
  MaterialExpressionSubtract_2 {
   class=MaterialExpressionSubtract
   inputs [
    A {from=MaterialExpressionNamedRerouteDeclaration_17}
    B {from=MaterialExpressionNamedRerouteUsage_109}
   ]
  }
  MaterialExpressionSwitch_0 {
   class=MaterialExpressionSwitch
   inputs [
    SwitchValue {from=MaterialExpressionNamedRerouteUsage_39, output=""}
    Default {from="", output=""}
    Whole {from=MaterialExpressionNamedRerouteUsage_118, output=""}
    Outer {from=MaterialExpressionNamedRerouteUsage_119, output=""}
    Inner {from=MaterialExpressionNamedRerouteUsage_120, output=""}
   ]
   properties {inputs [
     {
      inputName=Whole
      input {
       expression=&self:MaterialExpressionNamedRerouteUsage_118
       inputName=None
       mask=0
       maskA=0
       maskB=0
       maskG=0
       maskR=0
       outputIndex=0
      }
     }
     {
      inputName=Outer
      input {
       expression=&self:MaterialExpressionNamedRerouteUsage_119
       inputName=None
       mask=0
       maskA=0
       maskB=0
       maskG=0
       maskR=0
       outputIndex=0
      }
     }
     {
      inputName=Inner
      input {
       expression=&self:MaterialExpressionNamedRerouteUsage_120
       inputName=None
       mask=0
       maskA=0
       maskB=0
       maskG=0
       maskR=0
       outputIndex=0
      }
     }
    ]}
  }
  MaterialExpressionTextureCoordinate_1 {class=MaterialExpressionTextureCoordinate}
  MaterialExpressionTextureCoordinate_2 {class=MaterialExpressionTextureCoordinate, properties {uTiling=4, vTiling=2}}
  MaterialExpressionTextureObjectParameter_0 {
   class=MaterialExpressionTextureObjectParameter
   inputs [
    UVs {from="", output=""}
    Apply View MipBias {from="", output=""}
   ]
   properties {group=Rank, parameterName=Rank - Glyph, sortPriority=0, texture=/Game/UI/Foundation/Textures/Icons/ItemTypes/T-Icon-Blank}
  }
  MaterialExpressionTextureSample_1 {
   class=MaterialExpressionTextureSample
   inputs [
    UVs {from=MaterialExpressionNamedRerouteUsage_60}
    Tex {from=MaterialExpressionNamedRerouteDeclaration_51}
    Apply View MipBias {from=""}
   ]
   properties {samplerType=SAMPLERTYPE_Grayscale, texture=/CherrySmokeHeroRankingUI/Textures/T_UI_CherrySmokeHeroRank_Badge_RankA}
  }
  MaterialExpressionTextureSample_5 {
   class=MaterialExpressionTextureSample
   inputs [
    UVs {from=MaterialExpressionPanner_0}
    Tex {from=""}
    Apply View MipBias {from=""}
   ]
   properties {samplerType=SAMPLERTYPE_LinearGrayscale, texture=/CherrySmokeHeroRankingUI/Textures/T_UI_HeroRanking_Noise}
  }
  MaterialExpressionTime_0 {class=MaterialExpressionTime}
  MaterialExpressionVectorParameter_0 {
   class=MaterialExpressionVectorParameter
   properties {group=Background, parameterName="Background Gradient - Color [Start]", sortPriority=60, defaultValue=<rgba 0.0144 0.013 0.013 1>}
  }
  MaterialExpressionVectorParameter_1 {
   class=MaterialExpressionVectorParameter
   properties {group=Background, parameterName="Background Gradient - Color [End]", sortPriority=70, defaultValue=<rgba 0.8714 0.0999 0.0176 1>}
  }
  MaterialExpressionVectorParameter_10 {
   class=MaterialExpressionVectorParameter
   properties {group=Rank Streaks, parameterName=Streaks - Color End, sortPriority=50, defaultValue=<rgba 0.9304 0.1005 0 1>}
  }
  MaterialExpressionVectorParameter_2 {
   class=MaterialExpressionVectorParameter
   properties {group=Background, parameterName=Outer Background - Color, sortPriority=25, defaultValue=<rgba 0.7913 0 0 1>}
  }
  MaterialExpressionVectorParameter_3 {
   class=MaterialExpressionVectorParameter
   properties {group=Outlines, parameterName="Inner [Thick] - Color", sortPriority=5, defaultValue=<rgba 0.7835 0.147 0.0296 1>}
  }
  MaterialExpressionVectorParameter_4 {
   class=MaterialExpressionVectorParameter
   properties {group=Outlines, parameterName="Outer [Thick] - Color", sortPriority=45, defaultValue=<rgba 0.7835 0.147 0.0296 1>}
  }
  MaterialExpressionVectorParameter_5 {
   class=MaterialExpressionVectorParameter
   properties {group=Outlines, parameterName="Outer [Thin] - Color", sortPriority=25, defaultValue=<rgba 1 1 1 1>}
  }
  MaterialExpressionVectorParameter_6 {
   class=MaterialExpressionVectorParameter
   properties {group=Rank Glow, parameterName=Radial Glow - Color, sortPriority=40, defaultValue=<rgba 0 0.2783 0 1>}
  }
  MaterialExpressionVectorParameter_7 {
   class=MaterialExpressionVectorParameter
   properties {group=Rank, parameterName=Rank - Color, sortPriority=10, defaultValue=<rgba 0.92 0.92 0.92 1>}
  }
  MaterialExpressionVectorParameter_8 {
   class=MaterialExpressionVectorParameter
   properties {group=Background, parameterName=Inner Background - Color, sortPriority=7, defaultValue=<rgba 0 0.2783 0 1>}
  }
  MaterialExpressionVectorParameter_9 {
   class=MaterialExpressionVectorParameter
   properties {group=Rank Streaks, parameterName=Streaks - Color Start, sortPriority=40, defaultValue=<rgba 0.9375 0.5405 0.1074 1>}
  }
 }
 properties {bAutomaticallySetUsageInEditor=false, blendMode=BLEND_Translucent, materialDomain=MD_UI, usedShadingModels=Default Lit}
}
