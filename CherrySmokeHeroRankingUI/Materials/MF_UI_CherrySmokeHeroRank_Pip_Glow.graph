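# UI material function that draws a glowing pip for the CherrySmoke hero ranking UI, outputting Color and Opacity.
# A rounded-box SDF forms the dot and its inner and outer glows, while a circle SDF drives front and back radial glows through anti-aliasing thresholds.
# Each layer has a static-switch toggle and opacity input, glows blend inner to outer colors via SmoothStep, and chained lerps composite the layers.
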MaterialFunction /CherrySmokeHeroRankingUI/Materials/MF_UI_CherrySmokeHeroRank_Pip_Glow
graph {
 properties {previewBlendMode=BLEND_AlphaComposite, previewMaterialDomain=MD_UI}
 nodes [
  FunctionOutput(None=NamedRerouteUsage()) {outputName=Color, sortPriority=10}
  FunctionOutput(None=NamedRerouteUsage()) {outputName=Opacity, sortPriority=20}
  LinearInterpolate()
  n1 = FunctionInput(Preview=ScalarParameter() {defaultValue=1, group=Dot - Both Glows, parameterName=Back Radial Glow - Radius, sortPriority=20}) {bUsePreviewValueAsDefault=true, inputName="┌\tRadial Glow (Back) Radius", inputType=FunctionInput_Scalar, sortPriority=230}
  n2 = FunctionInput(Preview=MaterialFunctionCall() {materialFunction=&GetUserInterfaceUV}.Normalized UV) {bUsePreviewValueAsDefault=true, inputName=UVs, inputType=FunctionInput_Vector2, sortPriority=1}
  n3 = MaterialFunctionCall!(UVs=@n2, Radius=Constant()) {materialFunction=/FortUILibrary/MaterialFunctions/Shapes/SDFs/MF_SDF_Circle}
  n4 = FunctionInput(Preview=ScalarParameter() {defaultValue=0.4, group=Radial Glows, parameterName=Radial Glows - Falloff, sortPriority=10}) {
   bUsePreviewValueAsDefault=true
   description="1.0 = Normal Linear, >1.0 = Sharper Edge, <1.0 = Softer"
   inputName=Radial Glow Falloff
   inputType=FunctionInput_Scalar
   sortPriority=265
  }
  n5 = Power(Base=Saturate(None=Divide(A=@n3.SDF, B=@n1)), Exp=@n4)
  n6 = MaterialFunctionCall(DistanceField=@n5, Threshold=Multiply(A=@n1) {constB=0.25}, Smoothing=@n1) {materialFunction=&MF_UI_AntiAliasing}
  NamedRerouteDeclaration(None=Ceil(None=NamedRerouteDeclaration(None=@n6.Inner) {name=I_Glow_Back_Mask, nodeColor=<rgba 0 0 0 1>})) {name=I_Glow_Back_Range, nodeColor=<rgba 0 0 0 1>}
  n7 = FunctionInput(Preview=StaticBoolParameter() {defaultValue=true, group=Dot, parameterName=Dot, sortPriority=10}) {bUsePreviewValueAsDefault=true, inputName=Enable Dot, inputType=FunctionInput_StaticBool, sortPriority=10, previewValue=<1 0 0 1>}
  NamedRerouteDeclaration(None=@n7) {name=I_Enable_Dot, nodeColor=<rgba 0 0 0 1>}
  n8 = FunctionInput(Preview=StaticBoolParameter() {group=Dot - Inner Glow, parameterName=Dot - Inner Glow, sortPriority=10}) {
   bUsePreviewValueAsDefault=true
   inputName=Enable Dot - Inner Glow
   inputType=FunctionInput_StaticBool
   sortPriority=70
   previewValue=<1 0 0 1>
  }
  NamedRerouteDeclaration(None=@n8) {name=I_Enable_Dot_Glow_Inner, nodeColor=<rgba 0 0 0 1>}
  n9 = FunctionInput(Preview=StaticBoolParameter() {defaultValue=true, group=Dot - Outer Glow, parameterName=Dot - Outer Glow, sortPriority=10}) {
   bUsePreviewValueAsDefault=true
   inputName=Enable Dot - Outer Glow
   inputType=FunctionInput_StaticBool
   sortPriority=120
   previewValue=<1 0 0 1>
  }
  NamedRerouteDeclaration(None=@n9) {name=I_Enable_Dot_Glow_Outer, nodeColor=<rgba 0 0 0 1>}
  n10 = FunctionInput(Preview=StaticBoolParameter() {defaultValue=true, group=Radial Glow - Front, parameterName=Front Radial Glow, sortPriority=10}) {
   bUsePreviewValueAsDefault=true
   inputName=Enable Front Radial Glow
   inputType=FunctionInput_StaticBool
   sortPriority=170
   previewValue=<1 0 0 1>
  }
  NamedRerouteDeclaration(None=@n10) {name=I_Enable_Radial_Glow_Front, nodeColor=<rgba 0 0 0 1>}
  n11 = FunctionInput(Preview=StaticBoolParameter() {defaultValue=true, group=Radial Glow - Back, parameterName=Back Radial Glow, sortPriority=10}) {
   bUsePreviewValueAsDefault=true
   inputName=Enable Back Radial Glow
   inputType=FunctionInput_StaticBool
   sortPriority=220
   previewValue=<1 0 0 1>
  }
  NamedRerouteDeclaration(None=@n11) {name=I_Enable_Radial_Glow_Back, nodeColor=<rgba 0 0 0 1>}
  NamedRerouteDeclaration(None=FunctionInput(Preview=ScalarParameter() {defaultValue=0.9, group=Dot, parameterName=Dot - Opacity, sortPriority=80}) {bUsePreviewValueAsDefault=true, inputName="└ Dot Opacity", inputType=FunctionInput_Scalar, sortPriority=60}) {name=I_Opacity_Dot, nodeColor=<rgba 0 0 0 1>}
  NamedRerouteDeclaration(None=FunctionInput(Preview=ScalarParameter() {defaultValue=0.5, group=Dot - Inner Glow, parameterName=Dot Inner Glow - Opacity, sortPriority=50}) {bUsePreviewValueAsDefault=true, inputName="└ Dot - Inner Glow Opacity", inputType=FunctionInput_Scalar, sortPriority=110}) {name=I_Opacity_Dot_Glow_Inner, nodeColor=<rgba 0 0 0 1>}
  NamedRerouteDeclaration(None=FunctionInput(Preview=ScalarParameter() {defaultValue=1, group=Dot - Outer Glow, parameterName=Dot Outer Glow - Opacity, sortPriority=50}) {bUsePreviewValueAsDefault=true, inputName="└ Dot - Outer Glow Opacity", inputType=FunctionInput_Scalar, sortPriority=160}) {name=I_Opacity_Dot_Glow_Outer, nodeColor=<rgba 0 0 0 1>}
  NamedRerouteDeclaration(None=FunctionInput(Preview=ScalarParameter() {defaultValue=0.4, group=Radial Glow - Front, parameterName=Front Radial Glow - Opacity, sortPriority=50}) {bUsePreviewValueAsDefault=true, inputName="└ Front Radial Glow Opacity", inputType=FunctionInput_Scalar, sortPriority=210}) {name=I_Opacity_Radial_Glow_Front, nodeColor=<rgba 0 0 0 1>}
  NamedRerouteDeclaration(None=FunctionInput(Preview=ScalarParameter() {defaultValue=0.6, group=Radial Glow - Back, parameterName=Back Radial Glow - Opacity, sortPriority=50}) {bUsePreviewValueAsDefault=true, inputName="└ Back Radial Glow Opacity", inputType=FunctionInput_Scalar, sortPriority=260}) {name=I_Opacity_Radial_Glow_Back, nodeColor=<rgba 0 0 0 1>}
  n12 = VectorParameter() {group=Radial Glow - Back, parameterName=Back Radial Glow - Outer Color, sortPriority=40, defaultValue=<rgba 0.397 0.1911 0.0161 1>}
  n13 = VectorParameter() {group=Radial Glow - Back, parameterName=Back Radial Glow - Inner Color, sortPriority=30, defaultValue=<rgba 0.8 0.4073 0.0712 1>}
  n14 = NamedRerouteUsage()
  n15 = Multiply(
   A=LinearInterpolate(
    A=FunctionInput(Preview=@n12.RGB) {bUsePreviewValueAsDefault=true, inputName="│ Radial Glow Back - Outer Color", sortPriority=250}
    B=FunctionInput(Preview=@n13.RGB) {bUsePreviewValueAsDefault=true, inputName="│ Radial Glow Back - Inner Color", sortPriority=240}
    Alpha=SmoothStep(Value=@n14)
   )
   B=NamedRerouteUsage()
  )
  n16 = NamedRerouteUsage()
  n17 = VectorParameter() {group=Radial Glow - Front, parameterName=Front Radial Glow - Outer Color, sortPriority=40, defaultValue=<rgba 0.5972 0.2874 0.0242 1>}
  n18 = VectorParameter() {group=Radial Glow - Front, parameterName=Front Radial Glow - Inner Color, sortPriority=30, defaultValue=<rgba 1 0.5089 0.0887 1>}
  n19 = LinearInterpolate(
   A=FunctionInput(Preview=@n17.RGB) {bUsePreviewValueAsDefault=true, inputName="│ Radial Glow Front - Outer Color", sortPriority=200}
   B=FunctionInput(Preview=@n18.RGB) {bUsePreviewValueAsDefault=true, inputName="│ Radial Glow Front - Inner Color", sortPriority=190}
   Alpha=SmoothStep(Value=NamedRerouteUsage())
  )
  n20 = NamedRerouteUsage()
  n21 = VectorParameter() {group=Dot - Outer Glow, parameterName=Dot Outer Glow - Outer Color, sortPriority=40, defaultValue=<rgba 0.287 0.0498 0.0058 1>}
  n22 = VectorParameter() {group=Dot - Outer Glow, parameterName=Dot Outer Glow - Inner Color, sortPriority=30, defaultValue=<rgba 1 0.4179 0 1>}
  n23 = LinearInterpolate(
   A=FunctionInput(Preview=@n21.RGB) {bUsePreviewValueAsDefault=true, inputName="│ Dot Glow Outer - Outer Color", sortPriority=150}
   B=FunctionInput(Preview=@n22.RGB) {bUsePreviewValueAsDefault=true, inputName="│ Dot Glow Outer - Inner Color", sortPriority=140}
   Alpha=SmoothStep(Value=NamedRerouteUsage())
  )
  n24 = NamedRerouteUsage()
  n25 = VectorParameter() {group=Dot - Inner Glow, parameterName=Dot Inner Glow - Outer Color, sortPriority=40, defaultValue=<rgba 0.3043 0.0413 0 1>}
  n26 = VectorParameter() {group=Dot - Inner Glow, parameterName=Dot Inner Glow - Inner Color, sortPriority=30, defaultValue=<rgba 1 0.1912 0 1>}
  n27 = LinearInterpolate(
   A=FunctionInput(Preview=@n25.RGB) {bUsePreviewValueAsDefault=true, inputName="│ Dot Glow Inner - Outer Color", sortPriority=100}
   B=FunctionInput(Preview=@n26.RGB) {bUsePreviewValueAsDefault=true, inputName="│ Dot Glow Inner - Inner Color", sortPriority=90}
   Alpha=SmoothStep(Value=NamedRerouteUsage())
  )
  n28 = NamedRerouteUsage()
  n29 = FunctionInput(Preview=VectorParameter() {group=Dot, parameterName=Dot - Color, sortPriority=70, defaultValue=<rgba 1 0.1912 0 1>}.RGB) {bUsePreviewValueAsDefault=true, inputName="│ Dot Color", sortPriority=50}
  n30 = NamedRerouteUsage()
  n31 = NamedRerouteUsage()
  n32 = StaticSwitch(True=Multiply(A=Multiply(A=@n29, B=@n30), B=Multiply(A=@n31, B=@n30)), False=Constant(), Value=NamedRerouteUsage())
  n33 = StaticSwitch(True=LinearInterpolate(A=@n32, B=Multiply(A=@n27, B=@n28), Alpha=NamedRerouteUsage()), False=@n32, Value=NamedRerouteUsage())
  n34 = StaticSwitch(True=LinearInterpolate(A=@n33, B=Multiply(A=@n23, B=@n24), Alpha=NamedRerouteUsage()), False=@n33, Value=NamedRerouteUsage())
  n35 = StaticSwitch(True=LinearInterpolate(A=@n34, B=Multiply(A=@n19, B=@n20), Alpha=NamedRerouteUsage()), False=@n34, Value=NamedRerouteUsage())
  NamedRerouteDeclaration(None=StaticSwitch(True=LinearInterpolate(A=@n35, B=@n15, Alpha=@n16), False=@n35, Value=NamedRerouteUsage())) {name=Final_Color, nodeColor=<rgba 0 1 0 1>}
  n36 = FunctionInput(Preview=ScalarParameter() {defaultValue=0.88, group=Dot, parameterName=Dot - Edge Sharpness, sortPriority=65}) {bUsePreviewValueAsDefault=true, inputName="│ Dot Edge Sharpness", sortPriority=45}
  n37 = MaterialFunctionCall(
   X=ScalarParameter() {defaultValue=0.5, group=Dot, parameterName=Dot - Width, sortPriority=30}
   Y=ScalarParameter() {defaultValue=0.5, group=Dot, parameterName=Dot - Height, sortPriority=40}
  ) {materialFunction=&MakeFloat2}
  n38 = FunctionInput(Preview=ScalarParameter() {group=Dot, parameterName=Dot - Scale Multiplier, sortPriority=50}) {bUsePreviewValueAsDefault=true, inputName="│ Dot Scale", inputType=FunctionInput_Vector2, sortPriority=30}
  n39 = MaterialFunctionCall!(
   UVs=@n2
   Size=Multiply(
    A=FunctionInput(Preview=@n37.Result) {bUsePreviewValueAsDefault=true, inputName="┌ Dot Size", inputType=FunctionInput_Vector2, sortPriority=20}
    B=LinearInterpolate(Alpha=@n38) {constA=1, constB=4}
   )
   CornerRoundness=FunctionInput(Preview=ScalarParameter() {defaultValue=0.2, group=Dot, parameterName=Dot - Corner Rounding, sortPriority=65}) {bUsePreviewValueAsDefault=true, inputName="│ Dot Corner Rounding", sortPriority=40}
   KeepAspectRatio=FunctionInput(Preview=StaticBool()) {bUsePreviewValueAsDefault=true, inputName="│ Keep Aspect Ratio?", inputType=FunctionInput_StaticBool, sortPriority=35}
  ) {materialFunction=/FortUILibrary/MaterialFunctions/Shapes/SDFs/MF_SDF_Box}
  n40 = NamedRerouteDeclaration(None=MaterialFunctionCall(DistanceField=@n39.SDF, Smoothing=OneMinus(None=@n36)) {materialFunction=&MF_UI_AntiAliasing}.Inner) {name=I_Shape_Dot, nodeColor=<rgba 0 0 0 1>}
  n41 = FunctionInput(Preview=ScalarParameter() {defaultValue=0.1, group=Dot - Outer Glow, parameterName=Dot Outer Glow - Radius, sortPriority=20}) {bUsePreviewValueAsDefault=true, inputName="┌\tOuter Glow Radius", inputType=FunctionInput_Scalar, sortPriority=130}
  n42 = Subtract(
   A=MaterialFunctionCall(DistanceField=@n39.SDF, Threshold=Multiply(A=@n41) {constB=0.25}, Smoothing=@n41) {materialFunction=&MF_UI_AntiAliasing}.Inner
   B=@n40
  )
  NamedRerouteDeclaration(None=Ceil(None=NamedRerouteDeclaration(None=@n42) {name=I_Glow_Dot_Outer_Mask, nodeColor=<rgba 0 0 0 1>})) {name=I_Glow_Dot_Outer_Range, nodeColor=<rgba 0 0 0 1>}
  n43 = FunctionInput(Preview=ScalarParameter() {defaultValue=0.5, group=Dot - Inner Glow, parameterName=Dot Inner Glow - Radius, sortPriority=20}) {bUsePreviewValueAsDefault=true, inputName="┌\tInner Glow Radius", inputType=FunctionInput_Scalar, sortPriority=80}
  n44 = Multiply(A=@n43) {constB=-0.25}
  n45 = Multiply(
   A=OneMinus(None=MaterialFunctionCall(DistanceField=@n39.SDF, Threshold=@n44, Smoothing=@n43) {materialFunction=&MF_UI_AntiAliasing}.Inner)
   B=@n40
  )
  NamedRerouteDeclaration(None=Ceil(None=NamedRerouteDeclaration(None=@n45) {name=I_Glow_Dot_Inner_Mask, nodeColor=<rgba 0 0 0 1>})) {name=I_Glow_Dot_Inner_Range, nodeColor=<rgba 0 0 0 1>}
  n46 = Multiply(A=NamedRerouteUsage(), B=NamedRerouteUsage())
  n47 = NamedRerouteUsage()
  n48 = NamedRerouteUsage()
  n49 = NamedRerouteUsage()
  n50 = NamedRerouteUsage()
  n51 = NamedRerouteUsage()
  n52 = NamedRerouteUsage()
  n53 = NamedRerouteUsage()
  n54 = NamedRerouteUsage()
  n55 = StaticSwitch(True=LinearInterpolate(Alpha=Multiply(A=@n53, B=@n54)), False=Constant(), Value=NamedRerouteUsage())
  n56 = StaticSwitch(True=LinearInterpolate(A=@n55, Alpha=Multiply(A=@n51, B=@n52)), False=@n55, Value=NamedRerouteUsage())
  n57 = StaticSwitch(True=LinearInterpolate(A=@n56, Alpha=Multiply(A=@n49, B=@n50)), False=@n56, Value=NamedRerouteUsage())
  n58 = StaticSwitch(True=LinearInterpolate(A=@n57, Alpha=Multiply(A=@n47, B=@n48)), False=@n57, Value=NamedRerouteUsage())
  NamedRerouteDeclaration(None=StaticSwitch(True=LinearInterpolate(A=@n58, Alpha=@n46), False=@n58, Value=NamedRerouteUsage())) {name=Final_Opacity, nodeColor=<rgba 0 1 0 1>}
  n59 = ScalarParameter() {defaultValue=0.8, group=Radial Glow - Front, parameterName=Front Radial Glow - Radius, sortPriority=20}
  n60 = MaterialFunctionCall(
   DistanceField=@n5
   Threshold=Multiply(A=FunctionInput(Preview=@n59) {bUsePreviewValueAsDefault=true, inputName="┌\tRadial Glow (Front) Radius", inputType=FunctionInput_Scalar, sortPriority=180}) {constB=0.25}
   Smoothing=Constant() {r=1.5}
  ) {materialFunction=&MF_UI_AntiAliasing}
  NamedRerouteDeclaration(None=Ceil(None=NamedRerouteDeclaration(None=@n60.Inner) {name=I_Glow_Front_Mask, nodeColor=<rgba 0 0 0 1>})) {name=I_Glow_Front_Range, nodeColor=<rgba 0 0 0 1>}
 ]
}
