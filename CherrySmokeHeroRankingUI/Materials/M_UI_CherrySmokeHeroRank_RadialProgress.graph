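# UI material for the CherrySmoke hero ranking UI that draws a radial progress bar on a perspective-skewed hexagonal outline.
# An antialiased band from a six-sided polygon SDF is clipped to a start/end degree range, forming the dark backer.
# Two radial mask calls cut the main progress, tinted with a three-color gradient, and a separately colored delta segment.
# Parameters expose percents, inversion, fill direction, shape and offset, with alpha-composite blending.
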Material /CherrySmokeHeroRankingUI/Materials/M_UI_CherrySmokeHeroRank_RadialProgress
graph {
 properties {blendMode=BLEND_AlphaComposite, materialDomain=MD_UI, pixelDepthOffsetMode=PDOM_Legacy, usedShadingModels=Default Lit}
 nodes [
  n1 = NamedRerouteUsage()
  n2 = NamedRerouteUsage()
  n3 = NamedRerouteUsage()
  n4 = NamedRerouteUsage()
  n5 = StaticBool() {value=true}
  n6 = NamedRerouteUsage()
  n7 = MaterialFunctionCall!(
   UVs=@n1
   Progress=NamedRerouteUsage()
   "Start Angle (Deg°)"=@n2
   "End Angle (Deg°)"=@n3
   Fill Clockwise=@n4
   Clamp Progress to Range=@n5
   Edge Softness=@n6
  ) {desc=Progress Delta, materialFunction=/CherrySmokeHeroRankingUI/PlayerRank/Materials/Functions/MF_UI_RadialMask_v2}
  n8 = Multiply(A=OneMinus(None=@n7.Range Mask), B=@n7.Range Mask) {desc=Range Fill}
  n9 = Constant() {r=0}
  n10 = Constant() {r=0}
  n11 = ScalarParameter() {defaultValue=0.24, group=Shape, parameterName=Bar - Outer Size, sortPriority=20}
  n12 = MaterialFunctionCall!(UV=NamedRerouteUsage(), Perspective Rotation=Constant() {r=0.25}, Perspective Amount=Constant() {r=0.1}) {materialFunction=/FortUILibrary/MaterialFunctions/Basics/MF_UI_Perspective}
  n13 = Constant2Vector() {g=0.94, r=1}
  n14 = MaterialFunctionCall!(
   UVs=Add(A=Multiply(A=@n12.Perspective UVs, B=@n13), B=Constant2Vector() {g=-0.016})
   Size=Constant() {r=0.54}
   Sides=Constant() {r=6}
   InnerAngle
   OutlineThickness=Constant() {r=1}
   OutlineAlignment
   GlowMin
   GlowMax
  ) {materialFunction=/FortUILibrary/MaterialFunctions/Shapes/MF_UI_RegularPolygon}
  n15 = ScalarParameter() {defaultValue=-0.12, group=Shape, parameterName=Bar - Inner Size, sortPriority=10}
  n16 = Subtract(
   A=MaterialFunctionCall(DistanceField=@n14.SDF, Threshold=@n11) {desc=Inner Shape, materialFunction=&MF_UI_AntiAliasing}.Outer
   B=MaterialFunctionCall(DistanceField=@n14.SDF, Threshold=Subtract(A=@n11, B=@n15)) {desc=Outer Shape, materialFunction=&MF_UI_AntiAliasing}.Outer
  ) {desc=Shape Outline}
  n17 = Multiply(A=Multiply(
    A=@n16
    B=MaterialFunctionCall(DistanceField=@n8, Threshold=@n9, Smoothing=@n10) {materialFunction=&MF_UI_AntiAliasing}.Outer
   ) {desc=Progress Range}) {constB=0.5}
  n18 = VectorParameter() {group=Color, parameterName=Bar Color - Delta, sortPriority=70, defaultValue=<rgba 1 0.4342 0.1248 1>}
  n19 = SmoothStep(Min=Constant() {r=0}, Max=Constant() {r=0}, Value=Multiply(A=OneMinus(None=@n7.Progress Mask), B=@n7.Progress Mask) {desc=Delta Radial})
  n20 = Constant() {r=0}
  n21 = Constant() {r=0}
  n22 = MaterialFunctionCall(
   UVs=@n1
   Progress=NamedRerouteUsage()
   "Start Angle (Deg°)"=@n2
   "End Angle (Deg°)"=@n3
   Fill Clockwise=@n4
   Clamp Progress to Range=@n5
   Edge Softness=@n6
  ) {desc=Main + Progress, materialFunction=&MF_UI_RadialMask_v2}
  n23 = Multiply(A=OneMinus(None=@n22.Progress Mask), B=@n22.Progress Mask) {desc=Progress Radial}
  n24 = OneMinus(None=NamedRerouteDeclaration(None=SmoothStep(Min=@n20, Max=@n21, Value=@n23)) {name=I_Shape_Radial_Progress, nodeColor=<rgba 0 0 0 1>}) {desc=Final Opacity}
  n25 = Multiply(A=Multiply(A=Multiply(A=@n19, B=@n16) {desc=Delta Fill}, B=@n24), B=NamedRerouteUsage()) {constB=0.5}
  n26 = VectorParameter() {group=Color, parameterName=Bar Color - Start, sortPriority=20, defaultValue=<rgba 1 0.0666 0 1>}
  n27 = VectorParameter() {group=Color, parameterName=Bar Color - Mid, sortPriority=40, defaultValue=<rgba 1 0.2086 0 1>}
  n28 = VectorParameter() {group=Color, parameterName=Bar Color - End, sortPriority=60, defaultValue=<rgba 1 0.7231 0 1>}
  n29 = Subtract(A=@n7.Range Mask, B=Step(Y=@n7.Progress Mask) {constX=0})
  n30 = Saturate(None=Multiply(A=Multiply(A=@n29) {constB=0.8}))
  n31 = Subtract(A=@n16, B=@n24) {desc=Progress Fill}
  LinearInterpolate(
   A=LinearInterpolate(B=Constant3Vector() {constant=<rgba 0.0182 0.0182 0.0182 1>}, Alpha=@n17) {desc=Bar Backer}
   B=LinearInterpolate(
    A=Multiply(A=@n18.RGB, B=@n25)
    B=MaterialFunctionCall(A=@n26.RGB, B=@n27.RGB, C=@n28.RGB, Alpha=@n30) {materialFunction=&3ColorBlend}.Result
    Alpha=@n31
   )
   Alpha=Max(A=@n31, B=@n25)
  ) {desc=Final Color}
  n32 = Multiply(A=@n31, B=NamedRerouteUsage()) {constB=0.5}
  Multiply(A=Add(A=Multiply(A=@n32) {constB=0.1}, B=@n17) {desc=Final Opacity})
  NamedRerouteDeclaration(None=ScalarParameter() {defaultValue=0.8, group=Progress, parameterName=Percent - Delta, sliderMax=1, sortPriority=22}) {name=P_PercentDelta, nodeColor=<rgba 0 0 0 1>}
  n33 = ScalarParameter() {defaultValue=0.6, group=Progress, parameterName=Percent, sliderMax=1, sortPriority=20}
  n34 = StaticSwitch(True=OneMinus(None=@n33), False=@n33, Value=StaticBoolParameter() {group=Progress, parameterName=Progress - Invert, sortPriority=28})
  NamedRerouteDeclaration(None=Multiply(A=@n34, B=Step(X=@n34) {constX=0.001, constY=0.001})) {name=P_Percent, nodeColor=<rgba 0 0 0 1>}
  NamedRerouteDeclaration(None=StaticBoolParameter() {defaultValue=true, group=Progress, parameterName=Progress - Fill Clockwise, sortPriority=24}) {name=I_Progress_Fill_Clockwise_Bool, nodeColor=<rgba 0 0 0 1>}
  NamedRerouteDeclaration(None=ScalarParameter() {defaultValue=250, group=Progress, parameterName="Bar Range - End (Degrees)", sliderMax=360, sortPriority=50}) {name=P_DegreeRange_End, nodeColor=<rgba 0 0 0 1>}
  NamedRerouteDeclaration(None=ScalarParameter() {defaultValue=0.002, group=Shape, parameterName=Bar - Edge Softness, sortPriority=40}) {name=P_EdgeSoftness_Angle, nodeColor=<rgba 0 0 0 1>}
  NamedRerouteDeclaration(None=ScalarParameter() {defaultValue=1, group=Progress, parameterName=Progress - Opacity, sliderMax=1, sortPriority=35}) {name=P_PercentOpacity, nodeColor=<rgba 0 0 0 1>}
  n35 = Add(
   A=TextureCoordinate()
   B=MaterialFunctionCall(
    X=ScalarParameter() {group=Shape, parameterName=Offset - X, sortPriority=400}
    Y=ScalarParameter() {group=Shape, parameterName=Offset - Y, sortPriority=410}
   ) {materialFunction=&MakeFloat2}.Result
  )
  NamedRerouteDeclaration(None=MaterialFunctionCall(
    UVs=NamedRerouteDeclaration(None=@n35) {name=I_UVs_Unscaled, nodeColor=<rgba 0 0 0 1>}
    Texture Scale=ScalarParameter() {defaultValue=0.5, group=Shape, parameterName=Bar - Scale, sortPriority=30}
   ) {materialFunction=&ScaleUVsByCenter}.UVs) {name=I_UVs, nodeColor=<rgba 0 0 0 1>}
  NamedRerouteDeclaration(None=ScalarParameter() {defaultValue=110, group=Progress, parameterName="Bar Range - Start (Degrees)", sliderMax=360, sortPriority=40}) {name=P_DegreeRange_Start, nodeColor=<rgba 0 0 0 1>}
 ]
}
