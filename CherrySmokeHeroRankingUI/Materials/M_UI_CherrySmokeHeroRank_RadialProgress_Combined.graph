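# UI material for the hero rank radial progress bar, drawing a perspective-skewed hexagonal ring.
# Radial mask functions clip the polygon SDF ring to a degree range, separating backer, progress fill and delta segment, with a three-color blend along the fill.
# The static "Gradient" switch swaps color and opacity to a glow variant built from the same SDF.
# A glow mask mixes scratch and noise textures, shifted in quantized time steps.
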Material /CherrySmokeHeroRankingUI/Materials/M_UI_CherrySmokeHeroRank_RadialProgress_Combined
graph {
 properties {blendMode=BLEND_AlphaComposite, materialDomain=MD_UI, pixelDepthOffsetMode=PDOM_Legacy, usedShadingModels=Default Lit}
 nodes [
  NamedRerouteDeclaration(None=ScalarParameter() {defaultValue=0.8, group=Progress, parameterName=Percent - Delta, sliderMax=1, sortPriority=22}) {name=P_PercentDelta, nodeColor=<rgba 0 0 0 1>}
  n1 = ScalarParameter() {defaultValue=0.6, group=Progress, parameterName=Percent, sliderMax=1, sortPriority=20}
  n2 = StaticSwitch(True=OneMinus(None=@n1), False=@n1, Value=StaticBoolParameter() {group=Progress, parameterName=Progress - Invert, sortPriority=28})
  NamedRerouteDeclaration(None=Multiply(A=@n2, B=Step(X=@n2) {constX=0.001, constY=0.001})) {name=P_Percent, nodeColor=<rgba 0 0 0 1>}
  NamedRerouteDeclaration(None=StaticBoolParameter() {defaultValue=true, group=Progress, parameterName=Progress - Fill Clockwise, sortPriority=24}) {name=I_Progress_Fill_Clockwise_Bool, nodeColor=<rgba 0 0 0 1>}
  NamedRerouteDeclaration(None=ScalarParameter() {defaultValue=250, group=Progress, parameterName="Bar Range - End (Degrees)", sliderMax=360, sortPriority=50}) {name=P_DegreeRange_End, nodeColor=<rgba 0 0 0 1>}
  NamedRerouteDeclaration(None=ScalarParameter() {defaultValue=0.002, group=Shape, parameterName=Bar - Edge Softness, sortPriority=40}) {name=P_EdgeSoftness_Angle, nodeColor=<rgba 0 0 0 1>}
  NamedRerouteDeclaration(None=ScalarParameter() {defaultValue=1, group=Progress, parameterName=Progress - Opacity, sliderMax=1, sortPriority=35}) {name=P_PercentOpacity, nodeColor=<rgba 0 0 0 1>}
  n3 = Add(
   A=TextureCoordinate()
   B=MaterialFunctionCall(
    X=ScalarParameter() {group=Shape, parameterName=Offset - X, sortPriority=400}
    Y=ScalarParameter() {group=Shape, parameterName=Offset - Y, sortPriority=410}
   ) {materialFunction=&MakeFloat2}.Result
  )
  NamedRerouteDeclaration(None=MaterialFunctionCall(
    UVs=NamedRerouteDeclaration(None=@n3) {name=I_UVs_Unscaled, nodeColor=<rgba 0 0 0 1>}
    Texture Scale=ScalarParameter() {defaultValue=0.5, group=Shape, parameterName=Bar - Scale, sortPriority=30}
   ) {materialFunction=&ScaleUVsByCenter}.UVs) {name=I_UVs, nodeColor=<rgba 0 0 0 1>}
  n4 = MaterialFunctionCall(UVs=NamedRerouteUsage(), Texture Scale=ScalarParameter() {defaultValue=0.5, group=Shape, parameterName=Bar - Scale, sortPriority=30}) {materialFunction=&ScaleUVsByCenter}
  n5 = Frac(None=Time())
  n6 = Multiply(A=@n5) {constB=0.5}
  n7 = Constant() {r=12}
  n8 = StaticBool()
  n9 = TextureSample(UVs=Add(
    A=MaterialFunctionCall(UV=@n4.UVs) {materialFunction=&MF_UI_UVScaling}.Scaled To Fill
    B=MaterialFunctionCall!("Param (0-1)"=@n6, Tiles=@n7, Apply Offset=@n8) {materialFunction=/Game/Environments/Papaya/Materials/MaterialFunctions/MF_Papaya_Quantize_Scalar}.Result
   )) {samplerType=SAMPLERTYPE_LinearGrayscale, texture=/CherrySmokeHeroRankingUI/Textures/T_UI_CherrySmokeHeroRank_Scratches}
  n10 = Multiply(A=@n5) {constB=2}
  n11 = Constant() {r=5}
  n12 = TextureSample(UVs=Add(
    A=MaterialFunctionCall(UV=@n4.UVs) {materialFunction=&MF_UI_UVScaling}.Scaled to Fit
    B=MaterialFunctionCall("Param (0-1)"=@n10, Tiles=@n11) {materialFunction=&MF_Papaya_Quantize_Scalar}.Result
   )) {samplerType=SAMPLERTYPE_Normal, texture=/Game/Global/Textures/Default/Noise/64x64Noise}
  NamedRerouteDeclaration(None=Add(A=Power(Base=@n9.R), B=Multiply(A=@n12.R) {constB=0.1})) {name=I_Glow_Scratch_Mask, nodeColor=<rgba 0 0 0 1>}
  NamedRerouteDeclaration(None=ScalarParameter() {defaultValue=110, group=Progress, parameterName="Bar Range - Start (Degrees)", sliderMax=360, sortPriority=40}) {name=P_DegreeRange_Start, nodeColor=<rgba 0 0 0 1>}
  n13 = NamedRerouteUsage()
  n14 = Multiply(A=ScalarParameter() {defaultValue=0.24, group=Shape, parameterName=Bar - Outer Size, sortPriority=20}) {constB=0.5}
  n15 = MaterialFunctionCall!(UV=NamedRerouteUsage(), Perspective Rotation=Constant() {r=0.25}, Perspective Amount=Constant() {r=0.1}) {materialFunction=/FortUILibrary/MaterialFunctions/Basics/MF_UI_Perspective}
  n16 = Constant2Vector() {g=0.94, r=1}
  n17 = Constant() {r=1}
  n18 = MaterialFunctionCall!(
   UVs=Add(A=Multiply(A=@n15.Perspective UVs, B=@n16), B=Constant2Vector() {g=-0.016})
   Size=Constant() {r=0.54}
   Sides=Constant() {r=6}
   InnerAngle
   OutlineThickness=@n17
   OutlineAlignment
   GlowMin
   GlowMax
  ) {materialFunction=/FortUILibrary/MaterialFunctions/Shapes/MF_UI_RegularPolygon}
  n19 = MaterialFunctionCall(DistanceField=@n18.SDF, Threshold=Add(A=@n14) {constB=0.64}, Smoothing=@n17) {materialFunction=&MF_UI_AntiAliasing}
  n20 = Multiply(A=OneMinus(None=@n19.Inner)) {constB=10}
  n21 = Add(A=@n14) {constB=0.2}
  n22 = Divide(A=Subtract(A=@n21, B=@n18.SDFOutline)) {constB=0.01}
  n23 = MaterialFunctionCall(Alpha=@n18.SDF, Min=@n21, Max=Add(A=@n21) {constB=0.01}) {desc=Inner Shape, materialFunction=&SmoothStep}
  n24 = Multiply(A=Subtract(A=Saturate(None=@n22), B=OneMinus(None=@n23.Result)), B=Multiply(A=Divide(B=@n20), B=Power(Base=@n19.Inner) {constExponent=2.5}))
  n25 = Multiply(A=Saturate(None=Multiply(A=@n24, B=@n13))) {constB=0.7}
  n26 = Multiply(A=@n25, B=NamedRerouteUsage())
  NamedRerouteDeclaration(None=Multiply(A=Add(A=@n25, B=@n26)) {constB=0.1}) {name=I_Glow_Mask, nodeColor=<rgba 0 0 0 1>}
  n27 = Constant() {r=0}
  n28 = Constant() {r=0}
  n29 = NamedRerouteUsage()
  n30 = NamedRerouteUsage()
  n31 = NamedRerouteUsage()
  n32 = NamedRerouteUsage()
  n33 = StaticBool() {value=true}
  n34 = NamedRerouteUsage()
  n35 = MaterialFunctionCall!(
   UVs=@n29
   Progress=NamedRerouteUsage()
   "Start Angle (Deg°)"=@n30
   "End Angle (Deg°)"=@n31
   Fill Clockwise=@n32
   Clamp Progress to Range=@n33
   Edge Softness=@n34
  ) {desc=Main + Progress, materialFunction=/CherrySmokeHeroRankingUI/PlayerRank/Materials/Functions/MF_UI_RadialMask_v2}
  n36 = Multiply(A=OneMinus(None=@n35.Progress Mask), B=@n35.Progress Mask) {desc=Progress Radial}
  n37 = OneMinus(None=NamedRerouteDeclaration(None=SmoothStep(Min=@n27, Max=@n28, Value=@n36)) {name=I_Shape_Radial_Progress, nodeColor=<rgba 0 0 0 1>}) {desc=Final Opacity}
  n38 = ScalarParameter() {defaultValue=0.24, group=Shape, parameterName=Bar - Outer Size, sortPriority=20}
  n39 = ScalarParameter() {defaultValue=-0.12, group=Shape, parameterName=Bar - Inner Size, sortPriority=10}
  n40 = Subtract(
   A=MaterialFunctionCall(DistanceField=@n18.SDF, Threshold=@n38) {desc=Inner Shape, materialFunction=&MF_UI_AntiAliasing}.Outer
   B=MaterialFunctionCall(DistanceField=@n18.SDF, Threshold=Subtract(A=@n38, B=@n39)) {desc=Outer Shape, materialFunction=&MF_UI_AntiAliasing}.Outer
  ) {desc=Shape Outline}
  n41 = Subtract(A=@n40, B=@n37) {desc=Progress Fill}
  n42 = Multiply(A=Multiply(A=@n41, B=NamedRerouteUsage()) {constB=0.5}) {constB=0.1}
  n43 = MaterialFunctionCall(
   UVs=@n29
   Progress=NamedRerouteUsage()
   "Start Angle (Deg°)"=@n30
   "End Angle (Deg°)"=@n31
   Fill Clockwise=@n32
   Clamp Progress to Range=@n33
   Edge Softness=@n34
  ) {desc=Progress Delta, materialFunction=&MF_UI_RadialMask_v2}
  n44 = Multiply(A=OneMinus(None=@n43.Range Mask), B=@n43.Range Mask) {desc=Range Fill}
  n45 = Constant() {r=0}
  n46 = Constant() {r=0}
  n47 = Multiply(A=Multiply(
    A=@n40
    B=MaterialFunctionCall(DistanceField=@n44, Threshold=@n45, Smoothing=@n46) {materialFunction=&MF_UI_AntiAliasing}.Outer
   ) {desc=Progress Range}) {constB=0.5}
  StaticSwitchParameter(True=Multiply(A=NamedRerouteUsage()) {constB=2}, False=Multiply(A=Add(A=@n42, B=@n47) {desc=Final Opacity})) {group=Shape, parameterName=Gradient, sortPriority=100}
  n48 = Constant3Vector() {constant=<rgba 0.0182 0.0182 0.0182 1>}
  n49 = VectorParameter() {group=Color, parameterName=Bar Color - Delta, sortPriority=70, defaultValue=<rgba 1 0.4342 0.1248 1>}
  n50 = SmoothStep(Min=Constant() {r=0}, Max=Constant() {r=0}, Value=Multiply(A=OneMinus(None=@n43.Progress Mask), B=@n43.Progress Mask) {desc=Delta Radial})
  n51 = Multiply(A=Multiply(A=Multiply(A=@n50, B=@n40) {desc=Delta Fill}, B=@n37), B=NamedRerouteUsage()) {constB=0.5}
  n52 = Multiply(A=@n49.RGB, B=@n51)
  n53 = Multiply(A=Subtract(A=@n43.Range Mask, B=Step(Y=@n43.Progress Mask) {constX=0})) {constB=0.8}
  n54 = MaterialFunctionCall(
   A=VectorParameter() {group=Color, parameterName=Bar Color - Start, sortPriority=20, defaultValue=<rgba 1 0.0666 0 1>}.RGB
   B=VectorParameter() {group=Color, parameterName=Bar Color - Mid, sortPriority=40, defaultValue=<rgba 1 0.2086 0 1>}.RGB
   C=VectorParameter() {group=Color, parameterName=Bar Color - End, sortPriority=60, defaultValue=<rgba 1 0.7231 0 1>}.RGB
   Alpha=Saturate(None=Multiply(A=@n53))
  ) {materialFunction=&3ColorBlend}
  StaticSwitchParameter(
   True=LinearInterpolate(A=Constant3Vector(), B=Constant3Vector() {constant=<rgba 2 0.2297 0 1>}, Alpha=NamedRerouteUsage()) {desc=Progress Bar Glow}
   False=LinearInterpolate(A=LinearInterpolate(B=@n48, Alpha=@n47) {desc=Bar Backer}, B=LinearInterpolate(A=@n52, B=@n54.Result, Alpha=@n41), Alpha=Max(A=@n41, B=@n51)) {desc=Final Color}
  ) {group=Shape, parameterName=Gradient, sortPriority=100}
 ]
}
