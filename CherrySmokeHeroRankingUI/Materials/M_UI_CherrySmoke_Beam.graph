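# Translucent UI material that draws an animatable soft-edged glowing beam, apparently for the CherrySmoke hero ranking screen.
# Color overlay-blends Color_A with a lerp between two hotspot tints.
# The alpha shape smoothsteps a sine-of-UV mask offset by beam height, optionally dotted with a tinted horizontal bell falloff, then applies alpha and taper controls.
# Anim_ scalars drive start/stop lerps; a separate rotatable gradient feeds a Beam reroute with no visible consumer.
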Material /CherrySmokeHeroRankingUI/Materials/M_UI_CherrySmoke_Beam
graph {
 properties {blendMode=BLEND_Translucent, materialDomain=MD_UI, pixelDepthOffsetMode=PDOM_Legacy, usedShadingModels=Default Lit}
 nodes [
  MaterialFunctionCall(
   Base=LinearInterpolate(
    A=VectorParameter() {parameterName=Color_B_Start, defaultValue=<rgba 0.3454 1 0.7959 1>}.RGB
    B=VectorParameter() {parameterName=Color_B_Stop, defaultValue=<rgba 0 0.8494 1 1>}.RGB
    Alpha=ScalarParameter() {parameterName=Anim_HotspotColor}
   )
   Blend=VectorParameter() {parameterName=Color_A, defaultValue=<rgba 0 0.9681 1 1>}.RGB
  ) {materialFunction=&Blend_Overlay}
  n1 = ScalarParameter() {defaultValue=0.5, parameterName=Gradient_Start}
  n2 = ScalarParameter() {parameterName=Gradient_Stop}
  n3 = ScalarParameter() {defaultValue=1, parameterName=Anim_Grow}
  n4 = Power(
   Base=MaterialFunctionCall(GradientStart=LinearInterpolate(A=@n1, B=@n2, Alpha=@n3), RotationAngle=ScalarParameter() {defaultValue=-0.5, parameterName=Rotation}) {materialFunction=&MF_UI_LinearGradient}.Result
   Exp=ScalarParameter() {defaultValue=2, parameterName=BeamFadeRate}
  )
  NamedRerouteDeclaration(None=Multiply(A=Saturate(None=@n4), B=ScalarParameter() {defaultValue=1, parameterName=BeamAlpha})) {name=Beam, nodeColor=<rgba 0 1 0.1094 1>}
  n5 = Sine(None=TextureCoordinate()) {period=2}
  n6 = LinearInterpolate(
   A=ScalarParameter() {defaultValue=1, parameterName=BeamHeight_Start}
   B=ScalarParameter() {parameterName=BeamHeight_Stop}
   Alpha=ScalarParameter() {parameterName=Anim_BeamHeight}
  )
  n7 = SmoothStep(
   Max=ScalarParameter() {defaultValue=0.3, parameterName=EdgeSoftness}
   Value=Subtract(A=StaticComponentMaskParameter(None=@n5) {defaultG=true, parameterName=EdgeSoftenessMaskDirection}, B=OneMinus(None=@n6)) {constB=0.25}
  )
  n8 = MaterialFunctionCall(UV Channel=TextureCoordinate()) {materialFunction=&LinearGradient}
  n9 = OneMinus(None=@n8.UGradient)
  n10 = Multiply(
   A=Power(Base=Multiply(A=@n8.UGradient, B=@n9), Exp=ScalarParameter() {defaultValue=3.5, parameterName=Power}) {constExponent=3}
   B=ScalarParameter() {defaultValue=130, parameterName=Multiply}
  ) {constA=60}
  n11 = NamedRerouteDeclaration(None=Multiply(A=VectorParameter() {parameterName=Color, defaultValue=<rgba 0 0.2222 0.6667 1>}.RGB, B=Saturate(None=@n10))) {nodeColor=<rgba 0 1 0.5078 1>}
  n12 = Multiply(
   A=StaticSwitchParameter(True=DotProduct(A=@n7, B=@n11), False=@n7) {defaultValue=true, parameterName=HasFallOff}
   B=LinearInterpolate(
    A=ScalarParameter() {defaultValue=1, parameterName=Alpha_Start}
    B=ScalarParameter() {defaultValue=1, parameterName=Alpha_Stop}
    Alpha=ScalarParameter() {defaultValue=1, parameterName=Anim_Alpha}
   )
  )
  n13 = ScalarParameter() {defaultValue=1, parameterName=MultiplyGradient}
  n14 = Power(Base=Saturate(None=Multiply(A=@n12, B=@n13)), Exp=ScalarParameter() {defaultValue=1, parameterName=TaperAlphaGradient})
  Saturate(None=Multiply(A=Saturate(None=@n14), B=ScalarParameter() {defaultValue=1.5, parameterName=FinalAlpha}))
 ]
}
