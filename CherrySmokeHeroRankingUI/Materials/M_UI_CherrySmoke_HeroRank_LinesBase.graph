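# Translucent UI base material for the CherrySmoke hero ranking UI that draws animated, glitchy lines in a two-color vertical gradient.
# UVs warped by a panning glitch noise texture, with sine-pulsed strength, scroll vertically and feed a segmented ping-pong, thresholded through anti-aliasing into soft horizontal lines.
# A circle SDF on similarly warped UVs, scaled by a negated density, masks the lines before a global opacity multiply.
# Parameters expose colors, speeds, segment count, thickness, glitch and circle shape.
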Material /CherrySmokeHeroRankingUI/Materials/M_UI_CherrySmoke_HeroRank_LinesBase
graph {
 properties {bAutomaticallySetUsageInEditor=false, blendMode=BLEND_Translucent, materialDomain=MD_UI, usedShadingModels=Default Lit}
 nodes [
  LinearInterpolate(
   A=VectorParameter() {parameterName=Rarity Color 03, defaultValue=<rgba 0.5906 0.3712 0.0704 1>}.RGB
   B=VectorParameter() {parameterName=Color 01, defaultValue=<rgba 0.1529 0.5089 0.7682 1>}.RGB
   Alpha=MaterialFunctionCall() {materialFunction=&LinearGradient}.VGradient
  )
  n1 = MaterialFunctionCall() {materialFunction=&GetUserInterfaceUV}
  n2 = Multiply(A=@n1.Normalized UV) {constB=0.4}
  n3 = ComponentMask(None=TextureSample(UVs=Panner(Coordinate=@n2) {speedX=-0.02}) {samplerType=SAMPLERTYPE_LinearGrayscale, texture=/Game/Global/Textures/Default/Noise/glitch_noise}.RGB) {g=true, r=true}
  n4 = ScalarParameter() {defaultValue=0.02, parameterName=GlitchIntensity}
  n5 = MaterialFunctionCall(Speed=ScalarParameter() {defaultValue=1, parameterName=GlitchSpeed}) {materialFunction=&DebugTimeSine}
  n6 = Panner(
   Coordinate=LinearInterpolate(A=@n1.Normalized UV, B=@n3, Alpha=LinearInterpolate(B=@n4, Alpha=@n5.Result) {constB=0.003})
   Time=Multiply(A=Time(), B=ScalarParameter() {defaultValue=0.8, parameterName=LinesSpeed})
  ) {speedY=0.05}
  n7 = ScalarParameter() {defaultValue=15, parameterName=SegmentCount}
  n8 = MaterialFunctionCall!(Value=ComponentMask(None=Multiply(A=@n6, B=@n7)) {g=true}, Scale, Height) {materialFunction=/FortUILibrary/MaterialFunctions/Basics/MF_UI_PingPong}
  n9 = Constant() {r=0.1}
  n10 = ScalarParameter() {defaultValue=0.3, parameterName=Line Thickness}
  n11 = Constant() {r=0.1}
  n12 = MaterialFunctionCall!(
   UVs=LinearInterpolate(
    A=MaterialFunctionCall() {materialFunction=&GetUserInterfaceUV}.Normalized UV
    B=@n3
    Alpha=LinearInterpolate(B=@n11, Alpha=@n5.Result) {constB=0.003}
   )
   Radius=ScalarParameter() {defaultValue=0.8, parameterName=CircleRadius}
  ) {materialFunction=/FortUILibrary/MaterialFunctions/Shapes/SDFs/MF_SDF_Circle}
  n13 = Multiply(B=ScalarParameter() {defaultValue=5, parameterName=CircleDensity}) {constA=-1}
  Multiply(
   A=Multiply(
    A=MaterialFunctionCall(DistanceField=@n8.Result, Threshold=@n9, Smoothing=@n10) {materialFunction=&MF_UI_AntiAliasing}.Inner
    B=Multiply(A=@n12.SDF, B=@n13) {constB=-60}
   )
   B=ScalarParameter() {defaultValue=1, parameterName=Opacity}
  )
 ]
}
