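# Translucent UI material that draws animated, segmented outer rings for the CherrySmoke hero ranking display.
# Circle SDF outlines are cut into arcs by angle gradients, with Time-driven rotators spinning or swaying them.
# Radial segments form a dotted ring via smooth SDF union, while panned polar noise breaks up the middle ring.
# Layers combine by Max with per-ring opacity parameters; color blends blue shades with a rarity color through a circle mask.
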Material /CherrySmokeHeroRankingUI/Materials/M_UI_CherrySmoke_HeroRanking_OuterRing
graph {
 properties {bAutomaticallySetUsageInEditor=false, blendMode=BLEND_Translucent, materialDomain=MD_UI, usedShadingModels=Default Lit}
 nodes [
  n1 = Constant3Vector() {constant=<rgba 0.2346 0.5776 1 1>}
  n2 = Constant3Vector() {constant=<rgba 0.0999 0.4452 1 1>}
  n3 = NamedRerouteUsage()
  LinearInterpolate(
   A=LinearInterpolate(A=LinearInterpolate(A=@n1, B=@n2, Alpha=@n3), B=Constant3Vector() {constant=<rgba 0.0482 0.3467 1 1>}, Alpha=NamedRerouteUsage())
   B=VectorParameter() {parameterName=Rarity Color 2, defaultValue=<rgba 1 0.9216 0.2051 1>}.RGB
   Alpha=MaterialFunctionCall(Size=Constant() {r=0.775}) {materialFunction=&MF_UI_Circle}.Fill
  )
  n4 = Max(
   A=Multiply(A=NamedRerouteUsage(), B=ScalarParameter() {defaultValue=0.8, parameterName=CircleDots})
   B=Multiply(A=NamedRerouteUsage(), B=ScalarParameter() {defaultValue=0.4, parameterName=YellowRingOpacity})
  )
  n5 = Multiply(A=NamedRerouteUsage(), B=ScalarParameter() {defaultValue=0.25, parameterName=OuterRingOpacity}) {constB=0.25}
  n6 = NamedRerouteUsage()
  n7 = ScalarParameter() {defaultValue=0.2, parameterName=ThinnerRing}
  Max(
   A=Max(A=Max(A=@n4, B=@n5), B=Multiply(A=@n6, B=@n7))
   B=Multiply(A=NamedRerouteUsage(), B=ScalarParameter() {defaultValue=0.15, parameterName=MiddleRing})
  )
  n8 = NamedRerouteUsage()
  n9 = Constant() {r=0.7}
  n10 = Constant() {r=0.018}
  n11 = OneMinus(None=Add(A=MaterialFunctionCall(UVs=@n8, Size=@n9, OutlineThickness=@n10) {materialFunction=&MF_UI_Circle}.SDFOutline) {constB=0.05})
  n12 = Constant() {r=35}
  n13 = Constant() {r=0.75}
  n14 = Multiply(A=MaterialFunctionCall!(
    DistanceFieldA=SmoothStep(Value=@n11) {constMax=0.98, constMin=0.95}
    DistanceFieldB=MaterialFunctionCall!(UV, Segment Count=@n12, Segment Spacing=@n13, Fixed Spacing) {materialFunction=/FortUILibrary/MaterialFunctions/Basics/MF_UI_RadialSegments}.Segmented Triangles Gradient
    BlendAmount=Constant() {r=0.8}
   ) {materialFunction=/FortUILibrary/MaterialFunctions/Shapes/Modifiers/MF_UI_SDF_Union}.Smooth) {constB=4}
  n15 = NamedRerouteUsage()
  n16 = Constant() {r=0.01}
  n17 = Ceil(None=Add(A=MaterialFunctionCall!(UVs=@n15, Rotation=@n16) {materialFunction=/FortUILibrary/MaterialFunctions/Gradients/MF_UI_AngleGradient}.Result) {constB=-0.7})
  NamedRerouteDeclaration(None=Saturate(None=Multiply(A=@n14, B=@n17))) {name=CircleDots, nodeColor=<rgba 0 0 0 1>}
  n18 = Time()
  NamedRerouteDeclaration(None=MaterialFunctionCall(UVs=NamedRerouteUsage(), "Rotation Angle (0-1)"=Multiply(A=@n18) {constB=0.1}) {materialFunction=&CustomRotator}.Rotated Values) {name=CircleDotsRotation, nodeColor=<rgba 0 0 0 1>}
  n19 = Sine(None=Multiply(A=Time()) {constB=0.1})
  NamedRerouteDeclaration(None=MaterialFunctionCall(UVs=NamedRerouteUsage(), "Rotation Angle (0-1)"=LinearInterpolate(Alpha=@n19) {constB=0.03}) {materialFunction=&CustomRotator}.Rotated Values) {name=YellowRingTime01, nodeColor=<rgba 0 0 0 1>}
  n20 = Constant() {r=0.83}
  n21 = Constant() {r=0.026}
  n22 = OneMinus(None=Add(A=MaterialFunctionCall(Size=@n20, OutlineThickness=@n21) {materialFunction=&MF_UI_Circle}.SDFOutline) {constB=0.05})
  n23 = Multiply(A=Time()) {constB=0.05}
  n24 = MaterialFunctionCall(UVs=NamedRerouteUsage(), "Rotation Angle (0-1)"=LinearInterpolate(Alpha=Sine(None=@n23)) {constA=0.02, constB=0.05}) {materialFunction=&CustomRotator}
  n25 = Constant() {r=-0.06}
  n26 = Ceil(None=Add(A=MaterialFunctionCall(UVs=@n24.Rotated Values, Rotation=@n25) {materialFunction=&MF_UI_AngleGradient}.Result) {constB=-0.88})
  n27 = Constant() {r=-0.2}
  n28 = Ceil(None=Add(A=MaterialFunctionCall(UVs=@n24.Rotated Values, Rotation=@n27) {materialFunction=&MF_UI_AngleGradient}.Result) {constB=-0.95})
  NamedRerouteDeclaration(None=Multiply(A=SmoothStep(Value=@n22) {constMax=0.96, constMin=0.95}, B=Max(A=@n26, B=@n28))) {name=ThickerOuterRing, nodeColor=<rgba 0 0 0 1>}
  n29 = Constant() {r=0.78}
  n30 = Constant() {r=0.014}
  n31 = OneMinus(None=Add(A=MaterialFunctionCall(Size=@n29, OutlineThickness=@n30) {materialFunction=&MF_UI_Circle}.SDFOutline) {constB=0.05})
  n32 = NamedRerouteUsage()
  n33 = Time()
  n34 = LinearInterpolate(Alpha=Sine(None=Multiply(A=@n33) {constB=0.1})) {constB=0.03}
  n35 = Add(A=MaterialFunctionCall(
    UVs=MaterialFunctionCall(UVs=@n32, "Rotation Angle (0-1)"=@n34) {materialFunction=&CustomRotator}.Rotated Values
    Rotation=Constant() {r=0.02}
   ) {materialFunction=&MF_UI_AngleGradient}.Result) {constB=-0.7}
  NamedRerouteDeclaration(None=Multiply(A=SmoothStep(Value=@n31) {constMax=0.96, constMin=0.95}, B=Ceil(None=@n35))) {name=Thinner Ring, nodeColor=<rgba 0 0 0 1>}
  NamedRerouteDeclaration(None=MaterialFunctionCall(Size=Constant() {r=0.775}) {materialFunction=&MF_UI_Circle}.Fill) {name=CircleMask02, nodeColor=<rgba 0 0 0 1>}
  n36 = Constant() {r=0.76}
  n37 = Constant() {r=0.015}
  n38 = OneMinus(None=Add(A=MaterialFunctionCall(Size=@n36, OutlineThickness=@n37) {materialFunction=&MF_UI_Circle}.SDFOutline) {constB=0.05})
  n39 = MaterialFunctionCall(UVs=NamedRerouteUsage(), Rotation=Constant() {r=0.42}) {materialFunction=&MF_UI_AngleGradient}
  n40 = MaterialFunctionCall(UVs=NamedRerouteUsage(), Rotation=Constant() {r=0.28}) {materialFunction=&MF_UI_AngleGradient}
  n41 = Subtract(A=Ceil(None=Add(A=@n39.Result) {constB=-0.75}), B=Ceil(None=Add(A=@n40.Result) {constB=-0.95}))
  NamedRerouteDeclaration(None=Multiply(A=SmoothStep(Value=@n38) {constMax=0.96, constMin=0.95}, B=Saturate(None=@n41))) {name=YellowRing, nodeColor=<rgba 0 0 0 1>}
  n42 = Constant() {r=0.8}
  n43 = Constant() {r=0.02}
  n44 = OneMinus(None=Add(A=MaterialFunctionCall(Size=@n42, OutlineThickness=@n43) {materialFunction=&MF_UI_Circle}.SDFOutline) {constB=0.05})
  n45 = Constant() {r=0.55}
  n46 = Ceil(None=Add(A=MaterialFunctionCall(Rotation=@n45) {materialFunction=&MF_UI_AngleGradient}.Result) {constB=-0.62})
  n47 = Time()
  n48 = Multiply(A=Panner(
    Coordinate=MaterialFunctionCall() {materialFunction=&VectorToRadialValue}.Vector Converted to Angle
    Time=Multiply(A=@n47) {constB=0.2}
   ) {speedX=0.1}) {constB=2}
  n49 = SmoothStep(Value=Power(Base=TextureSample(UVs=@n48) {samplerType=SAMPLERTYPE_Grayscale, texture=&LowResBlurredNoise}.RGB)) {constMax=0.11, constMin=0.1}
  NamedRerouteDeclaration(None=Multiply(A=SmoothStep(Value=@n44) {constMax=0.96, constMin=0.95}, B=Multiply(A=@n46, B=@n49))) {name=MiddleRing, nodeColor=<rgba 0 0 0 1>}
  NamedRerouteDeclaration(None=MaterialFunctionCall() {materialFunction=&GetUserInterfaceUV}.Normalized UV) {name=UV, nodeColor=<rgba 0 0 0 1>}
  NamedRerouteDeclaration(None=MaterialFunctionCall(UVs=NamedRerouteUsage(), "Rotation Angle (0-1)"=LinearInterpolate(Alpha=@n19) {constB=0.05}) {materialFunction=&CustomRotator}.Rotated Values) {name=YellowRingTime02, nodeColor=<rgba 0 0 0 1>}
 ]
}
