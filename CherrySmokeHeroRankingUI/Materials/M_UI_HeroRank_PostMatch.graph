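# Translucent UI material that draws a procedural decorative plate for the post-match hero rank display.
# Bracket, outer frame, dot and background masks come from SDF and box functions on 9-slice UVs, named via reroutes.
# An opacity-weighted Max merges the masks, and a subtracted linear gradient fades the result.
# A separate Lerp chain builds color from white, orange and mask sums.
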Material /CherrySmokeHeroRankingUI/Materials/M_UI_HeroRank_PostMatch
graph {
 properties {blendMode=BLEND_Translucent, materialDomain=MD_UI, pixelDepthOffsetMode=PDOM_Legacy, usedShadingModels=Default Lit}
 nodes [
  n1 = NamedRerouteUsage()
  n2 = ScalarParameter() {defaultValue=1, parameterName=Left Bracket Opacity}
  LinearInterpolate(
   A=LinearInterpolate(A=Constant4Vector() {constant=<rgba 1 1 1 0>}.RGBA, B=Constant4Vector() {constant=<rgba 1 0.1413 0 1>}.RGBA, Alpha=NamedRerouteUsage())
   B=Add(A=Multiply(A=@n1, B=@n2), B=NamedRerouteUsage())
   Alpha=NamedRerouteUsage()
  ) {constAlpha=0}
  n3 = ScalarParameter() {defaultValue=-0.45, parameterName=Left Bracket - X position}
  n4 = Constant()
  n5 = MaterialFunctionCall(
   UVs=Add(A=MaterialFunctionCall!(UVs) {materialFunction=/FortUILibrary/MaterialFunctions/Basics/MF_UI_9SliceUVs}.UVs, B=AppendVector(A=@n3, B=@n4))
   "Rotation Angle (0-1)"=Constant()
  ) {materialFunction=&CustomRotator}
  n6 = MaterialFunctionCall(UVs=@n5.Rotated Values, "Rotation Angle (0-1)"=Constant() {r=0.75}) {materialFunction=&CustomRotator}
  n7 = Constant2Vector() {g=-0.1}
  n8 = Constant() {r=0.6}
  n9 = Multiply(A=MaterialFunctionCall!(UVs=Add(A=@n6.Rotated Values, B=@n7) {constB=0.02}, TopWidth=Subtract(A=@n8) {constB=0.3}, BottomWidth=@n8, Height=Constant() {r=0.2}) {materialFunction=/FortUILibrary/MaterialFunctions/Shapes/SDFs/MF_SDF_IsoscelesTrapezoid}.SDF) {constB=-1}
  n10 = MaterialFunctionCall!(UVs=@n5.Rotated Values, Size=Constant2Vector() {g=0.63, r=0.04}, CornerRoundness, KeepAspectRatio, PixelSize) {materialFunction=/FortUILibrary/MaterialFunctions/Shapes/SDFs/MF_SDF_Box}
  NamedRerouteDeclaration(None=MaterialFunctionCall(DistanceField=Max(A=@n9, B=@n10.SDF)) {desc=OuterFlareMask, materialFunction=&MF_UI_AntiAliasing}.Inner) {name=Left Bracket, nodeColor=<rgba 0.75 0.6098 0.375 0.75>}
  n11 = MaterialFunctionCall(
   UVs=Add(A=MaterialFunctionCall() {materialFunction=&MF_UI_9SliceUVs}.UVs, B=Constant2Vector() {r=0.02})
   Size=Constant2Vector() {g=0.6, r=0.83}
   CornerRoundness=Constant4Vector() {constant=<rgba 0 0 0.15 0.15>}.RGBA
   OutlineThickness=ScalarParameter() {defaultValue=0.03, parameterName=Outer Frame Thickness}
  ) {materialFunction=&MF_UI_Box}
  n12 = MaterialFunctionCall(
   UVs=Add(
    A=MaterialFunctionCall() {materialFunction=&MF_UI_9SliceUVs}.UVs
    B=Constant2Vector() {g=0.11, r=0.29}
   )
   Size=Constant2Vector() {g=0.795, r=0.8}
  ) {materialFunction=&MF_UI_Box}
  NamedRerouteDeclaration(None=Saturate(None=Subtract(A=@n11.Outline, B=@n12.Fill))) {name=Outer Frame, nodeColor=<rgba 0.7529 0.6105 0.3763 0.749>}
  n13 = MaterialFunctionCall() {materialFunction=&MF_UI_9SliceUVs}
  n14 = Constant2Vector() {g=0.3, r=-0.06}
  n15 = ScalarParameter() {defaultValue=0.05, parameterName=Dot Size}
  n16 = ScalarParameter() {defaultValue=0.05, parameterName=Dot Size}
  NamedRerouteDeclaration(None=MaterialFunctionCall(UVs=Add(A=@n13.UVs, B=@n14), Size=AppendVector(A=@n15, B=@n16), CornerRoundness=Constant() {r=0.02}) {materialFunction=&MF_UI_Box}.Fill) {name=Dot 01, nodeColor=<rgba 0.7529 0.6105 0.3763 0.749>}
  n17 = MaterialFunctionCall() {materialFunction=&MF_UI_9SliceUVs}
  n18 = Constant2Vector() {r=0.15}
  n19 = ScalarParameter() {defaultValue=1, parameterName=Background Size X}
  n20 = Constant() {r=0.5}
  NamedRerouteDeclaration(None=MaterialFunctionCall(UVs=Add(A=@n17.UVs, B=@n18), Size=AppendVector(A=@n19, B=@n20), CornerRoundness=Constant4Vector() {constant=<rgba 0 0 0.1 0.1>}.RGBA) {materialFunction=&MF_UI_Box}.Fill) {name=Background, nodeColor=<rgba 0.75 0.6098 0.375 0.75>}
  n21 = Multiply(A=NamedRerouteUsage(), B=ScalarParameter() {defaultValue=1, parameterName=Left Bracket Opacity})
  n22 = Multiply(A=NamedRerouteUsage(), B=ScalarParameter() {defaultValue=1, parameterName=Outer Frame Opacity})
  n23 = Max(
   A=Multiply(A=NamedRerouteUsage(), B=ScalarParameter() {defaultValue=1, parameterName=Dot Opacity})
   B=Multiply(A=Max(A=@n21, B=@n22)) {constB=0.5}
  )
  n24 = Multiply(A=NamedRerouteUsage(), B=ScalarParameter() {defaultValue=0.5, parameterName=Background Opacity}) {constB=0.5}
  n25 = MaterialFunctionCall() {materialFunction=&MF_UI_9SliceUVs}
  n26 = Constant() {r=0.48}
  n27 = Constant() {r=0.38}
  Saturate(None=Subtract(
    A=Max(A=@n23, B=@n24)
    B=MaterialFunctionCall(UVs=@n25.UVs, GradientStart=@n26, GradientEnd=@n27) {materialFunction=&MF_UI_LinearGradient}.Result
   ))
  TextureCoordinate()
 ]
}
