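# Translucent UI font material for the hero ranking UI, giving text a two-tone vertical gradient.
# A LinearGradient VGradient blends white with a soft salmon pink to form the color.
# The font's signed distance field feeds an anti-aliasing material function with a low constant threshold, producing smoothed glyph edges.
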Material /CherrySmokeHeroRankingUI/Materials/M_UI_HeroRanking_Font
graph {
 properties {bAutomaticallySetUsageInEditor=false, blendMode=BLEND_Translucent, materialDomain=MD_UI, usedShadingModels=Default Lit}
 nodes [
  LinearInterpolate(
   A=Constant3Vector() {constant=<rgba 1 1 1 0>}
   B=Constant3Vector() {constant=<rgba 0.8714 0.3989 0.3485 1>}
   Alpha=MaterialFunctionCall() {materialFunction=&LinearGradient}.VGradient
  )
  MaterialFunctionCall(DistanceField=FontSignedDistance().Signed Distance, Threshold=Constant() {r=0.04}) {materialFunction=&MF_UI_AntiAliasing}
 ]
}
