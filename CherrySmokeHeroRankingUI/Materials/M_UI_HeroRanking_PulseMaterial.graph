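# Translucent UI material for the hero ranking UI that draws a soft pink ring pulsing outward on a five-second loop.
# A looping time fraction drives two MF_UI_Circle fills, whose difference forms the ring.
# The inner circle grows slightly faster than the outer, so the ring thins as it expands.
# The same fraction, inverted and scaled down, fades the ring out over each cycle.
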Material /CherrySmokeHeroRankingUI/Materials/M_UI_HeroRanking_PulseMaterial
graph {
 properties {bAutomaticallySetUsageInEditor=false, blendMode=BLEND_Translucent, materialDomain=MD_UI, usedShadingModels=Default Lit}
 nodes [
  Constant3Vector() {constant=<rgba 0.8714 0.3989 0.3485 1>}
  n1 = Frac(None=Multiply(A=Time()) {constB=0.2})
  n2 = Multiply(A=Constant() {r=0.15}, B=@n1)
  n3 = MaterialFunctionCall(Size=Add(A=Constant() {r=0.59}, B=@n2) {constB=0}) {materialFunction=&MF_UI_Circle}
  n4 = Multiply(A=Constant() {r=0.03}, B=@n1)
  n5 = MaterialFunctionCall(Size=Add(A=Constant() {r=0.555}, B=Add(A=@n2, B=@n4)) {constB=0}) {materialFunction=&MF_UI_Circle}
  n6 = Multiply(A=@n1)
  Multiply(A=Multiply(A=Subtract(A=@n3.Fill, B=@n5.Fill), B=Subtract(B=@n6))) {constB=0.35}
 ]
}
