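# Translucent UI material that draws the hero ranking squad icon in a tintable fill color with a drop shadow.
# A grayscale distance-field texture passes through MF_UI_AntiAliasing to form a crisp icon mask, which blends in the FillColor parameter, orange by default.
# A second sample with offset UVs forms a faint copy shifted down-right, combined with the main mask by Max to add the shadow.
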Material /CherrySmokeHeroRankingUI/Materials/M_UI_HeroRanking_SquadIcon_Mat
graph {
 properties {bAutomaticallySetUsageInEditor=false, blendMode=BLEND_Translucent, materialDomain=MD_UI, usedShadingModels=Default Lit}
 nodes [
  n1 = Constant() {r=0.12}
  n2 = MaterialFunctionCall(
   DistanceField=TextureSample() {samplerType=SAMPLERTYPE_Grayscale, texture=/CherrySmokeHeroRankingUI/Textures/T_UI_HeroRanking_SquadIcon}.RGB
   Threshold=@n1
  ) {materialFunction=&MF_UI_AntiAliasing}
  LinearInterpolate(B=VectorParameter() {parameterName=FillColor, defaultValue=<rgba 1 0.3231 0 1>}.RGB, Alpha=@n2.Outer)
  n3 = TextureCoordinate()
  n4 = ComponentMask(None=@n3) {r=true}
  n5 = ComponentMask(None=@n3) {g=true}
  n6 = TextureSample(UVs=AppendVector(A=Add(A=@n4) {constB=-0.06}, B=Add(A=@n5) {constB=-0.06})) {samplerType=SAMPLERTYPE_Grayscale, texture=&T_UI_HeroRanking_SquadIcon}
  Max(
   A=@n2.Outer
   B=Multiply(A=MaterialFunctionCall(DistanceField=@n6.RGB, Threshold=@n1) {materialFunction=&MF_UI_AntiAliasing}.Outer) {constB=0.3}
  )
 ]
}
